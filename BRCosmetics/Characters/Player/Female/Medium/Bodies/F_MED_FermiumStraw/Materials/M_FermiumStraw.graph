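# Character material for the FermiumStraw medium female player body, assembling several surface layers into one parameterized look.
# Chains of BlendMaterialAttributes stack the layers by named-reroute masks, then scalar parameters blend further attribute sets over the result.
# Parallel lerp chains use masks to select per-region colors and values, including flat constant colors.
# Coordinate math with U Min/Max ranges, floor and frac appears alongside static switches and desaturated texture detail.
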Material /BRCosmetics/Characters/Player/Female/Medium/Bodies/F_MED_FermiumStraw/Materials/M_FermiumStraw
graph {
 expressions {
  MaterialExpressionAbs_0 {class=MaterialExpressionAbs, inputs [None {from=MaterialExpressionScalarParameter_6, output=""}], properties {}}
  MaterialExpressionAdd_0 {
   class=MaterialExpressionAdd
   inputs [
    A {from=MaterialExpressionMultiply_24}
    B {from=MaterialExpressionScalarParameter_25}
   ]
  }
  MaterialExpressionAdd_1 {
   class=MaterialExpressionAdd
   inputs [
    A {from=MaterialExpressionDivide_8}
    B {from=""}
   ]
   properties {constB=0.5}
  }
  MaterialExpressionAdd_10 {
   class=MaterialExpressionAdd
   inputs [
    A {from=MaterialExpressionScalarParameter_31}
    B {from=MaterialExpressionMultiply_11}
   ]
  }
  MaterialExpressionAdd_11 {
   class=MaterialExpressionAdd
   inputs [
    A {from=MaterialExpressionScalarParameter_29}
    B {from=MaterialExpressionMultiply_14}
   ]
  }
  MaterialExpressionAdd_12 {
   class=MaterialExpressionAdd
   inputs [
    A {from=MaterialExpressionScalarParameter_28}
    B {from=MaterialExpressionMultiply_48}
   ]
  }
  MaterialExpressionAdd_13 {
   class=MaterialExpressionAdd
   inputs [
    A {from=MaterialExpressionNamedRerouteUsage_23}
    B {from=MaterialExpressionVectorParameter_19, output=R}
   ]
  }
  MaterialExpressionAdd_2 {
   class=MaterialExpressionAdd
   inputs [
    A {from=MaterialExpressionMultiply_12}
    B {from=MaterialExpressionMultiply_0}
   ]
  }
  MaterialExpressionAdd_3 {
   class=MaterialExpressionAdd
   inputs [
    A {from=MaterialExpressionConvert_5}
    B {from=MaterialExpressionReroute_14}
   ]
  }
  MaterialExpressionAdd_5 {
   class=MaterialExpressionAdd
   inputs [
    A {from=MaterialExpressionNamedRerouteUsage_33}
    B {from=MaterialExpressionVectorParameter_11, output=R}
   ]
  }
  MaterialExpressionAdd_6 {
   class=MaterialExpressionAdd
   inputs [
    A {from=MaterialExpressionNamedRerouteUsage_43}
    B {from=MaterialExpressionVectorParameter_12, output=R}
   ]
  }
  MaterialExpressionAdd_7 {
   class=MaterialExpressionAdd
   inputs [
    A {from=MaterialExpressionNamedRerouteUsage_49}
    B {from=MaterialExpressionVectorParameter_10, output=R}
   ]
  }
  MaterialExpressionAdd_8 {
   class=MaterialExpressionAdd
   inputs [
    A {from=MaterialExpressionScalarParameter_30}
    B {from=MaterialExpressionMultiply_10}
   ]
  }
  MaterialExpressionAdd_9 {
   class=MaterialExpressionAdd
   inputs [
    A {from=MaterialExpressionMultiply_20}
    B {from=MaterialExpressionMultiply_27}
   ]
  }
  MaterialExpressionAppendVector_0 {
   class=MaterialExpressionAppendVector
   inputs [
    A {from=MaterialExpressionConstant_14, output=""}
    B {from=MaterialExpressionScalarParameter_21, output=""}
   ]
  }
  MaterialExpressionAppendVector_1 {
   class=MaterialExpressionAppendVector
   inputs [
    A {from=MaterialExpressionConstant_17, output=""}
    B {from=MaterialExpressionScalarParameter_67, output=""}
   ]
  }
  MaterialExpressionAppendVector_10 {
   class=MaterialExpressionAppendVector
   inputs [
    A {from=MaterialExpressionVectorParameter_8, output=U Min}
    B {from=MaterialExpressionVectorParameter_8, output=U Min}
   ]
  }
  MaterialExpressionAppendVector_11 {
   class=MaterialExpressionAppendVector
   inputs [
    A {from=MaterialExpressionVectorParameter_8, output=U Max}
    B {from=MaterialExpressionVectorParameter_8, output=U Max}
   ]
  }
  MaterialExpressionAppendVector_12 {
   class=MaterialExpressionAppendVector
   inputs [
    A {from=MaterialExpressionScalarParameter_41, output=""}
    B {from=MaterialExpressionScalarParameter_42, output=""}
   ]
  }
  MaterialExpressionAppendVector_13 {
   class=MaterialExpressionAppendVector
   inputs [
    A {from=MaterialExpressionConstant_4, output=""}
    B {from=MaterialExpressionScalarParameter_43, output=""}
   ]
  }
  MaterialExpressionAppendVector_14 {
   class=MaterialExpressionAppendVector
   inputs [
    A {from=MaterialExpressionMaterialFunctionCall_21, output=G}
    B {from=MaterialExpressionOneMinus_6, output=""}
   ]
  }
  MaterialExpressionAppendVector_15 {
   class=MaterialExpressionAppendVector
   inputs [
    A {from=MaterialExpressionMaterialFunctionCall_23, output=G}
    B {from=MaterialExpressionOneMinus_7, output=""}
   ]
  }
  MaterialExpressionAppendVector_16 {
   class=MaterialExpressionAppendVector
   inputs [
    A {from=MaterialExpressionMaterialFunctionCall_24, output=G}
    B {from=MaterialExpressionOneMinus_8, output=""}
   ]
  }
  MaterialExpressionAppendVector_17 {
   class=MaterialExpressionAppendVector
   inputs [
    A {from=MaterialExpressionMaterialFunctionCall_25, output=G}
    B {from=MaterialExpressionOneMinus_10, output=""}
   ]
  }
  MaterialExpressionAppendVector_18 {
   class=MaterialExpressionAppendVector
   inputs [
    A {from=MaterialExpressionScalarParameter_45, output=""}
    B {from=MaterialExpressionScalarParameter_46, output=""}
   ]
  }
  MaterialExpressionAppendVector_19 {
   class=MaterialExpressionAppendVector
   inputs [
    A {from=MaterialExpressionConstant_7, output=""}
    B {from=MaterialExpressionScalarParameter_47, output=""}
   ]
  }
  MaterialExpressionAppendVector_20 {
   class=MaterialExpressionAppendVector
   inputs [
    A {from=MaterialExpressionScalarParameter_49, output=""}
    B {from=MaterialExpressionScalarParameter_50, output=""}
   ]
  }
  MaterialExpressionAppendVector_21 {
   class=MaterialExpressionAppendVector
   inputs [
    A {from=MaterialExpressionConstant_10, output=""}
    B {from=MaterialExpressionScalarParameter_51, output=""}
   ]
  }
  MaterialExpressionAppendVector_22 {
   class=MaterialExpressionAppendVector
   inputs [
    A {from=MaterialExpressionScalarParameter_53, output=""}
    B {from=MaterialExpressionScalarParameter_54, output=""}
   ]
  }
  MaterialExpressionAppendVector_23 {
   class=MaterialExpressionAppendVector
   inputs [
    A {from=MaterialExpressionConstant_21, output=""}
    B {from=MaterialExpressionScalarParameter_69, output=""}
   ]
  }
  MaterialExpressionAppendVector_24 {
   class=MaterialExpressionAppendVector
   inputs [
    A {from=MaterialExpressionConstant_24, output=""}
    B {from=MaterialExpressionScalarParameter_71, output=""}
   ]
  }
  MaterialExpressionAppendVector_25 {
   class=MaterialExpressionAppendVector
   inputs [
    A {from=MaterialExpressionConstant_27, output=""}
    B {from=MaterialExpressionScalarParameter_73, output=""}
   ]
  }
  MaterialExpressionAppendVector_26 {
   class=MaterialExpressionAppendVector
   inputs [
    A {from=MaterialExpressionConstant_32, output=""}
    B {from=MaterialExpressionScalarParameter_83, output=""}
   ]
  }
  MaterialExpressionAppendVector_27 {
   class=MaterialExpressionAppendVector
   inputs [
    A {from=MaterialExpressionNamedRerouteDeclaration_8, output=""}
    B {from=MaterialExpressionNamedRerouteDeclaration_8, output=""}
   ]
  }
  MaterialExpressionAppendVector_28 {
   class=MaterialExpressionAppendVector
   inputs [
    A {from=MaterialExpressionNamedRerouteDeclaration_3, output=""}
    B {from=MaterialExpressionNamedRerouteDeclaration_6, output=""}
   ]
  }
  MaterialExpressionAppendVector_3 {
   class=MaterialExpressionAppendVector
   inputs [
    A {from=MaterialExpressionConstant_9, output=""}
    B {from=MaterialExpressionScalarParameter_39, output=""}
   ]
  }
  MaterialExpressionAppendVector_4 {
   class=MaterialExpressionAppendVector
   inputs [
    A {from=MaterialExpressionVectorParameter_0, output=U Min}
    B {from=MaterialExpressionVectorParameter_0, output=U Min}
   ]
  }
  MaterialExpressionAppendVector_5 {
   class=MaterialExpressionAppendVector
   inputs [
    A {from=MaterialExpressionVectorParameter_0, output=U Max}
    B {from=MaterialExpressionVectorParameter_0, output=U Max}
   ]
  }
  MaterialExpressionAppendVector_6 {
   class=MaterialExpressionAppendVector
   inputs [
    A {from=MaterialExpressionVectorParameter_1, output=U Min}
    B {from=MaterialExpressionVectorParameter_1, output=U Min}
   ]
  }
  MaterialExpressionAppendVector_7 {
   class=MaterialExpressionAppendVector
   inputs [
    A {from=MaterialExpressionVectorParameter_1, output=U Max}
    B {from=MaterialExpressionVectorParameter_1, output=U Max}
   ]
  }
  MaterialExpressionAppendVector_8 {
   class=MaterialExpressionAppendVector
   inputs [
    A {from=MaterialExpressionVectorParameter_5, output=U Min}
    B {from=MaterialExpressionVectorParameter_5, output=U Min}
   ]
  }
  MaterialExpressionAppendVector_9 {
   class=MaterialExpressionAppendVector
   inputs [
    A {from=MaterialExpressionVectorParameter_5, output=U Max}
    B {from=MaterialExpressionVectorParameter_5, output=U Max}
   ]
  }
  MaterialExpressionBlendMaterialAttributes_1 {
   class=MaterialExpressionBlendMaterialAttributes
   inputs [
    A {from=MaterialExpressionBlendMaterialAttributes_4, output=""}
    B {from=MaterialExpressionSetMaterialAttributes_3, output=""}
    Alpha {from=MaterialExpressionNamedRerouteUsage_67, output=""}
   ]
   properties {vertexAttributeBlendType=UseA}
  }
  MaterialExpressionBlendMaterialAttributes_10 {
   class=MaterialExpressionBlendMaterialAttributes
   inputs [
    A {from=MaterialExpressionBlendMaterialAttributes_9, output=""}
    B {from=MaterialExpressionSetMaterialAttributes_11, output=""}
    Alpha {from=MaterialExpressionNamedRerouteUsage_112, output=""}
   ]
   properties {vertexAttributeBlendType=UseA}
  }
  MaterialExpressionBlendMaterialAttributes_2 {
   class=MaterialExpressionBlendMaterialAttributes
   inputs [
    A {from=MaterialExpressionNamedRerouteDeclaration_13, output=""}
    B {from=MaterialExpressionReroute_5, output=""}
    Alpha {from=MaterialExpressionNamedRerouteUsage_2, output=""}
   ]
   properties {}
  }
  MaterialExpressionBlendMaterialAttributes_3 {
   class=MaterialExpressionBlendMaterialAttributes
   inputs [
    A {from=MaterialExpressionNamedRerouteUsage_58, output=""}
    B {from=MaterialExpressionSetMaterialAttributes_5, output=""}
    Alpha {from=MaterialExpressionNamedRerouteUsage_45, output=""}
   ]
   properties {vertexAttributeBlendType=UseA}
  }
  MaterialExpressionBlendMaterialAttributes_4 {
   class=MaterialExpressionBlendMaterialAttributes
   inputs [
    A {from=MaterialExpressionSetMaterialAttributes_2, output=""}
    B {from=MaterialExpressionSetMaterialAttributes_6, output=""}
    Alpha {from=MaterialExpressionNamedRerouteUsage_66, output=""}
   ]
   properties {vertexAttributeBlendType=UseA}
  }
  MaterialExpressionBlendMaterialAttributes_5 {
   class=MaterialExpressionBlendMaterialAttributes
   inputs [
    A {from=MaterialExpressionGetMaterialAttributes_4, output=MaterialAttributes}
    B {from=MaterialExpressionSetMaterialAttributes_7, output=""}
    Alpha {from=MaterialExpressionScalarParameter_32, output=""}
   ]
   properties {vertexAttributeBlendType=UseA}
  }
  MaterialExpressionBlendMaterialAttributes_6 {
   class=MaterialExpressionBlendMaterialAttributes
   inputs [
    A {from=MaterialExpressionStaticSwitchParameter_19, output=""}
    B {from=MaterialExpressionSetMaterialAttributes_1, output=""}
    Alpha {from=MaterialExpressionScalarParameter_33, output=""}
   ]
   properties {vertexAttributeBlendType=UseA}
  }
  MaterialExpressionBlendMaterialAttributes_7 {
   class=MaterialExpressionBlendMaterialAttributes
   inputs [
    A {from=MaterialExpressionBlendMaterialAttributes_5, output=""}
    B {from=MaterialExpressionSetMaterialAttributes_8, output=""}
    Alpha {from=MaterialExpressionScalarParameter_59, output=""}
   ]
   properties {}
  }
  MaterialExpressionBlendMaterialAttributes_8 {
   class=MaterialExpressionBlendMaterialAttributes
   inputs [
    A {from=MaterialExpressionBlendMaterialAttributes_3, output=""}
    B {from=MaterialExpressionSetMaterialAttributes_9, output=""}
    Alpha {from=MaterialExpressionNamedRerouteUsage_46, output=""}
   ]
   properties {vertexAttributeBlendType=UseA}
  }
  MaterialExpressionBlendMaterialAttributes_9 {
   class=MaterialExpressionBlendMaterialAttributes
   inputs [
    A {from=MaterialExpressionBlendMaterialAttributes_8, output=""}
    B {from=MaterialExpressionSetMaterialAttributes_10, output=""}
    Alpha {from=MaterialExpressionNamedRerouteUsage_75, output=""}
   ]
   properties {vertexAttributeBlendType=UseA}
  }
  MaterialExpressionComponentMask_0 {class=MaterialExpressionComponentMask, inputs [None {from=MaterialExpressionNamedRerouteDeclaration_64}], properties {g=true}}
  MaterialExpressionConstant2Vector_0 {class=MaterialExpressionConstant2Vector, properties {g=1, r=2}}
  MaterialExpressionConstant2Vector_2 {class=MaterialExpressionConstant2Vector, properties {g=0.5, r=1}}
  MaterialExpressionConstant2Vector_5 {class=MaterialExpressionConstant2Vector, properties {g=1, r=2}}
  MaterialExpressionConstant3Vector_0 {class=MaterialExpressionConstant3Vector, properties {constant=<rgba 1 0 0 0>}}
  MaterialExpressionConstant3Vector_1 {class=MaterialExpressionConstant3Vector, properties {constant=<rgba 1 0 0 0>}}
  MaterialExpressionConstant3Vector_10 {class=MaterialExpressionConstant3Vector, properties {constant=<rgba 1 0 1 0>}}
  MaterialExpressionConstant3Vector_2 {class=MaterialExpressionConstant3Vector, properties {constant=<rgba 1 1 0.5 0>}}
  MaterialExpressionConstant3Vector_3 {class=MaterialExpressionConstant3Vector, properties {constant=<rgba 1 1 0 0>}}
  MaterialExpressionConstant3Vector_4 {class=MaterialExpressionConstant3Vector, properties {constant=<rgba 0 1 0 0>}}
  MaterialExpressionConstant3Vector_5 {class=MaterialExpressionConstant3Vector, properties {constant=<rgba 0 0 1 0>}}
  MaterialExpressionConstant3Vector_6 {class=MaterialExpressionConstant3Vector, properties {constant=<rgba 1 1 0 0>}}
  MaterialExpressionConstant3Vector_7 {class=MaterialExpressionConstant3Vector, properties {constant=<rgba 0 1 0 0>}}
  MaterialExpressionConstant3Vector_8 {class=MaterialExpressionConstant3Vector, properties {constant=<rgba 0 1 1 0>}}
  MaterialExpressionConstant3Vector_9 {class=MaterialExpressionConstant3Vector, properties {constant=<rgba 0.125 0.125 1 0>}}
  MaterialExpressionConstant4Vector_0 {class=MaterialExpressionConstant4Vector, inputs [], properties {constant=<rgba 0 0 0 1>}}
  MaterialExpressionConstant_1 {class=MaterialExpressionConstant, properties {r=1}}
  MaterialExpressionConstant_10 {class=MaterialExpressionConstant, properties {r=1}}
  MaterialExpressionConstant_11 {class=MaterialExpressionConstant, properties {r=0.5}}
  MaterialExpressionConstant_12 {class=MaterialExpressionConstant}
  MaterialExpressionConstant_13 {class=MaterialExpressionConstant}
  MaterialExpressionConstant_14 {class=MaterialExpressionConstant}
  MaterialExpressionConstant_15 {class=MaterialExpressionConstant, properties {r=0.5}}
  MaterialExpressionConstant_16 {class=MaterialExpressionConstant, properties {r=0.499}}
  MaterialExpressionConstant_17 {class=MaterialExpressionConstant}
  MaterialExpressionConstant_18 {class=MaterialExpressionConstant, properties {r=0.5}}
  MaterialExpressionConstant_2 {class=MaterialExpressionConstant, properties {r=2}}
  MaterialExpressionConstant_20 {class=MaterialExpressionConstant, properties {r=0.499}}
  MaterialExpressionConstant_21 {class=MaterialExpressionConstant}
  MaterialExpressionConstant_22 {class=MaterialExpressionConstant, properties {r=0.5}}
  MaterialExpressionConstant_23 {class=MaterialExpressionConstant, properties {r=0.499}}
  MaterialExpressionConstant_24 {class=MaterialExpressionConstant}
  MaterialExpressionConstant_25 {class=MaterialExpressionConstant, properties {r=0.5}}
  MaterialExpressionConstant_26 {class=MaterialExpressionConstant, properties {r=0.499}}
  MaterialExpressionConstant_27 {class=MaterialExpressionConstant}
  MaterialExpressionConstant_28 {class=MaterialExpressionConstant, properties {r=0.5}}
  MaterialExpressionConstant_29 {class=MaterialExpressionConstant, properties {r=0.499}}
  MaterialExpressionConstant_30 {class=MaterialExpressionConstant, properties {r=0.5}}
  MaterialExpressionConstant_31 {class=MaterialExpressionConstant, properties {r=0.4999}}
  MaterialExpressionConstant_32 {class=MaterialExpressionConstant}
  MaterialExpressionConstant_4 {class=MaterialExpressionConstant, properties {r=1}}
  MaterialExpressionConstant_5 {class=MaterialExpressionConstant}
  MaterialExpressionConstant_6 {class=MaterialExpressionConstant, properties {r=0.5}}
  MaterialExpressionConstant_7 {class=MaterialExpressionConstant, properties {r=1}}
  MaterialExpressionConstant_8 {class=MaterialExpressionConstant, properties {r=2}}
  MaterialExpressionConstant_9 {class=MaterialExpressionConstant, properties {r=1}}
  MaterialExpressionConvert_0 {
   class=MaterialExpressionConvert
   inputs [None {from=MaterialExpressionNamedRerouteDeclaration_65}]
   properties {
    nodeName=Break Float2
    convertInputs [{
      type=Vector2
      expressionInput {
       expression=&self:MaterialExpressionNamedRerouteDeclaration_65
       inputName=None
       mask=0
       maskA=0
       maskB=0
       maskG=0
       maskR=0
       outputIndex=0
      }
     }]
    convertOutputs [{type=Scalar}, {type=Scalar}]
   }
  }
  MaterialExpressionConvert_1 {
   class=MaterialExpressionConvert
   inputs [None {from=MaterialExpressionNamedRerouteDeclaration_66}]
   properties {
    nodeName=Break Float2
    convertInputs [{
      type=Vector2
      expressionInput {
       expression=&self:MaterialExpressionNamedRerouteDeclaration_66
       inputName=None
       mask=0
       maskA=0
       maskB=0
       maskG=0
       maskR=0
       outputIndex=0
      }
     }]
    convertOutputs [{type=Scalar}, {type=Scalar}]
   }
  }
  MaterialExpressionConvert_11 {
   class=MaterialExpressionConvert
   inputs [None {from=MaterialExpressionLinearInterpolate_33}]
   properties {
    nodeName=Break Float4
    convertInputs [{
      type=Vector4
      defaultValue=<rgba 0 0 0 1>
      expressionInput {
       expression=&self:MaterialExpressionLinearInterpolate_33
       inputName=None
       mask=0
       maskA=0
       maskB=0
       maskG=0
       maskR=0
       outputIndex=0
      }
     }]
    convertOutputs [{type=Scalar}, {type=Scalar}, {type=Scalar}, {type=Scalar}]
   }
  }
  MaterialExpressionConvert_13 {
   class=MaterialExpressionConvert
   inputs [None {from=MaterialExpressionNamedRerouteUsage_17}]
   properties {
    nodeName=Break Float2
    convertInputs [{
      type=Vector2
      expressionInput {
       expression=&self:MaterialExpressionNamedRerouteUsage_17
       inputName=None
       mask=0
       maskA=0
       maskB=0
       maskG=0
       maskR=0
       outputIndex=0
      }
     }]
    convertOutputs [{type=Scalar}, {type=Scalar}]
   }
  }
  MaterialExpressionConvert_16 {
   class=MaterialExpressionConvert
   inputs [
    None {from=MaterialExpressionAdd_7}
    None {from=MaterialExpressionVectorParameter_10, output=G}
    None {from=MaterialExpressionVectorParameter_10, output=G}
    None {from=MaterialExpressionVectorParameter_10, output=G}
   ]
   properties {
    nodeName=Make Vector4
    convertInputs [
     {
      type=Scalar
      expressionInput {
       expression=&self:MaterialExpressionAdd_7
       inputName=None
       mask=0
       maskA=0
       maskB=0
       maskG=0
       maskR=0
       outputIndex=0
      }
     }
     {
      type=Scalar
      expressionInput {
       expression=&self:MaterialExpressionVectorParameter_10
       inputName=None
       mask=1
       maskA=0
       maskB=0
       maskG=1
       maskR=0
       outputIndex=2
      }
     }
     {
      type=Scalar
      expressionInput {
       expression=&self:MaterialExpressionVectorParameter_10
       inputName=None
       mask=1
       maskA=0
       maskB=1
       maskG=0
       maskR=0
       outputIndex=3
      }
     }
     {
      type=Scalar
      expressionInput {
       expression=&self:MaterialExpressionVectorParameter_10
       inputName=None
       mask=1
       maskA=1
       maskB=0
       maskG=0
       maskR=0
       outputIndex=4
      }
     }
    ]
    convertOutputs [{type=Vector4, defaultValue=<rgba 0 0 0 1>}]
   }
  }
  MaterialExpressionConvert_19 {
   class=MaterialExpressionConvert
   inputs [
    None {from=MaterialExpressionAdd_13}
    None {from=MaterialExpressionVectorParameter_19, output=G}
    None {from=MaterialExpressionVectorParameter_19, output=G}
    None {from=MaterialExpressionVectorParameter_19, output=G}
   ]
   properties {
    nodeName=Make Vector4
    convertInputs [
     {
      type=Scalar
      expressionInput {
       expression=&self:MaterialExpressionAdd_13
       inputName=None
       mask=0
       maskA=0
       maskB=0
       maskG=0
       maskR=0
       outputIndex=0
      }
     }
     {
      type=Scalar
      expressionInput {
       expression=&self:MaterialExpressionVectorParameter_19
       inputName=None
       mask=1
       maskA=0
       maskB=0
       maskG=1
       maskR=0
       outputIndex=2
      }
     }
     {
      type=Scalar
      expressionInput {
       expression=&self:MaterialExpressionVectorParameter_19
       inputName=None
       mask=1
       maskA=0
       maskB=1
       maskG=0
       maskR=0
       outputIndex=3
      }
     }
     {
      type=Scalar
      expressionInput {
       expression=&self:MaterialExpressionVectorParameter_19
       inputName=None
       mask=1
       maskA=1
       maskB=0
       maskG=0
       maskR=0
       outputIndex=4
      }
     }
    ]
    convertOutputs [{type=Vector4, defaultValue=<rgba 0 0 0 1>}]
   }
  }
  MaterialExpressionConvert_2 {
   class=MaterialExpressionConvert
   inputs [None {from=MaterialExpressionNamedRerouteDeclaration_67}]
   properties {
    nodeName=Break Float2
    convertInputs [{
      type=Vector2
      expressionInput {
       expression=&self:MaterialExpressionNamedRerouteDeclaration_67
       inputName=None
       mask=0
       maskA=0
       maskB=0
       maskG=0
       maskR=0
       outputIndex=0
      }
     }]
    convertOutputs [{type=Scalar}, {type=Scalar}]
   }
  }
  MaterialExpressionConvert_20 {
   class=MaterialExpressionConvert
   inputs [
    None {from=MaterialExpressionScalarParameter_18}
    None {from=MaterialExpressionScalarParameter_19}
    None {from=MaterialExpressionScalarParameter_80}
    None {from=MaterialExpressionScalarParameter_81}
   ]
   properties {
    nodeName=Make Vector4
    convertInputs [
     {
      type=Scalar
      expressionInput {
       expression=&self:MaterialExpressionScalarParameter_18
       inputName=None
       mask=0
       maskA=0
       maskB=0
       maskG=0
       maskR=0
       outputIndex=0
      }
     }
     {
      type=Scalar
      expressionInput {
       expression=&self:MaterialExpressionScalarParameter_19
       inputName=None
       mask=0
       maskA=0
       maskB=0
       maskG=0
       maskR=0
       outputIndex=0
      }
     }
     {
      type=Scalar
      expressionInput {
       expression=&self:MaterialExpressionScalarParameter_80
       inputName=None
       mask=0
       maskA=0
       maskB=0
       maskG=0
       maskR=0
       outputIndex=0
      }
     }
     {
      type=Scalar
      expressionInput {
       expression=&self:MaterialExpressionScalarParameter_81
       inputName=None
       mask=0
       maskA=0
       maskB=0
       maskG=0
       maskR=0
       outputIndex=0
      }
     }
    ]
    convertOutputs [{type=Vector4, defaultValue=<rgba 0 0 0 1>}]
   }
  }
  MaterialExpressionConvert_22 {
   class=MaterialExpressionConvert
   inputs [
    None {from=MaterialExpressionScalarParameter_13}
    None {from=MaterialExpressionScalarParameter_14}
    None {from=MaterialExpressionScalarParameter_82}
    None {from=MaterialExpressionScalarParameter_85}
   ]
   properties {
    nodeName=Make Vector4
    convertInputs [
     {
      type=Scalar
      expressionInput {
       expression=&self:MaterialExpressionScalarParameter_13
       inputName=None
       mask=0
       maskA=0
       maskB=0
       maskG=0
       maskR=0
       outputIndex=0
      }
     }
     {
      type=Scalar
      expressionInput {
       expression=&self:MaterialExpressionScalarParameter_14
       inputName=None
       mask=0
       maskA=0
       maskB=0
       maskG=0
       maskR=0
       outputIndex=0
      }
     }
     {
      type=Scalar
      expressionInput {
       expression=&self:MaterialExpressionScalarParameter_82
       inputName=None
       mask=0
       maskA=0
       maskB=0
       maskG=0
       maskR=0
       outputIndex=0
      }
     }
     {
      type=Scalar
      expressionInput {
       expression=&self:MaterialExpressionScalarParameter_85
       inputName=None
       mask=0
       maskA=0
       maskB=0
       maskG=0
       maskR=0
       outputIndex=0
      }
     }
    ]
    convertOutputs [{type=Vector4, defaultValue=<rgba 0 0 0 1>}]
   }
  }
  MaterialExpressionConvert_23 {
   class=MaterialExpressionConvert
   inputs [None {from=MaterialExpressionMultiply_5}]
   properties {
    nodeName=Break Float4
    convertInputs [{
      type=Vector4
      defaultValue=<rgba 0 0 0 1>
      expressionInput {
       expression=&self:MaterialExpressionMultiply_5
       inputName=None
       mask=0
       maskA=0
       maskB=0
       maskG=0
       maskR=0
       outputIndex=0
      }
     }]
    convertOutputs [{type=Scalar}, {type=Scalar}, {type=Scalar}, {type=Scalar}]
   }
  }
  MaterialExpressionConvert_3 {
   class=MaterialExpressionConvert
   inputs [None {from=MaterialExpressionNamedRerouteDeclaration_68}]
   properties {
    nodeName=Break Float2
    convertInputs [{
      type=Vector2
      expressionInput {
       expression=&self:MaterialExpressionNamedRerouteDeclaration_68
       inputName=None
       mask=0
       maskA=0
       maskB=0
       maskG=0
       maskR=0
       outputIndex=0
      }
     }]
    convertOutputs [{type=Scalar}, {type=Scalar}]
   }
  }
  MaterialExpressionConvert_4 {
   class=MaterialExpressionConvert
   inputs [None {from=MaterialExpressionSubtract_13}]
   properties {
    nodeName=Break Float2
    convertInputs [{
      type=Vector2
      expressionInput {
       expression=&self:MaterialExpressionSubtract_13
       inputName=None
       mask=0
       maskA=0
       maskB=0
       maskG=0
       maskR=0
       outputIndex=0
      }
     }]
    convertOutputs [{type=Scalar}, {type=Scalar}]
   }
  }
  MaterialExpressionConvert_5 {
   class=MaterialExpressionConvert
   inputs [None {from=MaterialExpressionScalarParameter_3}]
   properties {
    convertInputs [{
      type=Scalar
      expressionInput {
       expression=&self:MaterialExpressionScalarParameter_3
       inputName=None
       mask=0
       maskA=0
       maskB=0
       maskG=0
       maskR=0
       outputIndex=0
      }
     }]
    convertOutputs [{type=Vector3}]
   }
  }
  MaterialExpressionConvert_6 {
   class=MaterialExpressionConvert
   inputs [None {from=MaterialExpressionSubtract_8}]
   properties {
    nodeName=Break Float2
    convertInputs [{
      type=Vector2
      expressionInput {
       expression=&self:MaterialExpressionSubtract_8
       inputName=None
       mask=0
       maskA=0
       maskB=0
       maskG=0
       maskR=0
       outputIndex=0
      }
     }]
    convertOutputs [{type=Scalar}, {type=Scalar}]
   }
  }
  MaterialExpressionConvert_7 {
   class=MaterialExpressionConvert
   inputs [
    None {from=""}
    None {from=MaterialExpressionFloor_0}
   ]
   properties {
    nodeName=Make Vector2
    convertInputs [
     {type=Scalar, expressionInput {expression=null, inputName=None, mask=0, maskA=0, maskB=0, maskG=0, maskR=0, outputIndex=0}}
     {
      type=Scalar
      expressionInput {
       expression=&self:MaterialExpressionFloor_0
       inputName=None
       mask=0
       maskA=0
       maskB=0
       maskG=0
       maskR=0
       outputIndex=0
      }
     }
    ]
    convertOutputs [{type=Vector2}]
   }
  }
  MaterialExpressionConvert_8 {
   class=MaterialExpressionConvert
   inputs [
    None {from=MaterialExpressionAdd_5}
    None {from=MaterialExpressionVectorParameter_11, output=G}
    None {from=MaterialExpressionVectorParameter_11, output=G}
    None {from=MaterialExpressionVectorParameter_11, output=G}
   ]
   properties {
    nodeName=Make Vector4
    convertInputs [
     {
      type=Scalar
      expressionInput {
       expression=&self:MaterialExpressionAdd_5
       inputName=None
       mask=0
       maskA=0
       maskB=0
       maskG=0
       maskR=0
       outputIndex=0
      }
     }
     {
      type=Scalar
      expressionInput {
       expression=&self:MaterialExpressionVectorParameter_11
       inputName=None
       mask=1
       maskA=0
       maskB=0
       maskG=1
       maskR=0
       outputIndex=2
      }
     }
     {
      type=Scalar
      expressionInput {
       expression=&self:MaterialExpressionVectorParameter_11
       inputName=None
       mask=1
       maskA=0
       maskB=1
       maskG=0
       maskR=0
       outputIndex=3
      }
     }
     {
      type=Scalar
      expressionInput {
       expression=&self:MaterialExpressionVectorParameter_11
       inputName=None
       mask=1
       maskA=1
       maskB=0
       maskG=0
       maskR=0
       outputIndex=4
      }
     }
    ]
    convertOutputs [{type=Vector4, defaultValue=<rgba 0 0 0 1>}]
   }
  }
  MaterialExpressionConvert_9 {
   class=MaterialExpressionConvert
   inputs [
    None {from=MaterialExpressionAdd_6}
    None {from=MaterialExpressionVectorParameter_12, output=G}
    None {from=MaterialExpressionVectorParameter_12, output=G}
    None {from=MaterialExpressionVectorParameter_12, output=G}
   ]
   properties {
    nodeName=Make Vector4
    convertInputs [
     {
      type=Scalar
      expressionInput {
       expression=&self:MaterialExpressionAdd_6
       inputName=None
       mask=0
       maskA=0
       maskB=0
       maskG=0
       maskR=0
       outputIndex=0
      }
     }
     {
      type=Scalar
      expressionInput {
       expression=&self:MaterialExpressionVectorParameter_12
       inputName=None
       mask=1
       maskA=0
       maskB=0
       maskG=1
       maskR=0
       outputIndex=2
      }
     }
     {
      type=Scalar
      expressionInput {
       expression=&self:MaterialExpressionVectorParameter_12
       inputName=None
       mask=1
       maskA=0
       maskB=1
       maskG=0
       maskR=0
       outputIndex=3
      }
     }
     {
      type=Scalar
      expressionInput {
       expression=&self:MaterialExpressionVectorParameter_12
       inputName=None
       mask=1
       maskA=1
       maskB=0
       maskG=0
       maskR=0
       outputIndex=4
      }
     }
    ]
    convertOutputs [{type=Vector4, defaultValue=<rgba 0 0 0 1>}]
   }
  }
  MaterialExpressionDesaturation_0 {
   class=MaterialExpressionDesaturation
   inputs [
    None {from=MaterialExpressionNamedRerouteUsage_65, output=""}
    Fraction {from="", output=""}
   ]
   properties {}
  }
  MaterialExpressionDesaturation_1 {
   class=MaterialExpressionDesaturation
   inputs [
    None {from=MaterialExpressionTextureSampleParameter2D_11, output=RGB}
    Fraction {from="", output=""}
   ]
   properties {}
  }
  MaterialExpressionDivide_8 {
   class=MaterialExpressionDivide
   inputs [
    A {from=MaterialExpressionSubtract_16}
    B {from=MaterialExpressionMultiply_19}
   ]
  }
  MaterialExpressionFloor_0 {class=MaterialExpressionFloor, inputs [None {from=MaterialExpressionComponentMask_0, output=""}], properties {}}
  MaterialExpressionFrac_0 {class=MaterialExpressionFrac, inputs [None {from=MaterialExpressionNamedRerouteUsage_86, output=""}], properties {}}
  MaterialExpressionGetMaterialAttributes_0 {
   class=MaterialExpressionGetMaterialAttributes
   inputs [None {from=MaterialExpressionNamedRerouteUsage_24, output=""}]
   properties {attributeGetTypes [#1]}
  }
  MaterialExpressionGetMaterialAttributes_1 {
   class=MaterialExpressionGetMaterialAttributes
   inputs [None {from=MaterialExpressionReroute_1, output=""}]
   properties {attributeGetTypes [#1]}
  }
  MaterialExpressionGetMaterialAttributes_2 {
   class=MaterialExpressionGetMaterialAttributes
   inputs [None {from=MaterialExpressionBlendMaterialAttributes_10, output=""}]
   properties {attributeGetTypes [#1]}
  }
  MaterialExpressionGetMaterialAttributes_4 {
   class=MaterialExpressionGetMaterialAttributes
   inputs [None {from=MaterialExpressionBlendMaterialAttributes_6, output=""}]
   properties {attributeGetTypes [#1]}
  }
  MaterialExpressionGetMaterialAttributes_5 {
   class=MaterialExpressionGetMaterialAttributes
   inputs [None {from=MaterialExpressionNamedRerouteUsage_30, output=""}]
   properties {attributeGetTypes [#2]}
  }
  MaterialExpressionGetMaterialAttributes_7 {
   class=MaterialExpressionGetMaterialAttributes
   inputs [None {from=MaterialExpressionNamedRerouteUsage_27, output=""}]
   properties {attributeGetTypes [#1, #3]}
  }
  MaterialExpressionLinearInterpolate_1 {
   class=MaterialExpressionLinearInterpolate
   inputs [
    A {from=MaterialExpressionNamedRerouteUsage_51}
    B {from=MaterialExpressionNamedRerouteUsage_48}
    Alpha {from=MaterialExpressionNamedRerouteUsage_47}
   ]
  }
  MaterialExpressionLinearInterpolate_10 {
   class=MaterialExpressionLinearInterpolate
   inputs [
    A {from=MaterialExpressionLinearInterpolate_8}
    B {from=MaterialExpressionVectorParameter_16, output=RGB}
    Alpha {from=MaterialExpressionNamedRerouteUsage_35}
   ]
  }
  MaterialExpressionLinearInterpolate_11 {
   class=MaterialExpressionLinearInterpolate
   inputs [
    A {from=MaterialExpressionNamedRerouteUsage_15}
    B {from=MaterialExpressionNamedRerouteUsage_59}
    Alpha {from=MaterialExpressionMultiply_26}
   ]
  }
  MaterialExpressionLinearInterpolate_13 {
   class=MaterialExpressionLinearInterpolate
   inputs [
    A {from=MaterialExpressionLinearInterpolate_9}
    B {from=MaterialExpressionNamedRerouteUsage_61}
    Alpha {from=MaterialExpressionMultiply_54}
   ]
  }
  MaterialExpressionLinearInterpolate_14 {
   class=MaterialExpressionLinearInterpolate
   inputs [
    A {from=MaterialExpressionLinearInterpolate_13}
    B {from=MaterialExpressionNamedRerouteUsage_62}
    Alpha {from=MaterialExpressionMultiply_55}
   ]
  }
  MaterialExpressionLinearInterpolate_15 {
   class=MaterialExpressionLinearInterpolate
   inputs [
    A {from=MaterialExpressionLinearInterpolate_10}
    B {from=MaterialExpressionVectorParameter_4, output=RGB}
    Alpha {from=MaterialExpressionNamedRerouteUsage_36}
   ]
  }
  MaterialExpressionLinearInterpolate_16 {
   class=MaterialExpressionLinearInterpolate
   inputs [
    A {from=MaterialExpressionLinearInterpolate_17}
    B {from=MaterialExpressionNamedRerouteUsage_6}
    Alpha {from=MaterialExpressionNamedRerouteUsage_77}
   ]
  }
  MaterialExpressionLinearInterpolate_17 {
   class=MaterialExpressionLinearInterpolate
   inputs [
    A {from=MaterialExpressionConstant4Vector_0, output=RGBA}
    B {from=MaterialExpressionNamedRerouteUsage_103}
    Alpha {from=MaterialExpressionNamedRerouteUsage_72}
   ]
  }
  MaterialExpressionLinearInterpolate_18 {
   class=MaterialExpressionLinearInterpolate
   inputs [
    A {from=MaterialExpressionLinearInterpolate_16}
    B {from=MaterialExpressionNamedRerouteUsage_10}
    Alpha {from=MaterialExpressionNamedRerouteUsage_14}
   ]
  }
  MaterialExpressionLinearInterpolate_19 {
   class=MaterialExpressionLinearInterpolate
   inputs [
    A {from=""}
    B {from=""}
    Alpha {from=MaterialExpressionNamedRerouteUsage_37}
   ]
   properties {constA=1}
  }
  MaterialExpressionLinearInterpolate_20 {
   class=MaterialExpressionLinearInterpolate
   inputs [
    A {from=MaterialExpressionLinearInterpolate_19}
    B {from=""}
    Alpha {from=MaterialExpressionNamedRerouteUsage_39}
   ]
  }
  MaterialExpressionLinearInterpolate_21 {
   class=MaterialExpressionLinearInterpolate
   inputs [
    A {from=MaterialExpressionLinearInterpolate_20}
    B {from=""}
    Alpha {from=MaterialExpressionNamedRerouteUsage_40}
   ]
  }
  MaterialExpressionLinearInterpolate_22 {
   class=MaterialExpressionLinearInterpolate
   inputs [
    A {from=MaterialExpressionLinearInterpolate_21}
    B {from=""}
    Alpha {from=MaterialExpressionNamedRerouteUsage_41}
   ]
  }
  MaterialExpressionLinearInterpolate_23 {
   class=MaterialExpressionLinearInterpolate
   inputs [
    A {from=MaterialExpressionLinearInterpolate_22}
    B {from=""}
    Alpha {from=MaterialExpressionNamedRerouteUsage_42}
   ]
  }
  MaterialExpressionLinearInterpolate_24 {
   class=MaterialExpressionLinearInterpolate
   inputs [
    A {from=MaterialExpressionLinearInterpolate_23}
    B {from=""}
    Alpha {from=MaterialExpressionNamedRerouteUsage_44}
   ]
  }
  MaterialExpressionLinearInterpolate_25 {
   class=MaterialExpressionLinearInterpolate
   inputs [
    A {from=""}
    B {from=MaterialExpressionConstant3Vector_0}
    Alpha {from=MaterialExpressionNamedRerouteUsage_22}
   ]
  }
  MaterialExpressionLinearInterpolate_26 {
   class=MaterialExpressionLinearInterpolate
   inputs [
    A {from=MaterialExpressionLinearInterpolate_25}
    B {from=MaterialExpressionConstant3Vector_3}
    Alpha {from=MaterialExpressionNamedRerouteUsage_54}
   ]
  }
  MaterialExpressionLinearInterpolate_27 {
   class=MaterialExpressionLinearInterpolate
   inputs [
    A {from=MaterialExpressionLinearInterpolate_26}
    B {from=MaterialExpressionConstant3Vector_4}
    Alpha {from=MaterialExpressionNamedRerouteUsage_52}
   ]
  }
  MaterialExpressionLinearInterpolate_28 {
   class=MaterialExpressionLinearInterpolate
   inputs [
    A {from=MaterialExpressionLinearInterpolate_27}
    B {from=MaterialExpressionConstant3Vector_5}
    Alpha {from=MaterialExpressionNamedRerouteUsage_50}
   ]
  }
  MaterialExpressionLinearInterpolate_29 {
   class=MaterialExpressionLinearInterpolate
   inputs [
    A {from=MaterialExpressionLinearInterpolate_38}
    B {from=MaterialExpressionScalarParameter_56}
    Alpha {from=MaterialExpressionNamedRerouteUsage_95}
   ]
  }
  MaterialExpressionLinearInterpolate_30 {
   class=MaterialExpressionLinearInterpolate
   inputs [
    A {from=MaterialExpressionLinearInterpolate_32}
    B {from=MaterialExpressionLinearInterpolate_31}
    Alpha {from=MaterialExpressionSaturate_3}
   ]
  }
  MaterialExpressionLinearInterpolate_31 {
   class=MaterialExpressionLinearInterpolate
   inputs [
    A {from=MaterialExpressionAbs_0}
    B {from=""}
    Alpha {from=MaterialExpressionNamedRerouteUsage_116}
   ]
  }
  MaterialExpressionLinearInterpolate_32 {
   class=MaterialExpressionLinearInterpolate
   inputs [
    A {from=""}
    B {from=MaterialExpressionOneMinus_2}
    Alpha {from=MaterialExpressionNamedRerouteUsage_116}
   ]
  }
  MaterialExpressionLinearInterpolate_33 {
   class=MaterialExpressionLinearInterpolate
   inputs [
    A {from=MaterialExpressionLinearInterpolate_18}
    B {from=MaterialExpressionNamedRerouteUsage_25}
    Alpha {from=MaterialExpressionNamedRerouteUsage_79}
   ]
  }
  MaterialExpressionLinearInterpolate_36 {
   class=MaterialExpressionLinearInterpolate
   inputs [
    A {from=MaterialExpressionMultiply_20}
    B {from=MaterialExpressionAdd_9}
    Alpha {from=MaterialExpressionSaturate_2}
   ]
  }
  MaterialExpressionLinearInterpolate_37 {
   class=MaterialExpressionLinearInterpolate
   inputs [
    A {from=MaterialExpressionConstant_1}
    B {from=MaterialExpressionDesaturation_0}
    Alpha {from=MaterialExpressionScalarParameter_34}
   ]
   properties {constAlpha=0.6}
  }
  MaterialExpressionLinearInterpolate_38 {
   class=MaterialExpressionLinearInterpolate
   inputs [
    A {from=MaterialExpressionConstant_12}
    B {from=MaterialExpressionScalarParameter_2}
    Alpha {from=MaterialExpressionNamedRerouteUsage_94}
   ]
  }
  MaterialExpressionLinearInterpolate_39 {
   class=MaterialExpressionLinearInterpolate
   inputs [
    A {from=MaterialExpressionLinearInterpolate_29}
    B {from=MaterialExpressionScalarParameter_57}
    Alpha {from=MaterialExpressionNamedRerouteUsage_96}
   ]
  }
  MaterialExpressionLinearInterpolate_4 {
   class=MaterialExpressionLinearInterpolate
   inputs [
    A {from=""}
    B {from=MaterialExpressionMultiply_13}
    Alpha {from=MaterialExpressionScalarParameter_9}
   ]
   properties {constA=1}
  }
  MaterialExpressionLinearInterpolate_40 {
   class=MaterialExpressionLinearInterpolate
   inputs [
    A {from=MaterialExpressionLinearInterpolate_39}
    B {from=MaterialExpressionScalarParameter_58}
    Alpha {from=MaterialExpressionNamedRerouteUsage_97}
   ]
  }
  MaterialExpressionLinearInterpolate_41 {
   class=MaterialExpressionLinearInterpolate
   inputs [
    A {from=MaterialExpressionLinearInterpolate_42}
    B {from=MaterialExpressionScalarParameter_64}
    Alpha {from=MaterialExpressionNamedRerouteUsage_102}
   ]
  }
  MaterialExpressionLinearInterpolate_42 {
   class=MaterialExpressionLinearInterpolate
   inputs [
    A {from=MaterialExpressionConstant_13}
    B {from=MaterialExpressionScalarParameter_55}
    Alpha {from=MaterialExpressionNamedRerouteUsage_99}
   ]
  }
  MaterialExpressionLinearInterpolate_43 {
   class=MaterialExpressionLinearInterpolate
   inputs [
    A {from=MaterialExpressionMaterialFunctionCall_2, output="Island UV (V2)"}
    B {from=MaterialExpressionAppendVector_14}
    Alpha {from=MaterialExpressionScalarParameter_35}
   ]
  }
  MaterialExpressionLinearInterpolate_44 {
   class=MaterialExpressionLinearInterpolate
   inputs [
    A {from=MaterialExpressionReroute_4}
    B {from=MaterialExpressionAppendVector_15}
    Alpha {from=MaterialExpressionScalarParameter_36}
   ]
  }
  MaterialExpressionLinearInterpolate_45 {
   class=MaterialExpressionLinearInterpolate
   inputs [
    A {from=MaterialExpressionReroute_3}
    B {from=MaterialExpressionAppendVector_16}
    Alpha {from=MaterialExpressionScalarParameter_37}
   ]
  }
  MaterialExpressionLinearInterpolate_46 {
   class=MaterialExpressionLinearInterpolate
   inputs [
    A {from=MaterialExpressionMaterialFunctionCall_16, output="Island UV (V2)"}
    B {from=MaterialExpressionAppendVector_17}
    Alpha {from=MaterialExpressionScalarParameter_38}
   ]
  }
  MaterialExpressionLinearInterpolate_47 {
   class=MaterialExpressionLinearInterpolate
   inputs [
    A {from=MaterialExpressionLinearInterpolate_41}
    B {from=MaterialExpressionScalarParameter_65}
    Alpha {from=MaterialExpressionNamedRerouteUsage_109}
   ]
  }
  MaterialExpressionLinearInterpolate_48 {
   class=MaterialExpressionLinearInterpolate
   inputs [
    A {from=MaterialExpressionLinearInterpolate_47}
    B {from=MaterialExpressionScalarParameter_66}
    Alpha {from=MaterialExpressionNamedRerouteUsage_110}
   ]
  }
  MaterialExpressionLinearInterpolate_5 {
   class=MaterialExpressionLinearInterpolate
   inputs [
    A {from=MaterialExpressionGetMaterialAttributes_1, output=Base Color}
    B {from=MaterialExpressionVectorParameter_6, output=RGB}
    Alpha {from=MaterialExpressionNamedRerouteUsage_3}
   ]
   properties {constA=1}
  }
  MaterialExpressionLinearInterpolate_50 {
   class=MaterialExpressionLinearInterpolate
   inputs [
    A {from=MaterialExpressionLinearInterpolate_44}
    B {from=MaterialExpressionOneMinus_12}
    Alpha {from=MaterialExpressionScalarParameter_75}
   ]
  }
  MaterialExpressionLinearInterpolate_51 {
   class=MaterialExpressionLinearInterpolate
   inputs [
    A {from=MaterialExpressionLinearInterpolate_45}
    B {from=MaterialExpressionOneMinus_13}
    Alpha {from=MaterialExpressionScalarParameter_76}
   ]
  }
  MaterialExpressionLinearInterpolate_52 {
   class=MaterialExpressionLinearInterpolate
   inputs [
    A {from=MaterialExpressionLinearInterpolate_46}
    B {from=MaterialExpressionOneMinus_14}
    Alpha {from=MaterialExpressionScalarParameter_77}
   ]
  }
  MaterialExpressionLinearInterpolate_53 {
   class=MaterialExpressionLinearInterpolate
   inputs [
    A {from=MaterialExpressionLinearInterpolate_43}
    B {from=MaterialExpressionOneMinus_15}
    Alpha {from=MaterialExpressionScalarParameter_78}
   ]
  }
  MaterialExpressionLinearInterpolate_6 {
   class=MaterialExpressionLinearInterpolate
   inputs [
    A {from=MaterialExpressionLinearInterpolate_5}
    B {from=MaterialExpressionVectorParameter_13, output=RGB}
    Alpha {from=MaterialExpressionNamedRerouteUsage_4}
   ]
  }
  MaterialExpressionLinearInterpolate_7 {
   class=MaterialExpressionLinearInterpolate
   inputs [
    A {from=MaterialExpressionLinearInterpolate_6}
    B {from=MaterialExpressionVectorParameter_14, output=RGB}
    Alpha {from=MaterialExpressionNamedRerouteUsage_11}
   ]
  }
  MaterialExpressionLinearInterpolate_8 {
   class=MaterialExpressionLinearInterpolate
   inputs [
    A {from=MaterialExpressionLinearInterpolate_7}
    B {from=MaterialExpressionVectorParameter_15, output=RGB}
    Alpha {from=MaterialExpressionNamedRerouteUsage_34}
   ]
  }
  MaterialExpressionLinearInterpolate_9 {
   class=MaterialExpressionLinearInterpolate
   inputs [
    A {from=MaterialExpressionLinearInterpolate_11}
    B {from=MaterialExpressionNamedRerouteUsage_60}
    Alpha {from=MaterialExpressionMultiply_28}
   ]
  }
  MaterialExpressionMaterialFunctionCall_0 {
   class=MaterialExpressionMaterialFunctionCall
   inputs [
    Overide Texture Set UVs {from=""}
    Override Tech Art Mask UVs {from=""}
    Override Basic Emissive UVs {from=""}
   ]
   properties {materialFunction=&MF_CharacterParent_2023_PreFX}
  }
  MaterialExpressionMaterialFunctionCall_1 {
   class=MaterialExpressionMaterialFunctionCall
   inputs [
    Material {from=MaterialExpressionNamedRerouteUsage_71}
    M {from=MaterialExpressionMaterialFunctionCall_0, output=M}
    Pre Skinned Local Position {from=MaterialExpressionMaterialFunctionCall_0, output=M}
    Distance Blend {from=MaterialExpressionMaterialFunctionCall_0, output=M}
    Fresnel {from=MaterialExpressionNamedRerouteDeclaration_48}
   ]
   properties {materialFunction=&MF_CharacterParent_2023_PostFX}
  }
  MaterialExpressionMaterialFunctionCall_12 {
   class=MaterialExpressionMaterialFunctionCall
   inputs [Multiplier {from=MaterialExpressionConstant_2}]
   properties {materialFunction=&Pi}
  }
  MaterialExpressionMaterialFunctionCall_15 {
   class=MaterialExpressionMaterialFunctionCall
   inputs [
    UV in {from=MaterialExpressionNamedRerouteUsage_127}
    Island UV Min {from=MaterialExpressionAppendVector_4}
    Island UV Max {from=MaterialExpressionAppendVector_5}
   ]
   properties {materialFunction=/BRCosmetics/Characters/Player/Female/Medium/Bodies/F_MED_FermiumStraw/Materials/MaterialFunctions/MF_NormalizeUVIsland}
  }
  MaterialExpressionMaterialFunctionCall_16 {
   class=MaterialExpressionMaterialFunctionCall
   inputs [
    UV in {from=MaterialExpressionNamedRerouteUsage_130}
    Island UV Min {from=MaterialExpressionAppendVector_10}
    Island UV Max {from=MaterialExpressionAppendVector_11}
   ]
   properties {materialFunction=&MF_NormalizeUVIsland}
  }
  MaterialExpressionMaterialFunctionCall_18 {
   class=MaterialExpressionMaterialFunctionCall
   inputs [
    UV in {from=MaterialExpressionNamedRerouteUsage_129}
    Island UV Min {from=MaterialExpressionAppendVector_8}
    Island UV Max {from=MaterialExpressionAppendVector_9}
   ]
   properties {materialFunction=&MF_NormalizeUVIsland}
  }
  MaterialExpressionMaterialFunctionCall_2 {
   class=MaterialExpressionMaterialFunctionCall
   inputs [
    UV in {from=MaterialExpressionNamedRerouteUsage_128}
    Island UV Min {from=MaterialExpressionAppendVector_6}
    Island UV Max {from=MaterialExpressionAppendVector_7}
   ]
   properties {materialFunction=&MF_NormalizeUVIsland}
  }
  MaterialExpressionMaterialFunctionCall_21 {
   class=MaterialExpressionMaterialFunctionCall
   inputs [Float2 {from=MaterialExpressionMaterialFunctionCall_2, output="Island UV (V2)"}]
   properties {materialFunction=&BreakOutFloat2Components}
  }
  MaterialExpressionMaterialFunctionCall_23 {
   class=MaterialExpressionMaterialFunctionCall
   inputs [Float2 {from=MaterialExpressionMaterialFunctionCall_15, output="Island UV (V2)"}]
   properties {materialFunction=&BreakOutFloat2Components}
  }
  MaterialExpressionMaterialFunctionCall_24 {
   class=MaterialExpressionMaterialFunctionCall
   inputs [Float2 {from=MaterialExpressionMaterialFunctionCall_18, output="Island UV (V2)"}]
   properties {materialFunction=&BreakOutFloat2Components}
  }
  MaterialExpressionMaterialFunctionCall_25 {
   class=MaterialExpressionMaterialFunctionCall
   inputs [Float2 {from=MaterialExpressionMaterialFunctionCall_16, output="Island UV (V2)"}]
   properties {materialFunction=&BreakOutFloat2Components}
  }
  MaterialExpressionMaterialFunctionCall_26 {
   class=MaterialExpressionMaterialFunctionCall
   inputs [
    UVs {from=MaterialExpressionLinearInterpolate_50}
    Texture Scale {from=MaterialExpressionMultiply_49}
   ]
   properties {materialFunction=&ScaleUVsByCenter}
  }
  MaterialExpressionMaterialFunctionCall_3 {
   class=MaterialExpressionMaterialFunctionCall
   inputs [
    UVs {from=MaterialExpressionLinearInterpolate_51}
    Texture Scale {from=MaterialExpressionMultiply_31}
   ]
   properties {materialFunction=&ScaleUVsByCenter}
  }
  MaterialExpressionMaterialFunctionCall_36 {
   class=MaterialExpressionMaterialFunctionCall
   inputs [Float2 {from=MaterialExpressionNamedRerouteDeclaration_50}]
   properties {materialFunction=&BreakOutFloat2Components}
  }
  MaterialExpressionMaterialFunctionCall_4 {
   class=MaterialExpressionMaterialFunctionCall
   inputs [
    UVs {from=MaterialExpressionDivide_8}
    Texture Scale {from=""}
   ]
   properties {materialFunction=&ScaleUVsByCenter}
  }
  MaterialExpressionMaterialFunctionCall_5 {
   class=MaterialExpressionMaterialFunctionCall
   inputs [
    UVs {from=MaterialExpressionLinearInterpolate_52}
    Texture Scale {from=MaterialExpressionMultiply_32}
   ]
   properties {materialFunction=&ScaleUVsByCenter}
  }
  MaterialExpressionMaterialFunctionCall_6 {
   class=MaterialExpressionMaterialFunctionCall
   inputs [
    UVs {from=MaterialExpressionRotator_0}
    Texture Scale {from=MaterialExpressionMultiply_37}
   ]
   properties {materialFunction=&ScaleUVsByCenter}
  }
  MaterialExpressionMaterialFunctionCall_7 {
   class=MaterialExpressionMaterialFunctionCall
   inputs [Multiplier {from=MaterialExpressionConstant_8}]
   properties {materialFunction=&Pi}
  }
  MaterialExpressionMaterialFunctionCall_8 {
   class=MaterialExpressionMaterialFunctionCall
   inputs [
    UV Index {from=""}
    Texture {from=""}
    WPO Scale {from=""}
   ]
   properties {materialFunction=/BRCosmetics/Characters/Player/Female/Medium/Bodies/F_MED_FermiumStraw/Materials/MaterialFunctions/DEV/MF_UVTileDebug}
  }
  MaterialExpressionMaterialFunctionCall_9 {
   class=MaterialExpressionMaterialFunctionCall
   inputs [
    UVs {from=MaterialExpressionLinearInterpolate_53}
    Texture Scale {from=MaterialExpressionMultiply_33}
   ]
   properties {materialFunction=&ScaleUVsByCenter}
  }
  MaterialExpressionMax_2 {
   class=MaterialExpressionMax
   inputs [
    A {from=MaterialExpressionNamedRerouteDeclaration_43}
    B {from=MaterialExpressionNamedRerouteDeclaration_44}
   ]
  }
  MaterialExpressionMax_3 {
   class=MaterialExpressionMax
   inputs [
    A {from=MaterialExpressionMax_2}
    B {from=MaterialExpressionNamedRerouteDeclaration_45}
   ]
  }
  MaterialExpressionMax_4 {
   class=MaterialExpressionMax
   inputs [
    A {from=MaterialExpressionMax_6}
    B {from=MaterialExpressionMax_3}
   ]
  }
  MaterialExpressionMax_5 {
   class=MaterialExpressionMax
   inputs [
    A {from=MaterialExpressionNamedRerouteDeclaration_10}
    B {from=MaterialExpressionNamedRerouteDeclaration_31}
   ]
  }
  MaterialExpressionMax_6 {
   class=MaterialExpressionMax
   inputs [
    A {from=MaterialExpressionMax_5}
    B {from=MaterialExpressionNamedRerouteDeclaration_42}
   ]
  }
  MaterialExpressionMultiply_0 {
   class=MaterialExpressionMultiply
   inputs [
    A {from=MaterialExpressionScalarParameter_0}
    B {from=""}
   ]
   properties {constB=0}
  }
  MaterialExpressionMultiply_1 {
   class=MaterialExpressionMultiply
   inputs [
    A {from=MaterialExpressionConvert_22}
    B {from=MaterialExpressionNamedRerouteUsage_12}
   ]
  }
  MaterialExpressionMultiply_10 {
   class=MaterialExpressionMultiply
   inputs [
    A {from=MaterialExpressionScalarParameter_61}
    B {from=""}
   ]
   properties {constB=0}
  }
  MaterialExpressionMultiply_11 {
   class=MaterialExpressionMultiply
   inputs [
    A {from=MaterialExpressionScalarParameter_62}
    B {from=""}
   ]
   properties {constB=0}
  }
  MaterialExpressionMultiply_12 {
   class=MaterialExpressionMultiply
   inputs [
    A {from=MaterialExpressionNamedRerouteUsage_0}
    B {from=MaterialExpressionScalarParameter_5}
   ]
  }
  MaterialExpressionMultiply_13 {
   class=MaterialExpressionMultiply
   inputs [
    A {from=MaterialExpressionSphereMask_30}
    B {from=MaterialExpressionSphereMask_29}
   ]
  }
  MaterialExpressionMultiply_14 {
   class=MaterialExpressionMultiply
   inputs [
    A {from=MaterialExpressionScalarParameter_63}
    B {from=""}
   ]
   properties {constB=0}
  }
  MaterialExpressionMultiply_17 {
   class=MaterialExpressionMultiply
   inputs [
    A {from=MaterialExpressionMaterialFunctionCall_12, output=Result}
    B {from=MaterialExpressionNamedRerouteUsage_85}
   ]
  }
  MaterialExpressionMultiply_18 {
   class=MaterialExpressionMultiply
   inputs [
    A {from=MaterialExpressionSphereMask_4}
    B {from=MaterialExpressionSphereMask_13}
   ]
  }
  MaterialExpressionMultiply_19 {
   class=MaterialExpressionMultiply
   inputs [
    A {from=MaterialExpressionNamedRerouteUsage_90}
    B {from=MaterialExpressionConstant2Vector_5}
   ]
   properties {desc=aspect ratio}
  }
  MaterialExpressionMultiply_20 {
   class=MaterialExpressionMultiply
   inputs [
    A {from=MaterialExpressionVectorParameter_18, output=RGB}
    B {from=MaterialExpressionNamedRerouteUsage_115}
   ]
  }
  MaterialExpressionMultiply_21 {
   class=MaterialExpressionMultiply
   inputs [
    A {from=MaterialExpressionLinearInterpolate_28}
    B {from=MaterialExpressionLinearInterpolate_37}
   ]
  }
  MaterialExpressionMultiply_22 {
   class=MaterialExpressionMultiply
   inputs [
    A {from=MaterialExpressionNamedRerouteDeclaration_56}
    B {from=MaterialExpressionSwitch_1}
   ]
  }
  MaterialExpressionMultiply_24 {
   class=MaterialExpressionMultiply
   inputs [
    A {from=MaterialExpressionScalarParameter_24}
    B {from=MaterialExpressionPower_3}
   ]
  }
  MaterialExpressionMultiply_26 {
   class=MaterialExpressionMultiply
   inputs [
    A {from=MaterialExpressionNamedRerouteUsage_123}
    B {from=MaterialExpressionScalarParameter_7}
   ]
  }
  MaterialExpressionMultiply_27 {
   class=MaterialExpressionMultiply
   inputs [
    A {from=MaterialExpressionTextureSampleParameter2D_9, output=RGB}
    B {from=MaterialExpressionScalarParameter_27}
   ]
  }
  MaterialExpressionMultiply_28 {
   class=MaterialExpressionMultiply
   inputs [
    A {from=MaterialExpressionNamedRerouteUsage_124}
    B {from=MaterialExpressionScalarParameter_10}
   ]
  }
  MaterialExpressionMultiply_29 {
   class=MaterialExpressionMultiply
   inputs [
    A {from=MaterialExpressionTextureCoordinate_0}
    B {from=MaterialExpressionConstant2Vector_0}
   ]
  }
  MaterialExpressionMultiply_3 {
   class=MaterialExpressionMultiply
   inputs [
    A {from=MaterialExpressionConvert_20}
    B {from=MaterialExpressionScalarParameter_12}
   ]
  }
  MaterialExpressionMultiply_30 {
   class=MaterialExpressionMultiply
   inputs [
    A {from=MaterialExpressionTextureSampleParameter2D_8, output=RGB}
    B {from=MaterialExpressionVectorParameter_17, output=RGB}
   ]
  }
  MaterialExpressionMultiply_31 {
   class=MaterialExpressionMultiply
   inputs [
    A {from=MaterialExpressionAppendVector_13}
    B {from=MaterialExpressionScalarParameter_44}
   ]
  }
  MaterialExpressionMultiply_32 {
   class=MaterialExpressionMultiply
   inputs [
    A {from=MaterialExpressionAppendVector_19}
    B {from=MaterialExpressionScalarParameter_48}
   ]
  }
  MaterialExpressionMultiply_33 {
   class=MaterialExpressionMultiply
   inputs [
    A {from=MaterialExpressionAppendVector_21}
    B {from=MaterialExpressionScalarParameter_52}
   ]
  }
  MaterialExpressionMultiply_34 {
   class=MaterialExpressionMultiply
   inputs [
    A {from=MaterialExpressionTextureSampleParameter2D_10, output=RGB}
    B {from=MaterialExpressionMultiply_58}
   ]
  }
  MaterialExpressionMultiply_35 {
   class=MaterialExpressionMultiply
   inputs [
    A {from=MaterialExpressionMaterialFunctionCall_4, output=0-1 mask}
    B {from=""}
   ]
  }
  MaterialExpressionMultiply_36 {
   class=MaterialExpressionMultiply
   inputs [
    A {from=MaterialExpressionNamedRerouteUsage_87}
    B {from=MaterialExpressionMaterialFunctionCall_7, output=Result}
   ]
  }
  MaterialExpressionMultiply_37 {
   class=MaterialExpressionMultiply
   inputs [
    A {from=MaterialExpressionNamedRerouteUsage_92}
    B {from=MaterialExpressionConstant2Vector_2}
   ]
   properties {desc=aspect ratio}
  }
  MaterialExpressionMultiply_38 {
   class=MaterialExpressionMultiply
   inputs [
    A {from=MaterialExpressionNamedRerouteUsage_38}
    B {from=MaterialExpressionVectorParameter_2, output=RGB}
   ]
  }
  MaterialExpressionMultiply_39 {
   class=MaterialExpressionMultiply
   inputs [
    A {from=MaterialExpressionMaterialFunctionCall_16, output="Island 0-1 Mask (S)"}
    B {from=MaterialExpressionAdd_12}
   ]
  }
  MaterialExpressionMultiply_4 {
   class=MaterialExpressionMultiply
   inputs [
    A {from=MaterialExpressionNamedRerouteDeclaration_5}
    B {from=MaterialExpressionNamedRerouteUsage_16}
   ]
  }
  MaterialExpressionMultiply_40 {
   class=MaterialExpressionMultiply
   inputs [
    A {from=MaterialExpressionSphereMask_1}
    B {from=MaterialExpressionSphereMask_0}
   ]
  }
  MaterialExpressionMultiply_41 {
   class=MaterialExpressionMultiply
   inputs [
    A {from=MaterialExpressionNamedRerouteUsage_69}
    B {from=MaterialExpressionVectorParameter_23, output=RGB}
   ]
  }
  MaterialExpressionMultiply_42 {
   class=MaterialExpressionMultiply
   inputs [
    A {from=MaterialExpressionMaterialFunctionCall_18, output="Island 0-1 Mask (S)"}
    B {from=MaterialExpressionAdd_11}
   ]
  }
  MaterialExpressionMultiply_43 {
   class=MaterialExpressionMultiply
   inputs [
    A {from=MaterialExpressionMaterialFunctionCall_15, output="Island 0-1 Mask (S)"}
    B {from=MaterialExpressionAdd_8}
   ]
  }
  MaterialExpressionMultiply_44 {
   class=MaterialExpressionMultiply
   inputs [
    A {from=MaterialExpressionMaterialFunctionCall_2, output="Island 0-1 Mask (S)"}
    B {from=MaterialExpressionAdd_10}
   ]
  }
  MaterialExpressionMultiply_45 {
   class=MaterialExpressionMultiply
   inputs [
    A {from=MaterialExpressionSphereMask_2}
    B {from=MaterialExpressionSphereMask_3}
   ]
  }
  MaterialExpressionMultiply_46 {
   class=MaterialExpressionMultiply
   inputs [
    A {from=MaterialExpressionMultiply_22}
    B {from=MaterialExpressionDesaturation_1}
   ]
  }
  MaterialExpressionMultiply_47 {
   class=MaterialExpressionMultiply
   inputs [
    A {from=MaterialExpressionNamedRerouteUsage_108}
    B {from=MaterialExpressionLinearInterpolate_28}
   ]
  }
  MaterialExpressionMultiply_48 {
   class=MaterialExpressionMultiply
   inputs [
    A {from=MaterialExpressionScalarParameter_79}
    B {from=""}
   ]
   properties {constB=0}
  }
  MaterialExpressionMultiply_49 {
   class=MaterialExpressionMultiply
   inputs [
    A {from=MaterialExpressionAppendVector_3}
    B {from=MaterialExpressionScalarParameter_40}
   ]
  }
  MaterialExpressionMultiply_5 {
   class=MaterialExpressionMultiply
   inputs [
    A {from=MaterialExpressionMultiply_1}
    B {from=""}
   ]
   properties {constB=0.01}
  }
  MaterialExpressionMultiply_50 {
   class=MaterialExpressionMultiply
   inputs [
    A {from=MaterialExpressionSphereMask_5}
    B {from=MaterialExpressionSphereMask_6}
   ]
  }
  MaterialExpressionMultiply_51 {
   class=MaterialExpressionMultiply
   inputs [
    A {from=MaterialExpressionSphereMask_7}
    B {from=MaterialExpressionSphereMask_8}
   ]
  }
  MaterialExpressionMultiply_52 {
   class=MaterialExpressionMultiply
   inputs [
    A {from=MaterialExpressionSphereMask_9}
    B {from=MaterialExpressionSphereMask_10}
   ]
  }
  MaterialExpressionMultiply_53 {
   class=MaterialExpressionMultiply
   inputs [
    A {from=MaterialExpressionNamedRerouteUsage_98}
    B {from=MaterialExpressionVectorParameter_25, output=RGB}
   ]
  }
  MaterialExpressionMultiply_54 {
   class=MaterialExpressionMultiply
   inputs [
    A {from=MaterialExpressionNamedRerouteUsage_125}
    B {from=MaterialExpressionScalarParameter_16}
   ]
  }
  MaterialExpressionMultiply_55 {
   class=MaterialExpressionMultiply
   inputs [
    A {from=MaterialExpressionNamedRerouteUsage_126}
    B {from=MaterialExpressionScalarParameter_17}
   ]
  }
  MaterialExpressionMultiply_58 {
   class=MaterialExpressionMultiply
   inputs [
    A {from=MaterialExpressionSphereMask_11}
    B {from=MaterialExpressionSphereMask_12}
   ]
  }
  MaterialExpressionMultiply_6 {
   class=MaterialExpressionMultiply
   inputs [
    A {from=MaterialExpressionTextureSampleParameter2D_6, output=B}
    B {from=MaterialExpressionNamedRerouteUsage_56}
   ]
   properties {constB=0}
  }
  MaterialExpressionMultiply_8 {
   class=MaterialExpressionMultiply
   inputs [
    A {from=MaterialExpressionPower_0}
    B {from=MaterialExpressionVectorParameter_9, output=RGB}
   ]
  }
  MaterialExpressionNamedRerouteDeclaration_0 {
   class=MaterialExpressionNamedRerouteDeclaration
   inputs [None {from=MaterialExpressionConvert_8}]
   properties {name="SectionA UV Transforms (V4)", nodeColor=<rgba 1 0.6563 0 1>}
  }
  MaterialExpressionNamedRerouteDeclaration_1 {
   class=MaterialExpressionNamedRerouteDeclaration
   inputs [None {from=MaterialExpressionTextureCoordinate_6}]
   properties {name="UV Raw (V2)", nodeColor=<rgba 1 0 0.7969 1>}
  }
  MaterialExpressionNamedRerouteDeclaration_10 {
   class=MaterialExpressionNamedRerouteDeclaration
   inputs [None {from=MaterialExpressionTextureSampleParameter2D_5, output=R}]
   properties {name=Mask1_LoadingScreen, nodeColor=<rgba 1 0 0.9375 1>}
  }
  MaterialExpressionNamedRerouteDeclaration_11 {
   class=MaterialExpressionNamedRerouteDeclaration
   inputs [None {from=MaterialExpressionConvert_16}]
   properties {name="SectionC UV Transforms (V4)", nodeColor=<rgba 1 0.6562 0 1>}
  }
  MaterialExpressionNamedRerouteDeclaration_12 {
   class=MaterialExpressionNamedRerouteDeclaration
   inputs [None {from=MaterialExpressionSaturate_1}]
   properties {name="Enable User Transforms (V3)", nodeColor=<rgba 1 0 0.3047 1>}
  }
  MaterialExpressionNamedRerouteDeclaration_13 {
   class=MaterialExpressionNamedRerouteDeclaration
   inputs [None {from=MaterialExpressionMaterialFunctionCall_0, output=Result}]
   properties {name="Material In (MA)", nodeColor=<rgba 0.4531 1 0 1>}
  }
  MaterialExpressionNamedRerouteDeclaration_14 {
   class=MaterialExpressionNamedRerouteDeclaration
   inputs [None {from=MaterialExpressionMaterialFunctionCall_0, output=Tech Art Mask}]
   properties {name="TechArt Mask (V3)", nodeColor=<rgba 0.5703 0 1 1>}
  }
  MaterialExpressionNamedRerouteDeclaration_16 {
   class=MaterialExpressionNamedRerouteDeclaration
   inputs [None {from=MaterialExpressionNamedRerouteUsage_83}]
   properties {name="Projection - SectionB Mask (S)", nodeColor=<rgba 0.3828 0 1 1>}
  }
  MaterialExpressionNamedRerouteDeclaration_17 {
   class=MaterialExpressionNamedRerouteDeclaration
   inputs [None {from=MaterialExpressionSubtract_1}]
   properties {name="UV SectionA (V2)", nodeColor=<rgba 0 1 0.8828 1>}
  }
  MaterialExpressionNamedRerouteDeclaration_19 {
   class=MaterialExpressionNamedRerouteDeclaration
   inputs [None {from=MaterialExpressionLinearInterpolate_1}]
   properties {name="Outfit Color (V3)", nodeColor=<rgba 0 1 0.7891 1>}
  }
  MaterialExpressionNamedRerouteDeclaration_22 {
   class=MaterialExpressionNamedRerouteDeclaration
   inputs [None {from=MaterialExpressionSubtract_5}]
   properties {name="UV SectionB (V2)", nodeColor=<rgba 0 1 0.8828 1>}
  }
  MaterialExpressionNamedRerouteDeclaration_23 {
   class=MaterialExpressionNamedRerouteDeclaration
   inputs [None {from=MaterialExpressionSubtract_2}]
   properties {name="UV SectionC (V2)", nodeColor=<rgba 0 1 0.8828 1>}
  }
  MaterialExpressionNamedRerouteDeclaration_24 {
   class=MaterialExpressionNamedRerouteDeclaration
   inputs [None {from=MaterialExpressionNamedRerouteUsage_82}]
   properties {name="Projection - SectionA Mask (S)", nodeColor=<rgba 0.3828 0 1 1>}
  }
  MaterialExpressionNamedRerouteDeclaration_25 {
   class=MaterialExpressionNamedRerouteDeclaration
   inputs [None {from=MaterialExpressionMax_4}]
   properties {name=AllMasksCombined, nodeColor=<rgba 1 0 0.0703 1>}
  }
  MaterialExpressionNamedRerouteDeclaration_26 {
   class=MaterialExpressionNamedRerouteDeclaration
   inputs [None {from=MaterialExpressionAdd_2}]
   properties {name="Loading Screen Mask (S)", nodeColor=<rgba 1 0.1406 0 1>}
  }
  MaterialExpressionNamedRerouteDeclaration_27 {
   class=MaterialExpressionNamedRerouteDeclaration
   inputs [None {from=MaterialExpressionStaticSwitchParameter_2}]
   properties {name="Loading Screen Color (V3)", nodeColor=<rgba 0 0.1797 1 1>}
  }
  MaterialExpressionNamedRerouteDeclaration_28 {
   class=MaterialExpressionNamedRerouteDeclaration
   inputs [None {from=MaterialExpressionMultiply_8}]
   properties {name=Background Color, nodeColor=<rgba 0.5 1 0 1>}
  }
  MaterialExpressionNamedRerouteDeclaration_29 {
   class=MaterialExpressionNamedRerouteDeclaration
   inputs [None {from=MaterialExpressionConvert_23}]
   properties {name="User - SectionA Offset (S)", nodeColor=<rgba 0 0.625 1 1>}
  }
  MaterialExpressionNamedRerouteDeclaration_3 {
   class=MaterialExpressionNamedRerouteDeclaration
   inputs [None {from=MaterialExpressionConvert_11}]
   properties {name="U Offset (S)", nodeColor=<rgba 0.6172 0 1 1>}
  }
  MaterialExpressionNamedRerouteDeclaration_30 {
   class=MaterialExpressionNamedRerouteDeclaration
   inputs [None {from=MaterialExpressionGetMaterialAttributes_0, output=Base Color}]
   properties {name="BaseColor In (V3)", nodeColor=<rgba 0.6875 1 0 1>}
  }
  MaterialExpressionNamedRerouteDeclaration_31 {
   class=MaterialExpressionNamedRerouteDeclaration
   inputs [None {from=MaterialExpressionTextureSampleParameter2D_5, output=G}]
   properties {name=Mask2_PrimaryColor, nodeColor=<rgba 0 1 0 1>}
  }
  MaterialExpressionNamedRerouteDeclaration_32 {
   class=MaterialExpressionNamedRerouteDeclaration
   inputs [None {from=MaterialExpressionSubtract_4}]
   properties {name="UV SectionD (V2)", nodeColor=<rgba 0 1 0.8828 1>}
  }
  MaterialExpressionNamedRerouteDeclaration_33 {
   class=MaterialExpressionNamedRerouteDeclaration
   inputs [None {from=MaterialExpressionNamedRerouteUsage_84}]
   properties {name="Projection - SectionC Mask (S)", nodeColor=<rgba 0.2656 0 1 1>}
  }
  MaterialExpressionNamedRerouteDeclaration_34 {
   class=MaterialExpressionNamedRerouteDeclaration
   inputs [None {from=MaterialExpressionNamedRerouteUsage_100}]
   properties {name="Projection - SectionD Mask (S)", nodeColor=<rgba 0.2656 0 1 1>}
  }
  MaterialExpressionNamedRerouteDeclaration_35 {
   class=MaterialExpressionNamedRerouteDeclaration
   inputs [None {from=MaterialExpressionLinearInterpolate_14}]
   properties {name="Base UV (V2)", nodeColor=<rgba 0 0.1562 1 1>}
  }
  MaterialExpressionNamedRerouteDeclaration_36 {
   class=MaterialExpressionNamedRerouteDeclaration
   inputs [None {from=MaterialExpressionConvert_23}]
   properties {name="User - SectionB Offset (S)", nodeColor=<rgba 0 0.625 1 1>}
  }
  MaterialExpressionNamedRerouteDeclaration_37 {
   class=MaterialExpressionNamedRerouteDeclaration
   inputs [None {from=MaterialExpressionAppendVector_28}]
   properties {name="UV Offset (V2)", nodeColor=<rgba 0.5234 0 1 1>}
  }
  MaterialExpressionNamedRerouteDeclaration_38 {
   class=MaterialExpressionNamedRerouteDeclaration
   inputs [None {from=MaterialExpressionConvert_11}]
   properties {name="UV Rotation (S)", nodeColor=<rgba 0 0.5781 1 1>}
  }
  MaterialExpressionNamedRerouteDeclaration_39 {
   class=MaterialExpressionNamedRerouteDeclaration
   inputs [None {from=MaterialExpressionAppendVector_27}]
   properties {name="UV Scale (V2)", nodeColor=<rgba 0.3828 0 1 1>}
  }
  MaterialExpressionNamedRerouteDeclaration_4 {
   class=MaterialExpressionNamedRerouteDeclaration
   inputs [None {from=MaterialExpressionLinearInterpolate_4}]
   properties {name="UV Clamp Mask (S)", nodeColor=<rgba 0 0.4375 1 1>}
  }
  MaterialExpressionNamedRerouteDeclaration_40 {
   class=MaterialExpressionNamedRerouteDeclaration
   inputs [None {from=MaterialExpressionConvert_23}]
   properties {name="User - SectionC Offset (S)", nodeColor=<rgba 0 0.625 1 1>}
  }
  MaterialExpressionNamedRerouteDeclaration_41 {
   class=MaterialExpressionNamedRerouteDeclaration
   inputs [None {from=MaterialExpressionNamedRerouteDeclaration_5}]
   properties {name="LoadingScreen UV (V2)", nodeColor=<rgba 0 0.1562 1 1>}
  }
  MaterialExpressionNamedRerouteDeclaration_42 {
   class=MaterialExpressionNamedRerouteDeclaration
   inputs [None {from=MaterialExpressionTextureSampleParameter2D_5, output=B}]
   properties {name=Mask3_MutedColor, nodeColor=<rgba 1 0.0762 0.0762 1>}
  }
  MaterialExpressionNamedRerouteDeclaration_43 {
   class=MaterialExpressionNamedRerouteDeclaration
   inputs [None {from=MaterialExpressionTextureSampleParameter2D_6, output=R}]
   properties {name=Mask4_NeutralColor, nodeColor=<rgba 0 1 0.0859 1>}
  }
  MaterialExpressionNamedRerouteDeclaration_44 {
   class=MaterialExpressionNamedRerouteDeclaration
   inputs [None {from=MaterialExpressionTextureSampleParameter2D_6, output=G}]
   properties {name=Mask5_MetalColor, nodeColor=<rgba 1 0 0.2344 1>}
  }
  MaterialExpressionNamedRerouteDeclaration_45 {
   class=MaterialExpressionNamedRerouteDeclaration
   inputs [None {from=MaterialExpressionMultiply_6}]
   properties {name=Mask6_Splatter, nodeColor=<rgba 0 0.5547 1 1>}
  }
  MaterialExpressionNamedRerouteDeclaration_46 {
   class=MaterialExpressionNamedRerouteDeclaration
   inputs [None {from=MaterialExpressionBlendMaterialAttributes_1}]
   properties {name=Material_Out, nodeColor=<rgba 0 1 0.3906 1>}
  }
  MaterialExpressionNamedRerouteDeclaration_47 {
   class=MaterialExpressionNamedRerouteDeclaration
   inputs [None {from=MaterialExpressionVectorParameter_17, output=A}]
   properties {name=EnableSplatter, nodeColor=<rgba 0.6875 0 1 1>}
  }
  MaterialExpressionNamedRerouteDeclaration_48 {
   class=MaterialExpressionNamedRerouteDeclaration
   inputs [None {from=MaterialExpressionMaterialFunctionCall_0, output=Fresnel}]
   properties {name=Fresnel, nodeColor=<rgba 0.2422 0 1 1>}
  }
  MaterialExpressionNamedRerouteDeclaration_49 {
   class=MaterialExpressionNamedRerouteDeclaration
   inputs [None {from=MaterialExpressionMultiply_44}]
   properties {name=UV SectionB Mask, nodeColor=<rgba 0 1 0.8828 1>}
  }
  MaterialExpressionNamedRerouteDeclaration_5 {
   class=MaterialExpressionNamedRerouteDeclaration
   inputs [None {from=MaterialExpressionSubtract_17}]
   properties {name="Transformed UV (V2)", nodeColor=<rgba 0.7578 1 0 1>}
  }
  MaterialExpressionNamedRerouteDeclaration_50 {
   class=MaterialExpressionNamedRerouteDeclaration
   inputs [None {from=MaterialExpressionSubtract_0}]
   properties {name=ImageUV, nodeColor=<rgba 0.3594 1 0 1>}
  }
  MaterialExpressionNamedRerouteDeclaration_51 {
   class=MaterialExpressionNamedRerouteDeclaration
   inputs [None {from=MaterialExpressionMultiply_43}]
   properties {name=UV SectionA Mask, nodeColor=<rgba 0 1 0.8828 1>}
  }
  MaterialExpressionNamedRerouteDeclaration_52 {
   class=MaterialExpressionNamedRerouteDeclaration
   inputs [None {from=MaterialExpressionMultiply_42}]
   properties {name=UV SectionC Mask, nodeColor=<rgba 0 1 0.8828 1>}
  }
  MaterialExpressionNamedRerouteDeclaration_53 {
   class=MaterialExpressionNamedRerouteDeclaration
   inputs [None {from=MaterialExpressionMultiply_39}]
   properties {name=UV SectionD Mask, nodeColor=<rgba 0 1 0.8828 1>}
  }
  MaterialExpressionNamedRerouteDeclaration_54 {
   class=MaterialExpressionNamedRerouteDeclaration
   inputs [None {from=MaterialExpressionLinearInterpolate_40}]
   properties {name=UVChannel, nodeColor=<rgba 0 0.1562 1 1>}
  }
  MaterialExpressionNamedRerouteDeclaration_55 {
   class=MaterialExpressionNamedRerouteDeclaration
   inputs [None {from=MaterialExpressionLinearInterpolate_48}]
   properties {name=UDIMOffset, nodeColor=<rgba 0 0.1562 1 1>}
  }
  MaterialExpressionNamedRerouteDeclaration_56 {
   class=MaterialExpressionNamedRerouteDeclaration
   inputs [None {from=MaterialExpressionMultiply_18}]
   properties {name=DEBUG_UDIMMASK, nodeColor=<rgba 1 0 0.1641 1>}
  }
  MaterialExpressionNamedRerouteDeclaration_57 {
   class=MaterialExpressionNamedRerouteDeclaration
   inputs [None {from=MaterialExpressionGetMaterialAttributes_7, output=Base Color}]
   properties {name=InMat_BC, nodeColor=<rgba 1 0 0.4922 1>}
  }
  MaterialExpressionNamedRerouteDeclaration_58 {
   class=MaterialExpressionNamedRerouteDeclaration
   inputs [None {from=MaterialExpressionMultiply_45}]
   properties {name=UDIMMask_SectionA, nodeColor=<rgba 1 0 0.1641 1>}
  }
  MaterialExpressionNamedRerouteDeclaration_59 {
   class=MaterialExpressionNamedRerouteDeclaration
   inputs [None {from=MaterialExpressionMultiply_50}]
   properties {name=UDIMMask_SectionB, nodeColor=<rgba 1 0 0.1641 1>}
  }
  MaterialExpressionNamedRerouteDeclaration_6 {
   class=MaterialExpressionNamedRerouteDeclaration
   inputs [None {from=MaterialExpressionConvert_11}]
   properties {name="V Offset (S)", nodeColor=<rgba 0.6172 0 1 1>}
  }
  MaterialExpressionNamedRerouteDeclaration_60 {
   class=MaterialExpressionNamedRerouteDeclaration
   inputs [None {from=MaterialExpressionMultiply_52}]
   properties {name=UDIMMask_SectionC, nodeColor=<rgba 1 0 0.1641 1>}
  }
  MaterialExpressionNamedRerouteDeclaration_61 {
   class=MaterialExpressionNamedRerouteDeclaration
   inputs [None {from=MaterialExpressionMultiply_51}]
   properties {name=UDIMMask_SectionD, nodeColor=<rgba 1 0 0.1641 1>}
  }
  MaterialExpressionNamedRerouteDeclaration_62 {
   class=MaterialExpressionNamedRerouteDeclaration
   inputs [None {from=MaterialExpressionGetMaterialAttributes_7, output=Roughness}]
   properties {name=InMat_R, nodeColor=<rgba 1 0 0.4922 1>}
  }
  MaterialExpressionNamedRerouteDeclaration_63 {
   class=MaterialExpressionNamedRerouteDeclaration
   inputs [None {from=MaterialExpressionSubtract_6}]
   properties {name=UVRaw_SectionA_1, nodeColor=<rgba 0.2656 1 0 1>}
  }
  MaterialExpressionNamedRerouteDeclaration_64 {
   class=MaterialExpressionNamedRerouteDeclaration
   inputs [None {from=MaterialExpressionSwitch_0}]
   properties {name=DEBUG_UV, nodeColor=<rgba 0 0.9531 1 1>}
  }
  MaterialExpressionNamedRerouteDeclaration_65 {
   class=MaterialExpressionNamedRerouteDeclaration
   inputs [None {from=MaterialExpressionSubtract_7}]
   properties {name=UVRaw_SectionA, nodeColor=<rgba 0.2656 1 0 1>}
  }
  MaterialExpressionNamedRerouteDeclaration_66 {
   class=MaterialExpressionNamedRerouteDeclaration
   inputs [None {from=MaterialExpressionSubtract_10}]
   properties {name=UVRaw_SectionB, nodeColor=<rgba 0.2656 1 0 1>}
  }
  MaterialExpressionNamedRerouteDeclaration_67 {
   class=MaterialExpressionNamedRerouteDeclaration
   inputs [None {from=MaterialExpressionSubtract_11}]
   properties {name=UVRaw_SectionD, nodeColor=<rgba 0.2656 1 0 1>}
  }
  MaterialExpressionNamedRerouteDeclaration_68 {
   class=MaterialExpressionNamedRerouteDeclaration
   inputs [None {from=MaterialExpressionSubtract_12}]
   properties {name=UVRaw_SectionC, nodeColor=<rgba 0.2656 1 0 1>}
  }
  MaterialExpressionNamedRerouteDeclaration_69 {
   class=MaterialExpressionNamedRerouteDeclaration
   inputs [None {from=MaterialExpressionMultiply_46}]
   properties {name=BaseUV_Debug, nodeColor=<rgba 1 0.7266 0 1>}
  }
  MaterialExpressionNamedRerouteDeclaration_7 {
   class=MaterialExpressionNamedRerouteDeclaration
   inputs [None {from=MaterialExpressionConvert_9}]
   properties {name="SectionB UV Transforms (V4)", nodeColor=<rgba 1 0.6562 0 1>}
  }
  MaterialExpressionNamedRerouteDeclaration_70 {
   class=MaterialExpressionNamedRerouteDeclaration
   inputs [None {from=MaterialExpressionConvert_19}]
   properties {name="SectionD UV Transforms (V4)", nodeColor=<rgba 1 0.6562 0 1>}
  }
  MaterialExpressionNamedRerouteDeclaration_71 {
   class=MaterialExpressionNamedRerouteDeclaration
   inputs [None {from=MaterialExpressionConvert_23}]
   properties {name="User - SectionD Offset (S)", nodeColor=<rgba 0 0.625 1 1>}
  }
  MaterialExpressionNamedRerouteDeclaration_8 {
   class=MaterialExpressionNamedRerouteDeclaration
   inputs [None {from=MaterialExpressionConvert_11}]
   properties {name="UV Scale (S)", nodeColor=<rgba 0.3828 0 1 1>}
  }
  MaterialExpressionNamedRerouteUsage_0 {class=MaterialExpressionNamedRerouteUsage}
  MaterialExpressionNamedRerouteUsage_10 {class=MaterialExpressionNamedRerouteUsage}
  MaterialExpressionNamedRerouteUsage_100 {class=MaterialExpressionNamedRerouteUsage}
  MaterialExpressionNamedRerouteUsage_101 {class=MaterialExpressionNamedRerouteUsage}
  MaterialExpressionNamedRerouteUsage_102 {class=MaterialExpressionNamedRerouteUsage}
  MaterialExpressionNamedRerouteUsage_103 {class=MaterialExpressionNamedRerouteUsage}
  MaterialExpressionNamedRerouteUsage_108 {class=MaterialExpressionNamedRerouteUsage}
  MaterialExpressionNamedRerouteUsage_109 {class=MaterialExpressionNamedRerouteUsage}
  MaterialExpressionNamedRerouteUsage_11 {class=MaterialExpressionNamedRerouteUsage}
  MaterialExpressionNamedRerouteUsage_110 {class=MaterialExpressionNamedRerouteUsage}
  MaterialExpressionNamedRerouteUsage_112 {class=MaterialExpressionNamedRerouteUsage}
  MaterialExpressionNamedRerouteUsage_113 {class=MaterialExpressionNamedRerouteUsage}
  MaterialExpressionNamedRerouteUsage_115 {class=MaterialExpressionNamedRerouteUsage}
  MaterialExpressionNamedRerouteUsage_116 {class=MaterialExpressionNamedRerouteUsage}
  MaterialExpressionNamedRerouteUsage_117 {class=MaterialExpressionNamedRerouteUsage}
  MaterialExpressionNamedRerouteUsage_118 {class=MaterialExpressionNamedRerouteUsage}
  MaterialExpressionNamedRerouteUsage_12 {class=MaterialExpressionNamedRerouteUsage}
  MaterialExpressionNamedRerouteUsage_123 {class=MaterialExpressionNamedRerouteUsage}
  MaterialExpressionNamedRerouteUsage_124 {class=MaterialExpressionNamedRerouteUsage}
  MaterialExpressionNamedRerouteUsage_125 {class=MaterialExpressionNamedRerouteUsage}
  MaterialExpressionNamedRerouteUsage_126 {class=MaterialExpressionNamedRerouteUsage}
  MaterialExpressionNamedRerouteUsage_127 {class=MaterialExpressionNamedRerouteUsage}
  MaterialExpressionNamedRerouteUsage_128 {class=MaterialExpressionNamedRerouteUsage}
  MaterialExpressionNamedRerouteUsage_129 {class=MaterialExpressionNamedRerouteUsage}
  MaterialExpressionNamedRerouteUsage_130 {class=MaterialExpressionNamedRerouteUsage}
  MaterialExpressionNamedRerouteUsage_14 {class=MaterialExpressionNamedRerouteUsage}
  MaterialExpressionNamedRerouteUsage_15 {class=MaterialExpressionNamedRerouteUsage}
  MaterialExpressionNamedRerouteUsage_16 {class=MaterialExpressionNamedRerouteUsage}
  MaterialExpressionNamedRerouteUsage_17 {class=MaterialExpressionNamedRerouteUsage}
  MaterialExpressionNamedRerouteUsage_18 {class=MaterialExpressionNamedRerouteUsage}
  MaterialExpressionNamedRerouteUsage_2 {class=MaterialExpressionNamedRerouteUsage}
  MaterialExpressionNamedRerouteUsage_22 {class=MaterialExpressionNamedRerouteUsage}
  MaterialExpressionNamedRerouteUsage_23 {class=MaterialExpressionNamedRerouteUsage}
  MaterialExpressionNamedRerouteUsage_24 {class=MaterialExpressionNamedRerouteUsage}
  MaterialExpressionNamedRerouteUsage_25 {class=MaterialExpressionNamedRerouteUsage}
  MaterialExpressionNamedRerouteUsage_27 {class=MaterialExpressionNamedRerouteUsage}
  MaterialExpressionNamedRerouteUsage_3 {class=MaterialExpressionNamedRerouteUsage}
  MaterialExpressionNamedRerouteUsage_30 {class=MaterialExpressionNamedRerouteUsage}
  MaterialExpressionNamedRerouteUsage_33 {class=MaterialExpressionNamedRerouteUsage}
  MaterialExpressionNamedRerouteUsage_34 {class=MaterialExpressionNamedRerouteUsage}
  MaterialExpressionNamedRerouteUsage_35 {class=MaterialExpressionNamedRerouteUsage}
  MaterialExpressionNamedRerouteUsage_36 {class=MaterialExpressionNamedRerouteUsage}
  MaterialExpressionNamedRerouteUsage_37 {class=MaterialExpressionNamedRerouteUsage}
  MaterialExpressionNamedRerouteUsage_38 {class=MaterialExpressionNamedRerouteUsage}
  MaterialExpressionNamedRerouteUsage_39 {class=MaterialExpressionNamedRerouteUsage}
  MaterialExpressionNamedRerouteUsage_4 {class=MaterialExpressionNamedRerouteUsage}
  MaterialExpressionNamedRerouteUsage_40 {class=MaterialExpressionNamedRerouteUsage}
  MaterialExpressionNamedRerouteUsage_41 {class=MaterialExpressionNamedRerouteUsage}
  MaterialExpressionNamedRerouteUsage_42 {class=MaterialExpressionNamedRerouteUsage}
  MaterialExpressionNamedRerouteUsage_43 {class=MaterialExpressionNamedRerouteUsage}
  MaterialExpressionNamedRerouteUsage_44 {class=MaterialExpressionNamedRerouteUsage}
  MaterialExpressionNamedRerouteUsage_45 {class=MaterialExpressionNamedRerouteUsage}
  MaterialExpressionNamedRerouteUsage_46 {class=MaterialExpressionNamedRerouteUsage}
  MaterialExpressionNamedRerouteUsage_47 {class=MaterialExpressionNamedRerouteUsage}
  MaterialExpressionNamedRerouteUsage_48 {class=MaterialExpressionNamedRerouteUsage}
  MaterialExpressionNamedRerouteUsage_49 {class=MaterialExpressionNamedRerouteUsage}
  MaterialExpressionNamedRerouteUsage_5 {class=MaterialExpressionNamedRerouteUsage}
  MaterialExpressionNamedRerouteUsage_50 {class=MaterialExpressionNamedRerouteUsage}
  MaterialExpressionNamedRerouteUsage_51 {class=MaterialExpressionNamedRerouteUsage}
  MaterialExpressionNamedRerouteUsage_52 {class=MaterialExpressionNamedRerouteUsage}
  MaterialExpressionNamedRerouteUsage_53 {class=MaterialExpressionNamedRerouteUsage}
  MaterialExpressionNamedRerouteUsage_54 {class=MaterialExpressionNamedRerouteUsage}
  MaterialExpressionNamedRerouteUsage_55 {class=MaterialExpressionNamedRerouteUsage}
  MaterialExpressionNamedRerouteUsage_56 {class=MaterialExpressionNamedRerouteUsage}
  MaterialExpressionNamedRerouteUsage_58 {class=MaterialExpressionNamedRerouteUsage}
  MaterialExpressionNamedRerouteUsage_59 {class=MaterialExpressionNamedRerouteUsage}
  MaterialExpressionNamedRerouteUsage_6 {class=MaterialExpressionNamedRerouteUsage}
  MaterialExpressionNamedRerouteUsage_60 {class=MaterialExpressionNamedRerouteUsage}
  MaterialExpressionNamedRerouteUsage_61 {class=MaterialExpressionNamedRerouteUsage}
  MaterialExpressionNamedRerouteUsage_62 {class=MaterialExpressionNamedRerouteUsage}
  MaterialExpressionNamedRerouteUsage_63 {class=MaterialExpressionNamedRerouteUsage}
  MaterialExpressionNamedRerouteUsage_65 {class=MaterialExpressionNamedRerouteUsage}
  MaterialExpressionNamedRerouteUsage_66 {class=MaterialExpressionNamedRerouteUsage}
  MaterialExpressionNamedRerouteUsage_67 {class=MaterialExpressionNamedRerouteUsage}
  MaterialExpressionNamedRerouteUsage_68 {class=MaterialExpressionNamedRerouteUsage}
  MaterialExpressionNamedRerouteUsage_69 {class=MaterialExpressionNamedRerouteUsage}
  MaterialExpressionNamedRerouteUsage_70 {class=MaterialExpressionNamedRerouteUsage}
  MaterialExpressionNamedRerouteUsage_71 {class=MaterialExpressionNamedRerouteUsage}
  MaterialExpressionNamedRerouteUsage_72 {class=MaterialExpressionNamedRerouteUsage}
  MaterialExpressionNamedRerouteUsage_73 {class=MaterialExpressionNamedRerouteUsage}
  MaterialExpressionNamedRerouteUsage_74 {class=MaterialExpressionNamedRerouteUsage}
  MaterialExpressionNamedRerouteUsage_75 {class=MaterialExpressionNamedRerouteUsage}
  MaterialExpressionNamedRerouteUsage_76 {class=MaterialExpressionNamedRerouteUsage}
  MaterialExpressionNamedRerouteUsage_77 {class=MaterialExpressionNamedRerouteUsage}
  MaterialExpressionNamedRerouteUsage_78 {class=MaterialExpressionNamedRerouteUsage}
  MaterialExpressionNamedRerouteUsage_79 {class=MaterialExpressionNamedRerouteUsage}
  MaterialExpressionNamedRerouteUsage_80 {class=MaterialExpressionNamedRerouteUsage}
  MaterialExpressionNamedRerouteUsage_81 {class=MaterialExpressionNamedRerouteUsage}
  MaterialExpressionNamedRerouteUsage_82 {class=MaterialExpressionNamedRerouteUsage}
  MaterialExpressionNamedRerouteUsage_83 {class=MaterialExpressionNamedRerouteUsage}
  MaterialExpressionNamedRerouteUsage_84 {class=MaterialExpressionNamedRerouteUsage}
  MaterialExpressionNamedRerouteUsage_85 {class=MaterialExpressionNamedRerouteUsage}
  MaterialExpressionNamedRerouteUsage_86 {class=MaterialExpressionNamedRerouteUsage}
  MaterialExpressionNamedRerouteUsage_87 {class=MaterialExpressionNamedRerouteUsage}
  MaterialExpressionNamedRerouteUsage_88 {class=MaterialExpressionNamedRerouteUsage}
  MaterialExpressionNamedRerouteUsage_89 {class=MaterialExpressionNamedRerouteUsage}
  MaterialExpressionNamedRerouteUsage_90 {class=MaterialExpressionNamedRerouteUsage}
  MaterialExpressionNamedRerouteUsage_91 {class=MaterialExpressionNamedRerouteUsage}
  MaterialExpressionNamedRerouteUsage_92 {class=MaterialExpressionNamedRerouteUsage}
  MaterialExpressionNamedRerouteUsage_93 {class=MaterialExpressionNamedRerouteUsage}
  MaterialExpressionNamedRerouteUsage_94 {class=MaterialExpressionNamedRerouteUsage}
  MaterialExpressionNamedRerouteUsage_95 {class=MaterialExpressionNamedRerouteUsage}
  MaterialExpressionNamedRerouteUsage_96 {class=MaterialExpressionNamedRerouteUsage}
  MaterialExpressionNamedRerouteUsage_97 {class=MaterialExpressionNamedRerouteUsage}
  MaterialExpressionNamedRerouteUsage_98 {class=MaterialExpressionNamedRerouteUsage}
  MaterialExpressionNamedRerouteUsage_99 {class=MaterialExpressionNamedRerouteUsage}
  MaterialExpressionOneMinus_10 {class=MaterialExpressionOneMinus, inputs [None {from=MaterialExpressionMaterialFunctionCall_25, output=R}]}
  MaterialExpressionOneMinus_12 {class=MaterialExpressionOneMinus, inputs [None {from=MaterialExpressionLinearInterpolate_44}]}
  MaterialExpressionOneMinus_13 {class=MaterialExpressionOneMinus, inputs [None {from=MaterialExpressionLinearInterpolate_45}]}
  MaterialExpressionOneMinus_14 {class=MaterialExpressionOneMinus, inputs [None {from=MaterialExpressionLinearInterpolate_46}]}
  MaterialExpressionOneMinus_15 {class=MaterialExpressionOneMinus, inputs [None {from=MaterialExpressionLinearInterpolate_43}]}
  MaterialExpressionOneMinus_2 {class=MaterialExpressionOneMinus, inputs [None {from=MaterialExpressionAbs_0}]}
  MaterialExpressionOneMinus_6 {class=MaterialExpressionOneMinus, inputs [None {from=MaterialExpressionMaterialFunctionCall_21, output=R}]}
  MaterialExpressionOneMinus_7 {class=MaterialExpressionOneMinus, inputs [None {from=MaterialExpressionMaterialFunctionCall_23, output=R}]}
  MaterialExpressionOneMinus_8 {class=MaterialExpressionOneMinus, inputs [None {from=MaterialExpressionMaterialFunctionCall_24, output=R}]}
  MaterialExpressionPower_0 {
   class=MaterialExpressionPower
   inputs [
    Base {from=MaterialExpressionNamedRerouteDeclaration_30}
    Exp {from=MaterialExpressionScalarParameter_4}
   ]
   properties {constExponent=0.25}
  }
  MaterialExpressionPower_3 {
   class=MaterialExpressionPower
   inputs [
    Base {from=MaterialExpressionNamedRerouteUsage_55}
    Exp {from=MaterialExpressionScalarParameter_23}
   ]
  }
  MaterialExpressionPower_4 {
   class=MaterialExpressionPower
   inputs [
    Base {from=MaterialExpressionNamedRerouteUsage_55}
    Exp {from=MaterialExpressionScalarParameter_26}
   ]
  }
  MaterialExpressionReroute_1 {class=MaterialExpressionReroute, inputs [None {from=MaterialExpressionBlendMaterialAttributes_2}]}
  MaterialExpressionReroute_13 {class=MaterialExpressionReroute, inputs [None {from=MaterialExpressionReroute_8}]}
  MaterialExpressionReroute_14 {class=MaterialExpressionReroute, inputs [None {from=MaterialExpressionMaterialFunctionCall_8, output="WPO (V3)"}]}
  MaterialExpressionReroute_2 {class=MaterialExpressionReroute, inputs [None {from=MaterialExpressionMaterialFunctionCall_18, output="Island UV (V2)"}]}
  MaterialExpressionReroute_3 {class=MaterialExpressionReroute, inputs [None {from=MaterialExpressionReroute_2}]}
  MaterialExpressionReroute_4 {class=MaterialExpressionReroute, inputs [None {from=MaterialExpressionReroute_6}]}
  MaterialExpressionReroute_5 {class=MaterialExpressionReroute, inputs [None {from=MaterialExpressionSetMaterialAttributes_4}]}
  MaterialExpressionReroute_6 {class=MaterialExpressionReroute, inputs [None {from=MaterialExpressionMaterialFunctionCall_15, output="Island UV (V2)"}]}
  MaterialExpressionReroute_8 {class=MaterialExpressionReroute, inputs [None {from=MaterialExpressionMaterialFunctionCall_1, output=Result}]}
  MaterialExpressionRotator_0 {
   class=MaterialExpressionRotator
   inputs [
    Coordinate {from=MaterialExpressionNamedRerouteUsage_86, output=""}
    Time {from=MaterialExpressionMultiply_36, output=""}
   ]
   properties {speed=-2}
  }
  MaterialExpressionRotator_2 {
   class=MaterialExpressionRotator
   inputs [
    Coordinate {from=MaterialExpressionNamedRerouteUsage_88, output=""}
    Time {from=MaterialExpressionMultiply_17, output=""}
   ]
   properties {speed=-1}
  }
  MaterialExpressionRound_1 {class=MaterialExpressionRound, inputs [None {from=MaterialExpressionMultiply_3, output=""}], properties {}}
  MaterialExpressionSaturate_1 {class=MaterialExpressionSaturate, inputs [None {from=MaterialExpressionRound_1}]}
  MaterialExpressionSaturate_2 {class=MaterialExpressionSaturate, inputs [None {from=MaterialExpressionPower_4}]}
  MaterialExpressionSaturate_3 {class=MaterialExpressionSaturate, inputs [None {from=MaterialExpressionSign_0}]}
  MaterialExpressionScalarParameter_0 {
   class=MaterialExpressionScalarParameter
   properties {group=FermiumStraw | 03_DEV, parameterName=Enable Base Transforms, sliderMax=1, sortPriority=10}
  }
  MaterialExpressionScalarParameter_10 {class=MaterialExpressionScalarParameter, properties {group=00_FermiumStraw per mask Config, parameterName=SectionB_Enable, sortPriority=20}}
  MaterialExpressionScalarParameter_12 {class=MaterialExpressionScalarParameter, properties {group=FermiumStraw | 03_DEV, parameterName=Enable User Transforms, sortPriority=20}}
  MaterialExpressionScalarParameter_13 {
   class=MaterialExpressionScalarParameter
   properties {group=FermiumStraw | 02_Transforms | 02_User Transforms, parameterName=SectionA_LoadingScreen_UVTransform_UserOffset, sortPriority=15}
  }
  MaterialExpressionScalarParameter_14 {
   class=MaterialExpressionScalarParameter
   properties {group=FermiumStraw | 02_Transforms | 02_User Transforms, parameterName=SectionB_LoadingScreen_UVTransform_UserOffset, sortPriority=25}
  }
  MaterialExpressionScalarParameter_16 {class=MaterialExpressionScalarParameter, properties {group=00_FermiumStraw per mask Config, parameterName=SectionC_Enable, sortPriority=30}}
  MaterialExpressionScalarParameter_17 {class=MaterialExpressionScalarParameter, properties {group=00_FermiumStraw per mask Config, parameterName=SectionD_Enable, sortPriority=40}}
  MaterialExpressionScalarParameter_18 {
   class=MaterialExpressionScalarParameter
   properties {
    defaultValue=1
    group=FermiumStraw | 02_Transforms | 02_User Transforms
    parameterName=SectionA_LoadingScreen_UVTransform_EnableUserOffset
    sortPriority=15
   }
  }
  MaterialExpressionScalarParameter_19 {
   class=MaterialExpressionScalarParameter
   properties {
    defaultValue=1
    group=FermiumStraw | 02_Transforms | 02_User Transforms
    parameterName=SectionB_LoadingScreen_UVTransform_EnableUserOffset
    sortPriority=25
   }
  }
  MaterialExpressionScalarParameter_2 {
   class=MaterialExpressionScalarParameter
   properties {defaultValue=1, group=00_FermiumStraw per mask Config, parameterName=SectionA_UVChannel, sortPriority=15}
  }
  MaterialExpressionScalarParameter_21 {
   class=MaterialExpressionScalarParameter
   properties {group=00_FermiumStraw per mask Config, parameterName=SectionA_UDIMOffset, sortPriority=15}
  }
  MaterialExpressionScalarParameter_22 {class=MaterialExpressionScalarParameter, properties {defaultValue=1, group=00_FermiumStraw per mask Config, parameterName=DEBUG_UVChannel}}
  MaterialExpressionScalarParameter_23 {
   class=MaterialExpressionScalarParameter
   properties {defaultValue=0.5, group=FermiumStraw | 03_Customization, parameterName=Metal_ThinFilmWarp, sortPriority=4}
  }
  MaterialExpressionScalarParameter_24 {
   class=MaterialExpressionScalarParameter
   properties {defaultValue=1, group=FermiumStraw | 03_Customization, parameterName=Metal_ThinFilmScale, sortPriority=5}
  }
  MaterialExpressionScalarParameter_25 {class=MaterialExpressionScalarParameter, properties {group=FermiumStraw | 03_Customization, parameterName=Metal_ThinFilmOffset}}
  MaterialExpressionScalarParameter_26 {
   class=MaterialExpressionScalarParameter
   properties {defaultValue=2, group=FermiumStraw | 03_Customization, parameterName=Metal_ThinFilmExponent, sortPriority=3}
  }
  MaterialExpressionScalarParameter_27 {
   class=MaterialExpressionScalarParameter
   properties {defaultValue=1, group=FermiumStraw | 03_Customization, parameterName=Metal_ThinFilmIntensity, sortPriority=2}
  }
  MaterialExpressionScalarParameter_28 {class=MaterialExpressionScalarParameter, properties {group=00_FermiumStraw per mask Config, parameterName=SectionD_Enable, sortPriority=40}}
  MaterialExpressionScalarParameter_29 {class=MaterialExpressionScalarParameter, properties {group=00_FermiumStraw per mask Config, parameterName=SectionC_Enable, sortPriority=30}}
  MaterialExpressionScalarParameter_3 {
   class=MaterialExpressionScalarParameter
   properties {defaultValue=2, group=zz_SIGNAL DEBUG, parameterName=Flattened UV Z Offset, sortPriority=-1490}
  }
  MaterialExpressionScalarParameter_30 {
   class=MaterialExpressionScalarParameter
   properties {defaultValue=1, group=00_FermiumStraw per mask Config, parameterName=SectionA_Enable, sortPriority=10}
  }
  MaterialExpressionScalarParameter_31 {class=MaterialExpressionScalarParameter, properties {group=00_FermiumStraw per mask Config, parameterName=SectionB_Enable, sortPriority=20}}
  MaterialExpressionScalarParameter_32 {class=MaterialExpressionScalarParameter, properties {parameterName=DebugFermiumUVs}}
  MaterialExpressionScalarParameter_33 {class=MaterialExpressionScalarParameter, properties {parameterName=DebugFermiumMasks}}
  MaterialExpressionScalarParameter_34 {class=MaterialExpressionScalarParameter, properties {defaultValue=0.75, parameterName=DebugFermiumUVs_Opacity}}
  MaterialExpressionScalarParameter_35 {class=MaterialExpressionScalarParameter, properties {group=00_FermiumStraw per mask Config, parameterName=SectionB_Swizzle, sortPriority=25}}
  MaterialExpressionScalarParameter_36 {class=MaterialExpressionScalarParameter, properties {group=00_FermiumStraw per mask Config, parameterName=SectionA_Swizzle, sortPriority=15}}
  MaterialExpressionScalarParameter_37 {class=MaterialExpressionScalarParameter, properties {group=00_FermiumStraw per mask Config, parameterName=SectionC_Swizzle, sortPriority=35}}
  MaterialExpressionScalarParameter_38 {class=MaterialExpressionScalarParameter, properties {group=00_FermiumStraw per mask Config, parameterName=SectionD_Swizzle, sortPriority=45}}
  MaterialExpressionScalarParameter_39 {
   class=MaterialExpressionScalarParameter
   properties {defaultValue=1, group=00_FermiumStraw per mask Config, parameterName=SectionA_AspectRatio, sliderMax=4, sliderMin=0.5, sortPriority=15}
  }
  MaterialExpressionScalarParameter_4 {
   class=MaterialExpressionScalarParameter
   properties {
    defaultValue=0.25
    desc=To brighten up the basecolor below the loading screen before multiplying it
    group=FermiumStraw | 01_Appearance
    parameterName=LoadingScreen_BackgroundGamma
    sortPriority=10
   }
  }
  MaterialExpressionScalarParameter_40 {
   class=MaterialExpressionScalarParameter
   properties {defaultValue=1, group=00_FermiumStraw per mask Config, parameterName=SectionA_Zoom, sortPriority=15}
  }
  MaterialExpressionScalarParameter_41 {
   class=MaterialExpressionScalarParameter
   properties {group=00_FermiumStraw per mask Config, parameterName=SectionA_UVOffset_U, sliderMax=1, sliderMin=-1, sortPriority=15}
  }
  MaterialExpressionScalarParameter_42 {
   class=MaterialExpressionScalarParameter
   properties {group=00_FermiumStraw per mask Config, parameterName=SectionA_UVOffset_V, sliderMax=1, sliderMin=-1, sortPriority=15}
  }
  MaterialExpressionScalarParameter_43 {
   class=MaterialExpressionScalarParameter
   properties {defaultValue=1, group=00_FermiumStraw per mask Config, parameterName=SectionC_AspectRatio, sliderMax=4, sliderMin=0.5, sortPriority=35}
  }
  MaterialExpressionScalarParameter_44 {
   class=MaterialExpressionScalarParameter
   properties {defaultValue=1, group=00_FermiumStraw per mask Config, parameterName=SectionC_Zoom, sortPriority=35}
  }
  MaterialExpressionScalarParameter_45 {
   class=MaterialExpressionScalarParameter
   properties {group=00_FermiumStraw per mask Config, parameterName=SectionC_UVOffset_U, sliderMax=1, sliderMin=-1, sortPriority=35}
  }
  MaterialExpressionScalarParameter_46 {
   class=MaterialExpressionScalarParameter
   properties {group=00_FermiumStraw per mask Config, parameterName=SectionC_UVOffset_V, sliderMax=1, sliderMin=-1, sortPriority=35}
  }
  MaterialExpressionScalarParameter_47 {
   class=MaterialExpressionScalarParameter
   properties {defaultValue=1, group=00_FermiumStraw per mask Config, parameterName=SectionD_AspectRatio, sliderMax=4, sliderMin=0.5, sortPriority=45}
  }
  MaterialExpressionScalarParameter_48 {
   class=MaterialExpressionScalarParameter
   properties {defaultValue=1, group=00_FermiumStraw per mask Config, parameterName=SectionD_Zoom, sortPriority=45}
  }
  MaterialExpressionScalarParameter_49 {
   class=MaterialExpressionScalarParameter
   properties {group=00_FermiumStraw per mask Config, parameterName=SectionD_UVOffset_U, sliderMax=1, sliderMin=-1, sortPriority=45}
  }
  MaterialExpressionScalarParameter_5 {
   class=MaterialExpressionScalarParameter
   properties {defaultValue=1, group=FermiumStraw | 01_Appearance, parameterName=LoadingScreenEnabled, sliderMax=1, sortPriority=100}
  }
  MaterialExpressionScalarParameter_50 {
   class=MaterialExpressionScalarParameter
   properties {group=00_FermiumStraw per mask Config, parameterName=SectionD_UVOffset_V, sliderMax=1, sliderMin=-1, sortPriority=45}
  }
  MaterialExpressionScalarParameter_51 {
   class=MaterialExpressionScalarParameter
   properties {defaultValue=1, group=00_FermiumStraw per mask Config, parameterName=SectionB_AspectRatio, sliderMax=4, sliderMin=0.5, sortPriority=25}
  }
  MaterialExpressionScalarParameter_52 {
   class=MaterialExpressionScalarParameter
   properties {defaultValue=1, group=00_FermiumStraw per mask Config, parameterName=SectionB_Zoom, sortPriority=25}
  }
  MaterialExpressionScalarParameter_53 {
   class=MaterialExpressionScalarParameter
   properties {group=00_FermiumStraw per mask Config, parameterName=SectionB_UVOffset_U, sliderMax=1, sliderMin=-1, sortPriority=25}
  }
  MaterialExpressionScalarParameter_54 {
   class=MaterialExpressionScalarParameter
   properties {group=00_FermiumStraw per mask Config, parameterName=SectionB_UVOffset_V, sliderMax=1, sliderMin=-1, sortPriority=25}
  }
  MaterialExpressionScalarParameter_55 {
   class=MaterialExpressionScalarParameter
   properties {group=00_FermiumStraw per mask Config, parameterName=SectionA_UDIMOffset, sortPriority=15}
  }
  MaterialExpressionScalarParameter_56 {
   class=MaterialExpressionScalarParameter
   properties {defaultValue=1, group=00_FermiumStraw per mask Config, parameterName=SectionB_UVChannel, sortPriority=25}
  }
  MaterialExpressionScalarParameter_57 {
   class=MaterialExpressionScalarParameter
   properties {defaultValue=1, group=00_FermiumStraw per mask Config, parameterName=SectionC_UVChannel, sortPriority=35}
  }
  MaterialExpressionScalarParameter_58 {
   class=MaterialExpressionScalarParameter
   properties {defaultValue=1, group=00_FermiumStraw per mask Config, parameterName=SectionD_UVChannel, sortPriority=45}
  }
  MaterialExpressionScalarParameter_59 {class=MaterialExpressionScalarParameter, properties {group=FermiumStraw | 03_DEV, parameterName=DebugFermiumUDIM}}
  MaterialExpressionScalarParameter_6 {
   class=MaterialExpressionScalarParameter
   properties {group=FermiumStraw | 03_Customization, parameterName=Metal_RoughnessBias, sliderMax=1, sliderMin=-1}
  }
  MaterialExpressionScalarParameter_60 {class=MaterialExpressionScalarParameter, properties {defaultValue=1, group=FermiumStraw | 03_DEV, parameterName=DebugUVChannel}}
  MaterialExpressionScalarParameter_61 {
   class=MaterialExpressionScalarParameter
   properties {desc=This parameter does nothing, group=00_FermiumStraw per mask Config, parameterName=------------SectionA------------, sortPriority=9}
  }
  MaterialExpressionScalarParameter_62 {
   class=MaterialExpressionScalarParameter
   properties {desc=This parameter does nothing, group=00_FermiumStraw per mask Config, parameterName=------------SectionB------------, sortPriority=19}
  }
  MaterialExpressionScalarParameter_63 {
   class=MaterialExpressionScalarParameter
   properties {desc=This parameter does nothing, group=00_FermiumStraw per mask Config, parameterName=------------SectionC------------, sortPriority=29}
  }
  MaterialExpressionScalarParameter_64 {
   class=MaterialExpressionScalarParameter
   properties {defaultValue=1, group=00_FermiumStraw per mask Config, parameterName=SectionB_UDIMOffset, sortPriority=25}
  }
  MaterialExpressionScalarParameter_65 {
   class=MaterialExpressionScalarParameter
   properties {defaultValue=2, group=00_FermiumStraw per mask Config, parameterName=SectionC_UDIMOffset, sortPriority=35}
  }
  MaterialExpressionScalarParameter_66 {
   class=MaterialExpressionScalarParameter
   properties {defaultValue=3, group=00_FermiumStraw per mask Config, parameterName=SectionD_UDIMOffset, sortPriority=45}
  }
  MaterialExpressionScalarParameter_67 {
   class=MaterialExpressionScalarParameter
   properties {group=00_FermiumStraw per mask Config, parameterName=SectionA_UDIMOffset, sortPriority=15}
  }
  MaterialExpressionScalarParameter_68 {
   class=MaterialExpressionScalarParameter
   properties {defaultValue=1, group=00_FermiumStraw per mask Config, parameterName=SectionA_UVChannel, sortPriority=15}
  }
  MaterialExpressionScalarParameter_69 {
   class=MaterialExpressionScalarParameter
   properties {defaultValue=1, group=00_FermiumStraw per mask Config, parameterName=SectionB_UDIMOffset, sortPriority=25}
  }
  MaterialExpressionScalarParameter_7 {
   class=MaterialExpressionScalarParameter
   properties {defaultValue=1, group=00_FermiumStraw per mask Config, parameterName=SectionA_Enable, sortPriority=10}
  }
  MaterialExpressionScalarParameter_70 {
   class=MaterialExpressionScalarParameter
   properties {defaultValue=1, group=00_FermiumStraw per mask Config, parameterName=SectionB_UVChannel, sortPriority=25}
  }
  MaterialExpressionScalarParameter_71 {
   class=MaterialExpressionScalarParameter
   properties {defaultValue=3, group=00_FermiumStraw per mask Config, parameterName=SectionD_UDIMOffset, sortPriority=45}
  }
  MaterialExpressionScalarParameter_72 {
   class=MaterialExpressionScalarParameter
   properties {defaultValue=1, group=00_FermiumStraw per mask Config, parameterName=SectionD_UVChannel, sortPriority=45}
  }
  MaterialExpressionScalarParameter_73 {
   class=MaterialExpressionScalarParameter
   properties {defaultValue=2, group=00_FermiumStraw per mask Config, parameterName=SectionC_UDIMOffset, sortPriority=35}
  }
  MaterialExpressionScalarParameter_74 {
   class=MaterialExpressionScalarParameter
   properties {defaultValue=1, group=00_FermiumStraw per mask Config, parameterName=SectionC_UVChannel, sortPriority=35}
  }
  MaterialExpressionScalarParameter_75 {class=MaterialExpressionScalarParameter, properties {group=00_FermiumStraw per mask Config, parameterName=SectionA_Flip, sortPriority=15}}
  MaterialExpressionScalarParameter_76 {class=MaterialExpressionScalarParameter, properties {group=00_FermiumStraw per mask Config, parameterName=SectionC_Flip, sortPriority=35}}
  MaterialExpressionScalarParameter_77 {class=MaterialExpressionScalarParameter, properties {group=00_FermiumStraw per mask Config, parameterName=SectionD_Flip, sortPriority=45}}
  MaterialExpressionScalarParameter_78 {class=MaterialExpressionScalarParameter, properties {group=00_FermiumStraw per mask Config, parameterName=SectionB_Flip, sortPriority=25}}
  MaterialExpressionScalarParameter_79 {
   class=MaterialExpressionScalarParameter
   properties {desc=This parameter does nothing, group=00_FermiumStraw per mask Config, parameterName=------------SectionD------------, sortPriority=39}
  }
  MaterialExpressionScalarParameter_80 {
   class=MaterialExpressionScalarParameter
   properties {
    defaultValue=1
    group=FermiumStraw | 02_Transforms | 02_User Transforms
    parameterName=SectionC_LoadingScreen_UVTransform_EnableUserOffset
    sortPriority=25
   }
  }
  MaterialExpressionScalarParameter_81 {
   class=MaterialExpressionScalarParameter
   properties {
    defaultValue=1
    group=FermiumStraw | 02_Transforms | 02_User Transforms
    parameterName=SectionD_LoadingScreen_UVTransform_EnableUserOffset
    sortPriority=25
   }
  }
  MaterialExpressionScalarParameter_82 {
   class=MaterialExpressionScalarParameter
   properties {group=FermiumStraw | 02_Transforms | 02_User Transforms, parameterName=SectionA_LoadingScreen_UVTransform_UserOffset, sortPriority=15}
  }
  MaterialExpressionScalarParameter_83 {class=MaterialExpressionScalarParameter, properties {defaultValue=1, group=FermiumStraw | 03_DEV, parameterName=DebugUDIMOffset}}
  MaterialExpressionScalarParameter_84 {class=MaterialExpressionScalarParameter, properties {defaultValue=1, group=FermiumStraw | 03_DEV, parameterName=DebugUVChannel}}
  MaterialExpressionScalarParameter_85 {
   class=MaterialExpressionScalarParameter
   properties {group=FermiumStraw | 02_Transforms | 02_User Transforms, parameterName=SectionB_LoadingScreen_UVTransform_UserOffset, sortPriority=25}
  }
  MaterialExpressionScalarParameter_9 {class=MaterialExpressionScalarParameter, properties {group=FermiumStraw | 01_Appearance, parameterName=LoadingScreenClamp, sortPriority=140}}
  MaterialExpressionSetMaterialAttributes_0 {
   class=MaterialExpressionSetMaterialAttributes
   inputs [
    MaterialAttributes {from=MaterialExpressionReroute_13, output=""}
    World Position Offset {from=MaterialExpressionAdd_3, output=""}
   ]
   properties {attributeSetTypes [#4]}
  }
  MaterialExpressionSetMaterialAttributes_1 {
   class=MaterialExpressionSetMaterialAttributes
   inputs [
    MaterialAttributes {from="", output=""}
    Base Color {from=MaterialExpressionLinearInterpolate_15, output=""}
    Normal {from=MaterialExpressionGetMaterialAttributes_5, output=Normal}
   ]
   properties {attributeSetTypes [#1, #2]}
  }
  MaterialExpressionSetMaterialAttributes_10 {
   class=MaterialExpressionSetMaterialAttributes
   inputs [
    MaterialAttributes {from=MaterialExpressionNamedRerouteUsage_74, output=""}
    Base Color {from=MaterialExpressionMultiply_41, output=""}
   ]
   properties {attributeSetTypes [#1]}
  }
  MaterialExpressionSetMaterialAttributes_11 {
   class=MaterialExpressionSetMaterialAttributes
   inputs [
    MaterialAttributes {from=MaterialExpressionNamedRerouteUsage_101, output=""}
    Base Color {from=MaterialExpressionMultiply_53, output=""}
   ]
   properties {attributeSetTypes [#1]}
  }
  MaterialExpressionSetMaterialAttributes_2 {
   class=MaterialExpressionSetMaterialAttributes
   inputs [
    MaterialAttributes {from=MaterialExpressionGetMaterialAttributes_2, output=MaterialAttributes}
    Base Color {from=MaterialExpressionGetMaterialAttributes_2, output=MaterialAttributes}
    Fuzz Color or Subsurface Color {from=MaterialExpressionGetMaterialAttributes_2, output=MaterialAttributes}
   ]
   properties {attributeSetTypes [#1, #5]}
  }
  MaterialExpressionSetMaterialAttributes_3 {
   class=MaterialExpressionSetMaterialAttributes
   inputs [
    MaterialAttributes {from=MaterialExpressionBlendMaterialAttributes_4, output=""}
    Base Color {from=MaterialExpressionMultiply_30, output=""}
   ]
   properties {attributeSetTypes [#1]}
  }
  MaterialExpressionSetMaterialAttributes_4 {
   class=MaterialExpressionSetMaterialAttributes
   inputs [
    MaterialAttributes {from=MaterialExpressionNamedRerouteDeclaration_13, output=""}
    Base Color {from=MaterialExpressionNamedRerouteUsage_53, output=""}
   ]
   properties {attributeSetTypes [#1]}
  }
  MaterialExpressionSetMaterialAttributes_5 {
   class=MaterialExpressionSetMaterialAttributes
   inputs [
    MaterialAttributes {from=MaterialExpressionNamedRerouteUsage_58, output=""}
    Base Color {from=MaterialExpressionNamedRerouteUsage_5, output=""}
   ]
   properties {attributeSetTypes [#1]}
  }
  MaterialExpressionSetMaterialAttributes_6 {
   class=MaterialExpressionSetMaterialAttributes
   inputs [
    MaterialAttributes {from=MaterialExpressionNamedRerouteUsage_117, output=""}
    Base Color {from=MaterialExpressionLinearInterpolate_36, output=""}
    Roughness {from=MaterialExpressionLinearInterpolate_30, output=""}
   ]
   properties {attributeSetTypes [#1, #3]}
  }
  MaterialExpressionSetMaterialAttributes_7 {
   class=MaterialExpressionSetMaterialAttributes
   inputs [
    MaterialAttributes {from=MaterialExpressionGetMaterialAttributes_4, output=MaterialAttributes}
    Base Color {from=MaterialExpressionNamedRerouteUsage_73, output=""}
    Fuzz Color or Subsurface Color {from=MaterialExpressionNamedRerouteUsage_73, output=""}
   ]
   properties {attributeSetTypes [#1, #5]}
  }
  MaterialExpressionSetMaterialAttributes_8 {
   class=MaterialExpressionSetMaterialAttributes
   inputs [
    MaterialAttributes {from="", output=""}
    Base Color {from=MaterialExpressionMultiply_34, output=""}
   ]
   properties {attributeSetTypes [#1]}
  }
  MaterialExpressionSetMaterialAttributes_9 {
   class=MaterialExpressionSetMaterialAttributes
   inputs [
    MaterialAttributes {from=MaterialExpressionNamedRerouteUsage_63, output=""}
    Base Color {from=MaterialExpressionMultiply_38, output=""}
   ]
   properties {attributeSetTypes [#1]}
  }
  MaterialExpressionSign_0 {class=MaterialExpressionSign, inputs [None {from=MaterialExpressionScalarParameter_6, output=""}], properties {}}
  MaterialExpressionSphereMask_0 {
   class=MaterialExpressionSphereMask
   inputs [
    A {from=MaterialExpressionMaterialFunctionCall_36, output=G}
    B {from=MaterialExpressionConstant_6, output=""}
    Radius {from="", output=""}
    Hardness {from="", output=""}
   ]
   properties {attenuationRadius=0.5}
  }
  MaterialExpressionSphereMask_1 {
   class=MaterialExpressionSphereMask
   inputs [
    A {from=MaterialExpressionConstant_6, output=""}
    B {from=MaterialExpressionMaterialFunctionCall_36, output=R}
    Radius {from="", output=""}
    Hardness {from="", output=""}
   ]
   properties {attenuationRadius=0.5}
  }
  MaterialExpressionSphereMask_10 {
   class=MaterialExpressionSphereMask
   inputs [
    A {from=MaterialExpressionConvert_3, output=""}
    B {from=MaterialExpressionConstant_28, output=""}
    Radius {from=MaterialExpressionConstant_29, output=""}
    Hardness {from="", output=""}
   ]
   properties {attenuationRadius=0.499}
  }
  MaterialExpressionSphereMask_11 {
   class=MaterialExpressionSphereMask
   inputs [
    A {from=MaterialExpressionConvert_4, output=""}
    B {from=MaterialExpressionConstant_30, output=""}
    Radius {from=MaterialExpressionConstant_31, output=""}
    Hardness {from="", output=""}
   ]
   properties {attenuationRadius=0.499}
  }
  MaterialExpressionSphereMask_12 {
   class=MaterialExpressionSphereMask
   inputs [
    A {from=MaterialExpressionConvert_4, output=""}
    B {from=MaterialExpressionConstant_30, output=""}
    Radius {from=MaterialExpressionConstant_31, output=""}
    Hardness {from="", output=""}
   ]
   properties {attenuationRadius=0.499}
  }
  MaterialExpressionSphereMask_13 {
   class=MaterialExpressionSphereMask
   inputs [
    A {from=MaterialExpressionConvert_6, output=""}
    B {from=MaterialExpressionConstant_15, output=""}
    Radius {from=MaterialExpressionConstant_16, output=""}
    Hardness {from="", output=""}
   ]
   properties {attenuationRadius=0.499}
  }
  MaterialExpressionSphereMask_2 {
   class=MaterialExpressionSphereMask
   inputs [
    A {from=MaterialExpressionConvert_0, output=""}
    B {from=MaterialExpressionConstant_18, output=""}
    Radius {from=MaterialExpressionConstant_20, output=""}
    Hardness {from="", output=""}
   ]
   properties {attenuationRadius=0.499}
  }
  MaterialExpressionSphereMask_29 {
   class=MaterialExpressionSphereMask
   inputs [
    A {from=MaterialExpressionConstant_11, output=""}
    B {from=MaterialExpressionConvert_13, output=""}
    Radius {from=MaterialExpressionConstant_11, output=""}
    Hardness {from="", output=""}
   ]
   properties {attenuationRadius=0.5}
  }
  MaterialExpressionSphereMask_3 {
   class=MaterialExpressionSphereMask
   inputs [
    A {from=MaterialExpressionConvert_0, output=""}
    B {from=MaterialExpressionConstant_18, output=""}
    Radius {from=MaterialExpressionConstant_20, output=""}
    Hardness {from="", output=""}
   ]
   properties {attenuationRadius=0.499}
  }
  MaterialExpressionSphereMask_30 {
   class=MaterialExpressionSphereMask
   inputs [
    A {from=MaterialExpressionConstant_11, output=""}
    B {from=MaterialExpressionConvert_13, output=""}
    Radius {from=MaterialExpressionConstant_11, output=""}
    Hardness {from="", output=""}
   ]
   properties {
    attenuationRadius=0.5
    desc="2 * 1D mask instead of a single 2D mask\r\n=> keeps sharp angles at corners of UV\r\ntile, otherwise mask will be ovale"
   }
  }
  MaterialExpressionSphereMask_4 {
   class=MaterialExpressionSphereMask
   inputs [
    A {from=MaterialExpressionConvert_6, output=""}
    B {from=MaterialExpressionConstant_15, output=""}
    Radius {from=MaterialExpressionConstant_16, output=""}
    Hardness {from="", output=""}
   ]
   properties {attenuationRadius=0.499}
  }
  MaterialExpressionSphereMask_5 {
   class=MaterialExpressionSphereMask
   inputs [
    A {from=MaterialExpressionConvert_1, output=""}
    B {from=MaterialExpressionConstant_22, output=""}
    Radius {from=MaterialExpressionConstant_23, output=""}
    Hardness {from="", output=""}
   ]
   properties {attenuationRadius=0.499}
  }
  MaterialExpressionSphereMask_6 {
   class=MaterialExpressionSphereMask
   inputs [
    A {from=MaterialExpressionConvert_1, output=""}
    B {from=MaterialExpressionConstant_22, output=""}
    Radius {from=MaterialExpressionConstant_23, output=""}
    Hardness {from="", output=""}
   ]
   properties {attenuationRadius=0.499}
  }
  MaterialExpressionSphereMask_7 {
   class=MaterialExpressionSphereMask
   inputs [
    A {from=MaterialExpressionConvert_2, output=""}
    B {from=MaterialExpressionConstant_25, output=""}
    Radius {from=MaterialExpressionConstant_26, output=""}
    Hardness {from="", output=""}
   ]
   properties {attenuationRadius=0.499}
  }
  MaterialExpressionSphereMask_8 {
   class=MaterialExpressionSphereMask
   inputs [
    A {from=MaterialExpressionConvert_2, output=""}
    B {from=MaterialExpressionConstant_25, output=""}
    Radius {from=MaterialExpressionConstant_26, output=""}
    Hardness {from="", output=""}
   ]
   properties {attenuationRadius=0.499}
  }
  MaterialExpressionSphereMask_9 {
   class=MaterialExpressionSphereMask
   inputs [
    A {from=MaterialExpressionConvert_3, output=""}
    B {from=MaterialExpressionConstant_28, output=""}
    Radius {from=MaterialExpressionConstant_29, output=""}
    Hardness {from="", output=""}
   ]
   properties {attenuationRadius=0.499}
  }
  MaterialExpressionStaticSwitchParameter_19 {
   class=MaterialExpressionStaticSwitchParameter
   inputs [
    True {from=MaterialExpressionSetMaterialAttributes_0, output=""}
    False {from=MaterialExpressionReroute_8, output=""}
   ]
   properties {group=zz_SIGNAL DEBUG, parameterName=Flatten UV, sortPriority=-1500}
  }
  MaterialExpressionStaticSwitchParameter_2 {
   class=MaterialExpressionStaticSwitchParameter
   inputs [
    True {from=MaterialExpressionTextureSampleParameter2D_2, output=RGB}
    False {from=MaterialExpressionTextureSampleParameter2D_1, output=RGB}
   ]
   properties {group=FermiumStraw | 01_Appearance, parameterName=OverrideLoadingScreen, sortPriority=110}
  }
  MaterialExpressionSubtract_0 {
   class=MaterialExpressionSubtract
   inputs [
    A {from=MaterialExpressionMaterialFunctionCall_6, output=UVs}
    B {from=MaterialExpressionNamedRerouteUsage_93}
   ]
  }
  MaterialExpressionSubtract_1 {
   class=MaterialExpressionSubtract
   inputs [
    A {from=MaterialExpressionMaterialFunctionCall_26, output=UVs}
    B {from=MaterialExpressionAppendVector_12}
   ]
  }
  MaterialExpressionSubtract_10 {
   class=MaterialExpressionSubtract
   inputs [
    A {from=MaterialExpressionSwitch_3}
    B {from=MaterialExpressionAppendVector_23}
   ]
  }
  MaterialExpressionSubtract_11 {
   class=MaterialExpressionSubtract
   inputs [
    A {from=MaterialExpressionSwitch_4}
    B {from=MaterialExpressionAppendVector_24}
   ]
  }
  MaterialExpressionSubtract_12 {
   class=MaterialExpressionSubtract
   inputs [
    A {from=MaterialExpressionSwitch_5}
    B {from=MaterialExpressionAppendVector_25}
   ]
  }
  MaterialExpressionSubtract_13 {
   class=MaterialExpressionSubtract
   inputs [
    A {from=MaterialExpressionSwitch_6}
    B {from=MaterialExpressionAppendVector_26}
   ]
  }
  MaterialExpressionSubtract_16 {
   class=MaterialExpressionSubtract
   inputs [
    A {from=MaterialExpressionRotator_2}
    B {from=""}
   ]
   properties {constB=0.5}
  }
  MaterialExpressionSubtract_17 {
   class=MaterialExpressionSubtract
   inputs [
    A {from=MaterialExpressionAdd_1}
    B {from=MaterialExpressionNamedRerouteUsage_91}
   ]
  }
  MaterialExpressionSubtract_2 {
   class=MaterialExpressionSubtract
   inputs [
    A {from=MaterialExpressionMaterialFunctionCall_3, output=UVs}
    B {from=MaterialExpressionAppendVector_18}
   ]
  }
  MaterialExpressionSubtract_4 {
   class=MaterialExpressionSubtract
   inputs [
    A {from=MaterialExpressionMaterialFunctionCall_5, output=UVs}
    B {from=MaterialExpressionAppendVector_20}
   ]
  }
  MaterialExpressionSubtract_5 {
   class=MaterialExpressionSubtract
   inputs [
    A {from=MaterialExpressionMaterialFunctionCall_9, output=UVs}
    B {from=MaterialExpressionAppendVector_22}
   ]
  }
  MaterialExpressionSubtract_6 {
   class=MaterialExpressionSubtract
   inputs [
    A {from=MaterialExpressionSwitch_0}
    B {from=MaterialExpressionAppendVector_0}
   ]
  }
  MaterialExpressionSubtract_7 {
   class=MaterialExpressionSubtract
   inputs [
    A {from=MaterialExpressionSwitch_2}
    B {from=MaterialExpressionAppendVector_1}
   ]
  }
  MaterialExpressionSubtract_8 {
   class=MaterialExpressionSubtract
   inputs [
    A {from=MaterialExpressionNamedRerouteDeclaration_64}
    B {from=MaterialExpressionConvert_7}
   ]
  }
  MaterialExpressionSwitch_0 {
   class=MaterialExpressionSwitch
   inputs [
    SwitchValue {from=MaterialExpressionScalarParameter_60, output=""}
    Default {from="", output=""}
    UV1 {from=MaterialExpressionTextureCoordinate_2, output=""}
    UV2 {from=MaterialExpressionTextureCoordinate_1, output=""}
    UV3 {from=MaterialExpressionTextureCoordinate_3, output=""}
   ]
   properties {inputs [
     {
      inputName=UV1
      input {
       expression=&self:MaterialExpressionTextureCoordinate_2
       inputName=None
       mask=0
       maskA=0
       maskB=0
       maskG=0
       maskR=0
       outputIndex=0
      }
     }
     {
      inputName=UV2
      input {
       expression=&self:MaterialExpressionTextureCoordinate_1
       inputName=None
       mask=0
       maskA=0
       maskB=0
       maskG=0
       maskR=0
       outputIndex=0
      }
     }
     {
      inputName=UV3
      input {
       expression=&self:MaterialExpressionTextureCoordinate_3
       inputName=None
       mask=0
       maskA=0
       maskB=0
       maskG=0
       maskR=0
       outputIndex=0
      }
     }
    ]}
  }
  MaterialExpressionSwitch_1 {
   class=MaterialExpressionSwitch
   inputs [
    SwitchValue {from=MaterialExpressionFloor_0, output=""}
    Default {from="", output=""}
    0 {from=MaterialExpressionConstant3Vector_1, output=""}
    1 {from=MaterialExpressionConstant3Vector_6, output=""}
    2 {from=MaterialExpressionConstant3Vector_7, output=""}
    3 {from=MaterialExpressionConstant3Vector_8, output=""}
    4 {from=MaterialExpressionConstant3Vector_9, output=""}
    5 {from=MaterialExpressionConstant3Vector_10, output=""}
   ]
   properties {inputs [
     {
      inputName=0
      input {
       expression=&self:MaterialExpressionConstant3Vector_1
       inputName=None
       mask=1
       maskA=0
       maskB=1
       maskG=1
       maskR=1
       outputIndex=0
      }
     }
     {
      inputName=1
      input {
       expression=&self:MaterialExpressionConstant3Vector_6
       inputName=None
       mask=1
       maskA=0
       maskB=1
       maskG=1
       maskR=1
       outputIndex=0
      }
     }
     {
      inputName=2
      input {
       expression=&self:MaterialExpressionConstant3Vector_7
       inputName=None
       mask=1
       maskA=0
       maskB=1
       maskG=1
       maskR=1
       outputIndex=0
      }
     }
     {
      inputName=3
      input {
       expression=&self:MaterialExpressionConstant3Vector_8
       inputName=None
       mask=1
       maskA=0
       maskB=1
       maskG=1
       maskR=1
       outputIndex=0
      }
     }
     {
      inputName=4
      input {
       expression=&self:MaterialExpressionConstant3Vector_9
       inputName=None
       mask=1
       maskA=0
       maskB=1
       maskG=1
       maskR=1
       outputIndex=0
      }
     }
     {
      inputName=5
      input {
       expression=&self:MaterialExpressionConstant3Vector_10
       inputName=None
       mask=1
       maskA=0
       maskB=1
       maskG=1
       maskR=1
       outputIndex=0
      }
     }
    ]}
  }
  MaterialExpressionSwitch_2 {
   class=MaterialExpressionSwitch
   inputs [
    SwitchValue {from=MaterialExpressionScalarParameter_68, output=""}
    Default {from="", output=""}
    UV1 {from=MaterialExpressionTextureCoordinate_8, output=""}
    UV2 {from=MaterialExpressionTextureCoordinate_7, output=""}
    UV3 {from=MaterialExpressionTextureCoordinate_9, output=""}
   ]
   properties {inputs [
     {
      inputName=UV1
      input {
       expression=&self:MaterialExpressionTextureCoordinate_8
       inputName=None
       mask=0
       maskA=0
       maskB=0
       maskG=0
       maskR=0
       outputIndex=0
      }
     }
     {
      inputName=UV2
      input {
       expression=&self:MaterialExpressionTextureCoordinate_7
       inputName=None
       mask=0
       maskA=0
       maskB=0
       maskG=0
       maskR=0
       outputIndex=0
      }
     }
     {
      inputName=UV3
      input {
       expression=&self:MaterialExpressionTextureCoordinate_9
       inputName=None
       mask=0
       maskA=0
       maskB=0
       maskG=0
       maskR=0
       outputIndex=0
      }
     }
    ]}
  }
  MaterialExpressionSwitch_3 {
   class=MaterialExpressionSwitch
   inputs [
    SwitchValue {from=MaterialExpressionScalarParameter_70, output=""}
    Default {from="", output=""}
    UV1 {from=MaterialExpressionTextureCoordinate_12, output=""}
    UV2 {from=MaterialExpressionTextureCoordinate_11, output=""}
    UV3 {from=MaterialExpressionTextureCoordinate_13, output=""}
   ]
   properties {inputs [
     {
      inputName=UV1
      input {
       expression=&self:MaterialExpressionTextureCoordinate_12
       inputName=None
       mask=0
       maskA=0
       maskB=0
       maskG=0
       maskR=0
       outputIndex=0
      }
     }
     {
      inputName=UV2
      input {
       expression=&self:MaterialExpressionTextureCoordinate_11
       inputName=None
       mask=0
       maskA=0
       maskB=0
       maskG=0
       maskR=0
       outputIndex=0
      }
     }
     {
      inputName=UV3
      input {
       expression=&self:MaterialExpressionTextureCoordinate_13
       inputName=None
       mask=0
       maskA=0
       maskB=0
       maskG=0
       maskR=0
       outputIndex=0
      }
     }
    ]}
  }
  MaterialExpressionSwitch_4 {
   class=MaterialExpressionSwitch
   inputs [
    SwitchValue {from=MaterialExpressionScalarParameter_72, output=""}
    Default {from="", output=""}
    UV1 {from=MaterialExpressionTextureCoordinate_15, output=""}
    UV2 {from=MaterialExpressionTextureCoordinate_14, output=""}
    UV3 {from=MaterialExpressionTextureCoordinate_16, output=""}
   ]
   properties {inputs [
     {
      inputName=UV1
      input {
       expression=&self:MaterialExpressionTextureCoordinate_15
       inputName=None
       mask=0
       maskA=0
       maskB=0
       maskG=0
       maskR=0
       outputIndex=0
      }
     }
     {
      inputName=UV2
      input {
       expression=&self:MaterialExpressionTextureCoordinate_14
       inputName=None
       mask=0
       maskA=0
       maskB=0
       maskG=0
       maskR=0
       outputIndex=0
      }
     }
     {
      inputName=UV3
      input {
       expression=&self:MaterialExpressionTextureCoordinate_16
       inputName=None
       mask=0
       maskA=0
       maskB=0
       maskG=0
       maskR=0
       outputIndex=0
      }
     }
    ]}
  }
  MaterialExpressionSwitch_5 {
   class=MaterialExpressionSwitch
   inputs [
    SwitchValue {from=MaterialExpressionScalarParameter_74, output=""}
    Default {from="", output=""}
    UV1 {from=MaterialExpressionTextureCoordinate_18, output=""}
    UV2 {from=MaterialExpressionTextureCoordinate_17, output=""}
    UV3 {from=MaterialExpressionTextureCoordinate_19, output=""}
   ]
   properties {inputs [
     {
      inputName=UV1
      input {
       expression=&self:MaterialExpressionTextureCoordinate_18
       inputName=None
       mask=0
       maskA=0
       maskB=0
       maskG=0
       maskR=0
       outputIndex=0
      }
     }
     {
      inputName=UV2
      input {
       expression=&self:MaterialExpressionTextureCoordinate_17
       inputName=None
       mask=0
       maskA=0
       maskB=0
       maskG=0
       maskR=0
       outputIndex=0
      }
     }
     {
      inputName=UV3
      input {
       expression=&self:MaterialExpressionTextureCoordinate_19
       inputName=None
       mask=0
       maskA=0
       maskB=0
       maskG=0
       maskR=0
       outputIndex=0
      }
     }
    ]}
  }
  MaterialExpressionSwitch_6 {
   class=MaterialExpressionSwitch
   inputs [
    SwitchValue {from=MaterialExpressionScalarParameter_84, output=""}
    Default {from="", output=""}
    UV1 {from=MaterialExpressionTextureCoordinate_21, output=""}
    UV2 {from=MaterialExpressionTextureCoordinate_20, output=""}
    UV3 {from=MaterialExpressionTextureCoordinate_22, output=""}
   ]
   properties {inputs [
     {
      inputName=UV1
      input {
       expression=&self:MaterialExpressionTextureCoordinate_21
       inputName=None
       mask=0
       maskA=0
       maskB=0
       maskG=0
       maskR=0
       outputIndex=0
      }
     }
     {
      inputName=UV2
      input {
       expression=&self:MaterialExpressionTextureCoordinate_20
       inputName=None
       mask=0
       maskA=0
       maskB=0
       maskG=0
       maskR=0
       outputIndex=0
      }
     }
     {
      inputName=UV3
      input {
       expression=&self:MaterialExpressionTextureCoordinate_22
       inputName=None
       mask=0
       maskA=0
       maskB=0
       maskG=0
       maskR=0
       outputIndex=0
      }
     }
    ]}
  }
  MaterialExpressionTextureCoordinate_0 {class=MaterialExpressionTextureCoordinate}
  MaterialExpressionTextureCoordinate_1 {class=MaterialExpressionTextureCoordinate, properties {coordinateIndex=1}}
  MaterialExpressionTextureCoordinate_11 {class=MaterialExpressionTextureCoordinate, properties {coordinateIndex=1}}
  MaterialExpressionTextureCoordinate_12 {class=MaterialExpressionTextureCoordinate}
  MaterialExpressionTextureCoordinate_13 {class=MaterialExpressionTextureCoordinate, properties {coordinateIndex=2}}
  MaterialExpressionTextureCoordinate_14 {class=MaterialExpressionTextureCoordinate, properties {coordinateIndex=1}}
  MaterialExpressionTextureCoordinate_15 {class=MaterialExpressionTextureCoordinate}
  MaterialExpressionTextureCoordinate_16 {class=MaterialExpressionTextureCoordinate, properties {coordinateIndex=2}}
  MaterialExpressionTextureCoordinate_17 {class=MaterialExpressionTextureCoordinate, properties {coordinateIndex=1}}
  MaterialExpressionTextureCoordinate_18 {class=MaterialExpressionTextureCoordinate}
  MaterialExpressionTextureCoordinate_19 {class=MaterialExpressionTextureCoordinate, properties {coordinateIndex=2}}
  MaterialExpressionTextureCoordinate_2 {class=MaterialExpressionTextureCoordinate}
  MaterialExpressionTextureCoordinate_20 {class=MaterialExpressionTextureCoordinate, properties {coordinateIndex=1}}
  MaterialExpressionTextureCoordinate_21 {class=MaterialExpressionTextureCoordinate}
  MaterialExpressionTextureCoordinate_22 {class=MaterialExpressionTextureCoordinate, properties {coordinateIndex=2}}
  MaterialExpressionTextureCoordinate_3 {class=MaterialExpressionTextureCoordinate, properties {coordinateIndex=2}}
  MaterialExpressionTextureCoordinate_6 {class=MaterialExpressionTextureCoordinate}
  MaterialExpressionTextureCoordinate_7 {class=MaterialExpressionTextureCoordinate, properties {coordinateIndex=1}}
  MaterialExpressionTextureCoordinate_8 {class=MaterialExpressionTextureCoordinate}
  MaterialExpressionTextureCoordinate_9 {class=MaterialExpressionTextureCoordinate, properties {coordinateIndex=2}}
  MaterialExpressionTextureSampleParameter2D_1 {
   class=MaterialExpressionTextureSampleParameter2D
   inputs [
    UVs {from=MaterialExpressionNamedRerouteUsage_18, output=""}
    Apply View MipBias {from="", output=""}
   ]
   properties {
    group=FermiumStraw | 01_Appearance
    parameterName=LoadingScreen
    samplerSource=SSM_Wrap_WorldGroupSettings
    sortPriority=120
    texture=/BRCosmetics/2dAssets/Loadingscreens/Season41/T_LS_MKART_loadingscreen_s41LineUp
   }
  }
  MaterialExpressionTextureSampleParameter2D_10 {
   class=MaterialExpressionTextureSampleParameter2D
   inputs [
    UVs {from=MaterialExpressionSubtract_13, output=""}
    Apply View MipBias {from="", output=""}
   ]
   properties {
    group=FermiumStraw | 01_Appearance
    parameterName=LoadingScreenOverride
    samplerSource=SSM_Wrap_WorldGroupSettings
    sortPriority=130
    texture=&T_LS_MKART_loadingscreen_s41LineUp
   }
  }
  MaterialExpressionTextureSampleParameter2D_11 {
   class=MaterialExpressionTextureSampleParameter2D
   inputs [
    UVs {from=MaterialExpressionNamedRerouteUsage_70, output=""}
    Apply View MipBias {from="", output=""}
   ]
   properties {
    group=FermiumStraw | 01_Appearance
    parameterName=LoadingScreenOverride
    samplerSource=SSM_Wrap_WorldGroupSettings
    sortPriority=130
    texture=&T_LS_MKART_loadingscreen_s41LineUp
   }
  }
  MaterialExpressionTextureSampleParameter2D_2 {
   class=MaterialExpressionTextureSampleParameter2D
   inputs [
    UVs {from=MaterialExpressionNamedRerouteUsage_89, output=""}
    Apply View MipBias {from="", output=""}
   ]
   properties {
    group=FermiumStraw | 01_Appearance
    parameterName=LoadingScreenOverride
    samplerSource=SSM_Wrap_WorldGroupSettings
    sortPriority=130
    texture=&T_LS_MKART_loadingscreen_s41LineUp
   }
  }
  MaterialExpressionTextureSampleParameter2D_5 {
   class=MaterialExpressionTextureSampleParameter2D
   inputs [
    UVs {from="", output=""}
    Apply View MipBias {from="", output=""}
   ]
   properties {
    group=FermiumStraw | 03_DEV
    parameterName=StyleMask1
    texture=/BRCosmetics/Characters/Player/Female/Medium/Bodies/F_MED_FermiumStraw/Textures/T_FermiumStraw_Torso_StyleB_M1
   }
  }
  MaterialExpressionTextureSampleParameter2D_6 {
   class=MaterialExpressionTextureSampleParameter2D
   inputs [
    UVs {from="", output=""}
    Apply View MipBias {from="", output=""}
   ]
   properties {
    group=FermiumStraw | 03_DEV
    parameterName=StyleMask2
    texture=/BRCosmetics/Characters/Player/Female/Medium/Bodies/F_MED_FermiumStraw/Textures/T_FermiumStraw_Torso_StyleB_M2
   }
  }
  MaterialExpressionTextureSampleParameter2D_8 {
   class=MaterialExpressionTextureSampleParameter2D
   inputs [
    UVs {from=MaterialExpressionMultiply_29, output=""}
    Apply View MipBias {from="", output=""}
   ]
   properties {
    group=FermiumStraw | 01_Appearance
    parameterName=SplatterTexture
    samplerSource=SSM_Wrap_WorldGroupSettings
    samplerType=SAMPLERTYPE_LinearColor
    texture=/BRCosmetics/Characters/Player/Female/Medium/Bodies/F_MED_FermiumStraw/Textures/T_FermiumStraw_Paintsplatter_D
   }
  }
  MaterialExpressionTextureSampleParameter2D_9 {
   class=MaterialExpressionTextureSampleParameter2D
   inputs [
    UVs {from=MaterialExpressionAdd_0, output=""}
    Apply View MipBias {from="", output=""}
   ]
   properties {
    group=FermiumStraw | 03_Customization
    parameterName=Metal_ThinFilm
    texture=/Game/Global/Textures/Default/Thinfilm/T_F_MED_Global_FB_ThinFilm_color
   }
  }
  MaterialExpressionVectorParameter_0 {
   class=MaterialExpressionVectorParameter
   properties {
    group=00_FermiumStraw per mask Config
    parameterName=SectionA_UVBounds
    sortPriority=15
    channelNames {a=V Max, b=U Max, g=V Min, r=U Min}
    defaultValue=<rgba 0 0 1 1>
   }
  }
  MaterialExpressionVectorParameter_1 {
   class=MaterialExpressionVectorParameter
   properties {
    group=00_FermiumStraw per mask Config
    parameterName=SectionB_UVBounds
    sortPriority=25
    channelNames {a=V Max, b=U Max, g=V Min, r=U Min}
    defaultValue=<rgba 0 0 1 1>
   }
  }
  MaterialExpressionVectorParameter_10 {
   class=MaterialExpressionVectorParameter
   properties {
    group=FermiumStraw | 02_Transforms | 01_Base Transforms
    parameterName=SectionC_LoadingScreen_UVTransform
    sortPriority=30
    defaultValue=<rgba 0 0 0 1>
   }
  }
  MaterialExpressionVectorParameter_11 {
   class=MaterialExpressionVectorParameter
   properties {
    group=FermiumStraw | 02_Transforms | 01_Base Transforms
    parameterName=SectionA_LoadingScreen_UVTransform
    sortPriority=15
    defaultValue=<rgba 0 0 0 1>
   }
  }
  MaterialExpressionVectorParameter_12 {
   class=MaterialExpressionVectorParameter
   properties {
    group=FermiumStraw | 02_Transforms | 01_Base Transforms
    parameterName=SectionB_LoadingScreen_UVTransform
    sortPriority=25
    defaultValue=<rgba 0 0 0 1>
   }
  }
  MaterialExpressionVectorParameter_13 {
   class=MaterialExpressionVectorParameter
   properties {group=FermiumStraw | 03_DEV, parameterName=PrimaryColor_Debug, defaultValue=<rgba 0 1 0 1>}
  }
  MaterialExpressionVectorParameter_14 {
   class=MaterialExpressionVectorParameter
   properties {group=FermiumStraw | 03_DEV, parameterName=MutedColor_Debug, defaultValue=<rgba 1 0.0729 0.0729 1>}
  }
  MaterialExpressionVectorParameter_15 {
   class=MaterialExpressionVectorParameter
   properties {group=FermiumStraw | 03_DEV, parameterName=NeutralColor_Debug, defaultValue=<rgba 0.05 0.1292 1 1>}
  }
  MaterialExpressionVectorParameter_16 {
   class=MaterialExpressionVectorParameter
   properties {group=FermiumStraw | 03_DEV, parameterName=MetalColor_Debug, defaultValue=<rgba 1 1 0 1>}
  }
  MaterialExpressionVectorParameter_17 {
   class=MaterialExpressionVectorParameter
   properties {group=FermiumStraw | 03_Customization, parameterName=SplatterColor, defaultValue=<rgba 0 1 1 1>}
  }
  MaterialExpressionVectorParameter_18 {
   class=MaterialExpressionVectorParameter
   properties {group=FermiumStraw | 03_Customization, parameterName=Metal_Color, defaultValue=<rgba 1 1 0 1>}
  }
  MaterialExpressionVectorParameter_19 {
   class=MaterialExpressionVectorParameter
   properties {
    group=FermiumStraw | 02_Transforms | 01_Base Transforms
    parameterName=SectionD_LoadingScreen_UVTransform
    sortPriority=15
    defaultValue=<rgba 0 0 0 1>
   }
  }
  MaterialExpressionVectorParameter_2 {
   class=MaterialExpressionVectorParameter
   properties {group=FermiumStraw | 03_Customization, parameterName=PrimaryColor, defaultValue=<rgba 0 1 0 1>}
  }
  MaterialExpressionVectorParameter_23 {
   class=MaterialExpressionVectorParameter
   properties {group=FermiumStraw | 03_Customization, parameterName=MutedColor, defaultValue=<rgba 1 0.0729 0.0729 1>}
  }
  MaterialExpressionVectorParameter_25 {
   class=MaterialExpressionVectorParameter
   properties {group=FermiumStraw | 03_Customization, parameterName=NeutralColor, defaultValue=<rgba 0.05 0.1292 1 1>}
  }
  MaterialExpressionVectorParameter_4 {
   class=MaterialExpressionVectorParameter
   properties {group=FermiumStraw | 03_DEV, parameterName=SplatterColor_Debug, defaultValue=<rgba 0 1 1 1>}
  }
  MaterialExpressionVectorParameter_5 {
   class=MaterialExpressionVectorParameter
   properties {
    group=00_FermiumStraw per mask Config
    parameterName=SectionC_UVBounds
    sortPriority=35
    channelNames {a=V Max, b=U Max, g=V Min, r=U Min}
    defaultValue=<rgba 0 0 1 1>
   }
  }
  MaterialExpressionVectorParameter_6 {
   class=MaterialExpressionVectorParameter
   properties {group=FermiumStraw | 03_DEV, parameterName=LoadingScreenDebugColor, defaultValue=<rgba 1 0 0.8446 1>}
  }
  MaterialExpressionVectorParameter_8 {
   class=MaterialExpressionVectorParameter
   properties {
    group=00_FermiumStraw per mask Config
    parameterName=SectionD_UVBounds
    sortPriority=45
    channelNames {a=V Max, b=U Max, g=V Min, r=U Min}
    defaultValue=<rgba 0 0 1 1>
   }
  }
  MaterialExpressionVectorParameter_9 {
   class=MaterialExpressionVectorParameter
   properties {group=FermiumStraw | 01_Appearance, parameterName=BackgroundColor, sortPriority=20, defaultValue=<rgba 1 1 1 1>}
  }
 }
 properties {
  bUseMaterialAttributes=true
  bUsedWithMorphTargets=true
  bUsedWithSkeletalMesh=true
  bUsedWithStaticMesh=false
  pixelDepthOffsetMode=PDOM_Legacy
  shadingModel=MSM_FromMaterialExpression
  usedShadingModels=Default Lit | Subsurface | Clear Coat | Subsurface Profile | Cloth
 }
}
