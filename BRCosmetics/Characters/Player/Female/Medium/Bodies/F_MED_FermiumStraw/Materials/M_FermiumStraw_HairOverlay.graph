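# Masked character material for the hair overlay on the FermiumStraw female medium body, built on the 2023 Character Parent functions.
# PreFX output passes through the mess hair overlay function, then its base color is multiplied by the MutedColor customization parameter.
# The green channel of the FX1 Map mask blends the tinted attributes over the untinted ones before they feed PostFX.
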Material /BRCosmetics/Characters/Player/Female/Medium/Bodies/F_MED_FermiumStraw/Materials/M_FermiumStraw_HairOverlay
graph {
 properties {
  bUseMaterialAttributes=true
  bUsedWithClothing=true
  bUsedWithInstancedStaticMeshes=true
  bUsedWithMeshDeformer=true
  bUsedWithMorphTargets=true
  bUsedWithNanite=true
  bUsedWithSkeletalMesh=true
  blendMode=BLEND_Masked
  pixelDepthOffsetMode=PDOM_Legacy
  shadingModel=MSM_FromMaterialExpression
  usedShadingModels=Default Lit | Subsurface | Clear Coat | Subsurface Profile | Cloth
 }
 nodes [
  n1 = MaterialFunctionCall() {materialFunction=&MF_CharacterParent_2023_PreFX}
  n2 = MaterialFunctionCall!(MA=@n1.Result, PreSkinLOC=@n1.Result) {materialFunction=/Game/Global/Materials/Character_Parent_2023/Hair_Parent_2023/MaterialFunctions/MF_MessHairOverlay_2023}
  n3 = GetMaterialAttributes(None=@n2.Result) {attributeGetTypes [#1]}
  n4 = Multiply(A=@n3.Base Color, B=VectorParameter() {group=FermiumStraw | 03_Customization, parameterName=MutedColor, defaultValue=<rgba 1 0 0 1>}.RGB)
  MaterialFunctionCall(
   Material=BlendMaterialAttributes(
    A=@n2.Result
    B=SetMaterialAttributes!(MaterialAttributes=@n3.MaterialAttributes, Base Color=@n4) {attributeSetTypes [#1]}
    Alpha=TextureSampleParameter2D() {parameterName=FX1 Map, samplerType=SAMPLERTYPE_Masks, texture=/Game/Athena/Aircraft/PartyBus_RT/Textures/AllBlackMasks}.G
   )
   M=@n1.M
   Pre Skinned Local Position=@n1.M
   Distance Blend=@n1.M
   Fresnel=@n1.M
  ) {materialFunction=&MF_CharacterParent_2023_PostFX}
  Reroute(None=@n1.Tech Art Mask)
 ]
}
