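# Post-process-style animation Blueprint for the FermiumStraw female medium body's legs mesh, adding secondary motion to three zippers and a leg pack.
# It copies the attached parent's pose, runs the legs Control Rig, then feeds the cached result into two parallel AnimDynamics spring chains.
# The chains differ in sim-space settings: FE variables use partial angular alpha, the others zero.
# A bool blend picks one chain for output.
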AnimBlueprint /BRCosmetics/Characters/Player/Female/Medium/Bodies/F_MED_FermiumStraw/Meshes/F_MED_FermiumStraw_Legs_AnimBP
target_skeleton=&/F_MED_FermiumStraw_Skeleton
graphs [
 AnimGraph {nodes [
   Output Pose(Result=Use cached pose 'AnimDyn_Cache'()) {layerGroup=None}
   n1 = Control Rig(Source=Copy Pose From Mesh() {bUseAttachedParent=true}) {controlRigAssetReference {controlRigAsset=&/F_MED_FermiumStraw_Legs_CtrlRig}}
   Save cached pose 'CR_Cache'(Pose=@n1) {cacheName=CR_Cache}
   n2 = Anim Dynamics - Bone: dyn_zipper_front_lf!(
    ComponentPose=Local To Component(LocalPose=Use cached pose 'CR_Cache'())
    Alpha=1
    GravityOverride:Vector="0,0,-980"
    SimSpaceSettings:Anim Phys Sim Space Settings Structure=Get FE Zipper Sim Space Settings() -> (FE Zipper Sim Space Settings:Anim Phys Sim Space Settings Structure)
   ) -> (Pose) {
    angularDampingOverride=1
    angularSpringConstant=9000
    bAngularSpring=true
    bOverrideAngularDamping=true
    gravityScale=0.35
    simulationSpace=RootRelative
    boundBone {boneName=dyn_zipper_front_lf}
    componentAppliedLinearAccClamp=<3000 3000 3000>
    componentLinearAccScale=<0.4 0.4 0.4>
    componentLinearVelScale=<0.4 0.4 0.4>
    physicsBodyDefinitions [{
      boundBone {boneName=dyn_zipper_front_lf}
      boxExtents=<8 8 20>
      constraintSetup {angularLimitsMax=<0 20 70>, angularLimitsMin=<0 -20 0>, angularTarget=<1 0 0>}
      localJointOffset=<3 0 0>
     }]
    simSpaceSettings {maxAngularAcceleration=20, maxAngularVelocity=10, simSpaceAngularAlpha=1}
   }
   n3 = Anim Dynamics - Bone: dyn_zipper_front_rt!(
    ComponentPose=@n2
    Alpha=1
    GravityOverride:Vector="0,0,-980"
    SimSpaceSettings:Anim Phys Sim Space Settings Structure=Get FE Zipper Sim Space Settings()
   ) -> (Pose) {
    angularDampingOverride=1
    angularSpringConstant=9000
    bAngularSpring=true
    bOverrideAngularDamping=true
    gravityScale=0.35
    simulationSpace=RootRelative
    boundBone {boneName=dyn_zipper_front_rt}
    componentAppliedLinearAccClamp=<3000 3000 3000>
    componentLinearAccScale=<0.4 0.4 0.4>
    componentLinearVelScale=<0.4 0.4 0.4>
    physicsBodyDefinitions [{
      boundBone {boneName=dyn_zipper_front_rt}
      boxExtents=<8 8 20>
      constraintSetup {angularLimitsMax=<0 20 70>, angularLimitsMin=<0 -20 0>, angularTarget=<1 0 0>}
      localJointOffset=<3 0 0>
     }]
    simSpaceSettings {maxAngularAcceleration=20, maxAngularVelocity=10, simSpaceAngularAlpha=1}
   }
   n4 = Anim Dynamics - Bone: dyn_zipper_back_rt!(
    ComponentPose=@n3
    Alpha=1
    GravityOverride:Vector="0,0,-980"
    SimSpaceSettings:Anim Phys Sim Space Settings Structure=Get FE Zipper Sim Space Settings()
   ) -> (Pose) {
    angularDampingOverride=1
    angularSpringConstant=9000
    bAngularSpring=true
    bOverrideAngularDamping=true
    gravityScale=0.35
    simulationSpace=RootRelative
    boundBone {boneName=dyn_zipper_back_rt}
    componentAppliedLinearAccClamp=<3000 3000 3000>
    componentLinearAccScale=<0.4 0.4 0.4>
    componentLinearVelScale=<0.4 0.4 0.4>
    physicsBodyDefinitions [{
      boundBone {boneName=dyn_zipper_back_rt}
      boxExtents=<8 8 20>
      constraintSetup {angularLimitsMax=<0 20 70>, angularLimitsMin=<0 -20 0>, angularTarget=<1 0 0>}
      localJointOffset=<3 0 0>
     }]
    simSpaceSettings {maxAngularAcceleration=20, maxAngularVelocity=10, simSpaceAngularAlpha=1}
   }
   n5 = Anim Dynamics - Bone: dyn_leg_pack!(
    ComponentPose=@n4
    Alpha=1
    SimSpaceSettings:Anim Phys Sim Space Settings Structure=Get FE Pack Sim Space Settings() -> (FE Pack Sim Space Settings:Anim Phys Sim Space Settings Structure)
    ExternalForce="0,0,0"
   ) -> (Pose) {
    angularDampingOverride=0.99
    angularSpringConstant=10000
    bAngularSpring=true
    bOverrideAngularDamping=true
    gravityScale=0
    simulationSpace=RootRelative
    boundBone {boneName=dyn_leg_pack}
    componentAppliedLinearAccClamp=<1400 1400 1400>
    componentLinearAccScale=<1 1 1>
    componentLinearVelScale=<1 1 1>
    physicsBodyDefinitions [{
      boundBone {boneName=dyn_leg_pack}
      boxExtents=<20 30 30>
      constraintSetup {angularLimitsMax=<0 0 60>, angularLimitsMin=<0 0 -10>, angularTarget=<1 0 0>}
      localJointOffset=<3 2 -2>
     }]
   }
   n6 = Anim Dynamics - Bone: dyn_zipper_front_lf(
    ComponentPose=Local To Component(LocalPose=Use cached pose 'CR_Cache'())
    Alpha=1
    GravityOverride="0,0,-980"
    SimSpaceSettings=Get Zipper Sim Space Settings() -> (Zipper Sim Space Settings:Anim Phys Sim Space Settings Structure)
   ) {
    angularDampingOverride=1
    angularSpringConstant=9000
    bAngularSpring=true
    bOverrideAngularDamping=true
    gravityScale=0.35
    simulationSpace=RootRelative
    boundBone {boneName=dyn_zipper_front_lf}
    componentAppliedLinearAccClamp=<3000 3000 3000>
    componentLinearAccScale=<0.4 0.4 0.4>
    componentLinearVelScale=<0.4 0.4 0.4>
    physicsBodyDefinitions [{
      boundBone {boneName=dyn_zipper_front_lf}
      boxExtents=<8 8 20>
      constraintSetup {angularLimitsMax=<0 20 70>, angularLimitsMin=<0 -20 0>, angularTarget=<1 0 0>}
      localJointOffset=<3 0 0>
     }]
    simSpaceSettings {maxAngularAcceleration=20, maxAngularVelocity=10, simSpaceAngularAlpha=1}
   }
   n7 = Anim Dynamics - Bone: dyn_zipper_front_rt(ComponentPose=@n6, Alpha=1, GravityOverride="0,0,-980", SimSpaceSettings=Get Zipper Sim Space Settings()) {
    angularDampingOverride=1
    angularSpringConstant=9000
    bAngularSpring=true
    bOverrideAngularDamping=true
    gravityScale=0.35
    simulationSpace=RootRelative
    boundBone {boneName=dyn_zipper_front_rt}
    componentAppliedLinearAccClamp=<3000 3000 3000>
    componentLinearAccScale=<0.4 0.4 0.4>
    componentLinearVelScale=<0.4 0.4 0.4>
    physicsBodyDefinitions [{
      boundBone {boneName=dyn_zipper_front_rt}
      boxExtents=<8 8 20>
      constraintSetup {angularLimitsMax=<0 20 70>, angularLimitsMin=<0 -20 0>, angularTarget=<1 0 0>}
      localJointOffset=<3 0 0>
     }]
    simSpaceSettings {maxAngularAcceleration=20, maxAngularVelocity=10, simSpaceAngularAlpha=1}
   }
   n8 = Anim Dynamics - Bone: dyn_zipper_back_rt(ComponentPose=@n7, Alpha=1, GravityOverride="0,0,-980", SimSpaceSettings=Get Zipper Sim Space Settings()) {
    angularDampingOverride=1
    angularSpringConstant=9000
    bAngularSpring=true
    bOverrideAngularDamping=true
    gravityScale=0.35
    simulationSpace=RootRelative
    boundBone {boneName=dyn_zipper_back_rt}
    componentAppliedLinearAccClamp=<3000 3000 3000>
    componentLinearAccScale=<0.4 0.4 0.4>
    componentLinearVelScale=<0.4 0.4 0.4>
    physicsBodyDefinitions [{
      boundBone {boneName=dyn_zipper_back_rt}
      boxExtents=<8 8 20>
      constraintSetup {angularLimitsMax=<0 20 70>, angularLimitsMin=<0 -20 0>, angularTarget=<1 0 0>}
      localJointOffset=<3 0 0>
     }]
    simSpaceSettings {maxAngularAcceleration=20, maxAngularVelocity=10, simSpaceAngularAlpha=1}
   }
   n9 = Anim Dynamics - Bone: dyn_leg_pack(
    ComponentPose=@n8
    Alpha=1
    SimSpaceSettings=Get Pack Sim Space Settings() -> (Pack Sim Space Settings:Anim Phys Sim Space Settings Structure)
    ExternalForce="0,0,0"
   ) {
    angularDampingOverride=0.99
    angularSpringConstant=10000
    bAngularSpring=true
    bOverrideAngularDamping=true
    gravityScale=0
    simulationSpace=RootRelative
    boundBone {boneName=dyn_leg_pack}
    componentAppliedLinearAccClamp=<1400 1400 1400>
    componentLinearAccScale=<1 1 1>
    componentLinearVelScale=<1 1 1>
    physicsBodyDefinitions [{
      boundBone {boneName=dyn_leg_pack}
      boxExtents=<20 30 30>
      constraintSetup {angularLimitsMax=<0 0 60>, angularLimitsMin=<0 0 -10>, angularTarget=<1 0 0>}
      localJointOffset=<3 2 -2>
     }]
   }
   Save cached pose 'AnimDyn_Cache'(Pose=Blend Poses by bool(BlendPose_0=Component To Local(ComponentPose=@n5), BlendPose_1=Component To Local(ComponentPose=@n9))) {cacheName=AnimDyn_Cache}
  ]}
 EventGraph {nodes [Event BlueprintUpdateAnimation()]}
]
variables [
 FE Pack Sim Space Settings {
  type=AnimPhysSimSpaceSettings
  default {maxAngularAcceleration=10000, maxAngularVelocity=10000, simSpaceAngularAlpha=0.5, externalAngularVelocity=<0 0 0>}
 }
 FE Zipper Sim Space Settings {
  type=AnimPhysSimSpaceSettings
  default {maxAngularAcceleration=20, maxAngularVelocity=10, simSpaceAngularAlpha=0.5, externalAngularVelocity=<0 0 0>}
 }
 Pack Sim Space Settings {
  type=AnimPhysSimSpaceSettings
  default {maxAngularAcceleration=10000, maxAngularVelocity=10000, simSpaceAngularAlpha=0, externalAngularVelocity=<0 0 0>}
 }
 Zipper Sim Space Settings {
  type=AnimPhysSimSpaceSettings
  default {maxAngularAcceleration=20, maxAngularVelocity=10, simSpaceAngularAlpha=0, externalAngularVelocity=<0 0 0>}
 }
]
