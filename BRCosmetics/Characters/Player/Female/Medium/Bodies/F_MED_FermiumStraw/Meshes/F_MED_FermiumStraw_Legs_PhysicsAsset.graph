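# Physics asset for the legs mesh of the F_MED_FermiumStraw female medium player body cosmetic.
# It gives the pelvis and the left and right thigh, calf and foot bones one capsule body each.
# Constraints chain each foot to its calf, calf to thigh and thigh to the pelvis, all with the same swing and twist limit angles.
# The right thigh capsule is noticeably wider and differently angled than the left.
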PhysicsAsset /BRCosmetics/Characters/Player/Female/Medium/Bodies/F_MED_FermiumStraw/Meshes/F_MED_FermiumStraw_Legs_PhysicsAsset
bodies {
 calf_l [{shapeType=Capsule, length=24.167, radius=7.8269, center=<-21.9665 -1.5447 -0.2703>, rotation=<rot 0 89.2207 -88.1534>}]
 calf_r [{shapeType=Capsule, length=24.1005, radius=7.6179, center=<22.0118 1.3868 0.3322>, rotation=<rot 0 89.3044 -86.7946>}]
 foot_l [{shapeType=Capsule, length=14.0049, radius=9.8446, center=<-3.1163 6.4696 -0.0666>, rotation=<rot 0 27.9546 -86.9078>}]
 foot_r [{shapeType=Capsule, length=14.0046, radius=9.8406, center=<3.1032 -6.4684 0.2803>, rotation=<rot 0 27.8377 -86.891>}]
 pelvis [{shapeType=Capsule, length=14.7103, radius=14.6839, center=<-1.4016 -3.0828 -0.1955>, rotation=<rot 0 -24.8308 -138.2775>}]
 thigh_l [{shapeType=Capsule, length=23.8321, radius=10.6378, center=<-21.4752 -1.702 1.3714>, rotation=<rot 0 -86.3374 -89.9886>}]
 thigh_r [{shapeType=Capsule, length=23.623, radius=15.254, center=<20.8173 1.6554 4.1037>, rotation=<rot 20.8381 -90 -102.9107>}]
}
constraints [
 {bone1Name=thigh_l, bone2Name=pelvis, swing1LimitDegrees=45, swing2LimitDegrees=45, twistLimitDegrees=45}
 {bone1Name=calf_l, bone2Name=thigh_l, swing1LimitDegrees=45, swing2LimitDegrees=45, twistLimitDegrees=45}
 {bone1Name=foot_l, bone2Name=calf_l, swing1LimitDegrees=45, swing2LimitDegrees=45, twistLimitDegrees=45}
 {bone1Name=thigh_r, bone2Name=pelvis, swing1LimitDegrees=45, swing2LimitDegrees=45, twistLimitDegrees=45}
 {bone1Name=calf_r, bone2Name=thigh_r, swing1LimitDegrees=45, swing2LimitDegrees=45, twistLimitDegrees=45}
 {bone1Name=foot_r, bone2Name=calf_r, swing1LimitDegrees=45, swing2LimitDegrees=45, twistLimitDegrees=45}
]
