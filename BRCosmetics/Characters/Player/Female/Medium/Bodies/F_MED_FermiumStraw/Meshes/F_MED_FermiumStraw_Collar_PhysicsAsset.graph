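# Physics asset for the collar mesh of the FermiumStraw female medium body cosmetic, defining collision and joints for swinging collar and lapel bones.
# Left and right dyn collar and lapel capsules are constrained to small spheres on their base bones.
# Each hinges mainly on swing1, tight for collars and wide for lapels, with other angular axes locked or zero-limit.
# Head, neck, upper spine and upper-arm capsules supply body collision shapes.
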PhysicsAsset /BRCosmetics/Characters/Player/Female/Medium/Bodies/F_MED_FermiumStraw/Meshes/F_MED_FermiumStraw_Collar_PhysicsAsset
bodies {
 base_collar_lf [{shapeType=Sphere, radius=1}]
 base_collar_rt [{shapeType=Sphere, radius=1}]
 base_lapel_lf [{shapeType=Sphere, radius=0.505, shapeName=base_lapel_lf_sphere}]
 base_lapel_rt [{shapeType=Sphere, radius=0.505, shapeName=base_lapel_rt_sphere}]
 dyn_collar_lf [
  {shapeType=Capsule, length=6, radius=1.5, shapeName=dyn_collar_lf_capsule, center=<4.5782 0.5325 0>, rotation=<rot 89.2455 -93.1053 -90.0055>}
  {shapeType=Capsule, length=5.9089, radius=1, center=<3.7331 0.7377 3.5193>, rotation=<rot 44.2473 3.2602 4.6675>}
 ]
 dyn_collar_rt [
  {
   shapeType=Capsule
   length=6
   radius=1.5
   shapeName=dyn_collar_rt_capsule
   center=<5.8997 1.0576 0.0078>
   rotation=<rot 89.246 -93.0711 -86.309>
  }
  {shapeType=Capsule, length=3.9199, radius=1, center=<4.7084 0.6391 -3.2639>, rotation=<rot -44.3279 21.698 -17.2285>}
 ]
 dyn_lapel_lf [{shapeType=Capsule, length=5.7185, radius=0.9547, shapeName=dyn_lapel_lf_capsule, center=<1.2803 0 0>}]
 dyn_lapel_rt [{shapeType=Capsule, length=5.7185, radius=0.9547, shapeName=dyn_lapel_rt_capsule, center=<1.2803 0 0>}]
 head [{shapeType=Capsule, length=2, radius=13.5, center=<6.4812 3.2033 -0.125>, rotation=<rot 0 -90 0>}]
 neck_01 [{shapeType=Capsule, length=3.1273, radius=5.2401, shapeName=neck_01_capsule, center=<2.4331 0.5028 0>, rotation=<rot 0 -90.0001 90>}]
 spine_05 [{shapeType=Capsule, length=16.7751, radius=9.8029, shapeName=spine_05_capsule, center=<-0.9865 -1.7946 0>}]
 upperarm_l [{
   shapeType=Capsule
   length=15.9384
   radius=5.7879
   shapeName=upperarm_r_capsule
   center=<9.2367 -0.8435 -0.3286>
   rotation=<rot 86.9747 177.4972 0.0134>
  }]
 upperarm_r [{
   shapeType=Capsule
   length=15.9384
   radius=5.7879
   shapeName=upperarm_l_capsule
   center=<-9.2367 0.8435 0.3286>
   rotation=<rot 86.9747 177.4972 -179.9866>
  }]
}
constraints [
 {bone1Name=dyn_collar_lf, bone2Name=base_collar_lf, swing1LimitDegrees=20, swing2LimitDegrees=0, twistLimitDegrees=0}
 {bone1Name=dyn_collar_rt, bone2Name=base_collar_rt, swing1LimitDegrees=20, swing2LimitDegrees=0, twistLimitDegrees=0}
 {
  bone1Name=dyn_lapel_lf
  bone2Name=base_lapel_lf
  swing1LimitDegrees=70
  swing2LimitDegrees=45
  swing2Motion=Locked
  twistLimitDegrees=45
  twistMotion=Locked
 }
 {
  bone1Name=dyn_lapel_rt
  bone2Name=base_lapel_rt
  swing1LimitDegrees=70
  swing2LimitDegrees=45
  swing2Motion=Locked
  twistLimitDegrees=45
  twistMotion=Locked
 }
]
