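# Animation Blueprint for the Fermium Straw female body's torso, adding secondary motion to the arm band, arm dynamic bones, zipper and collar.
# Pose copied from the attached parent runs through a Control Rig, then two parallel AnimDynamics and collar RigidBody chains, joined by a bool blend.
# The chains differ in sim-space settings, one using the "FE" variables.
# Events track hang-gliding state and update a gravity override from a parameter asset.
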AnimBlueprint /BRCosmetics/Characters/Player/Female/Medium/Bodies/F_MED_FermiumStraw/Meshes/F_MED_FermiumStraw_Torso_AnimBP
target_skeleton=&/F_MED_FermiumStraw_Skeleton
graphs [
 AnimGraph {nodes [
   Output Pose(Result=Use cached pose 'CRCache_Dyn'()) {layerGroup=None}
   n1 = Control Rig!(Source=Copy Pose From Mesh() {bUseAttachedParent=true}, Alpha, isHangGliding:bool=false) -> (Pose) {controlRigAssetReference {controlRigAsset=&/F_MED_FermiumStraw_Torso_CtrlRig}}
   Save cached pose 'CRCache_Input'(Pose=@n1) {cacheName=CRCache_Input}
   n2 = Anim Dynamics - Bone: base_arm_band!(
    ComponentPose=Local To Component(LocalPose=Use cached pose 'CRCache_Input'())
    Alpha=1
    SimSpaceSettings:Anim Phys Sim Space Settings Structure=Get Arm Band Sim Space Settings() -> (Arm Band Sim Space Settings:Anim Phys Sim Space Settings Structure)
    ExternalForce="0,0,0"
   ) -> (Pose) {
    alpha=0.9
    angularDampingOverride=0.9
    angularSpringConstant=7500
    bAngularSpring=true
    bGravityOverrideInSimSpace=true
    bOverrideAngularDamping=true
    bUseGravityOverride=true
    numSolverIterationsPostUpdate=4
    numSolverIterationsPreUpdate=16
    simulationSpace=RootRelative
    alphaBoolBlend {blendOption=Linear}
    boundBone {boneName=base_arm_band}
    componentAppliedLinearAccClamp=<1400 1400 1400>
    componentLinearAccScale=<1 1 1>
    componentLinearVelScale=<1 1 1>
    gravityOverride {z=-980}
    physicsBodyDefinitions [{
      sphereCollisionRadius=0
      boundBone {boneName=base_arm_band}
      boxExtents=<30 30 30>
      constraintSetup {angularTargetAxis=AxisZ, twistAxis=AxisY, angularLimitsMax=<20 15 20>, angularLimitsMin=<-20 -15 -20>, angularTarget=<0 0 1>}
      localJointOffset=<0 -2 0>
     }]
    relativeSpaceBone {boneName=upperarm_r}
   }
   n3 = Anim Dynamics - Bone: dyn_arm_upper!(ComponentPose=@n2, Alpha=1, SimSpaceSettings:Anim Phys Sim Space Settings Structure=Get Arm Band Sim Space Settings(), ExternalForce="0,0,0") -> (Pose) {
    angularDampingOverride=0.99
    angularSpringConstant=10000
    bAngularSpring=true
    bOverrideAngularDamping=true
    bUseGravityOverride=true
    gravityScale=0.35
    simulationSpace=RootRelative
    alphaBoolBlend {blendOption=Linear}
    boundBone {boneName=dyn_arm_upper}
    componentAppliedLinearAccClamp=<1400 1400 1400>
    componentLinearAccScale=<1 1 1>
    componentLinearVelScale=<1 1 1>
    gravityOverride {z=-980}
    physicsBodyDefinitions [{
      boundBone {boneName=dyn_arm_upper}
      boxExtents=<20 30 30>
      constraintSetup {angularLimitsMax=<45 45 60>, angularLimitsMin=<-45 -45 -15>, angularTarget=<1 0 0>}
      localJointOffset=<8 2 -2>
     }]
    relativeSpaceBone {boneName=calf_r}
   }
   n4 = Anim Dynamics - Bone: dyn_arm_lower!(ComponentPose=@n3, Alpha=1, SimSpaceSettings:Anim Phys Sim Space Settings Structure=Get Arm Band Sim Space Settings(), ExternalForce="0,0,0") -> (Pose) {
    angularDampingOverride=0.99
    angularSpringConstant=10000
    bAngularSpring=true
    bOverrideAngularDamping=true
    bUseGravityOverride=true
    gravityScale=0.35
    simulationSpace=RootRelative
    alphaBoolBlend {blendOption=Linear}
    boundBone {boneName=dyn_arm_lower}
    componentAppliedLinearAccClamp=<1400 1400 1400>
    componentLinearAccScale=<1 1 1>
    componentLinearVelScale=<1 1 1>
    gravityOverride {z=-980}
    physicsBodyDefinitions [{
      boundBone {boneName=dyn_arm_lower}
      boxExtents=<20 30 30>
      constraintSetup {angularLimitsMax=<5 5 5>, angularLimitsMin=<-5 -5 -5>, angularTarget=<1 0 0>}
      localJointOffset=<8 -4 3>
     }]
    relativeSpaceBone {boneName=calf_r}
   }
   n5 = Anim Dynamics - Bone: dyn_zipper_upper!(
    ComponentPose=@n4
    Alpha=1
    GravityOverride:Vector="0,0,-980"
    SimSpaceSettings:Anim Phys Sim Space Settings Structure=Get Zipper Sim Space Settings() -> (Zipper Sim Space Settings:Anim Phys Sim Space Settings Structure)
   ) -> (Pose) {
    angularDampingOverride=1
    angularSpringConstant=9000
    bAngularSpring=true
    bOverrideAngularDamping=true
    gravityScale=0.35
    simulationSpace=RootRelative
    boundBone {boneName=dyn_zipper_upper}
    componentAppliedLinearAccClamp=<3000 3000 3000>
    componentLinearAccScale=<0.4 0.4 0.4>
    componentLinearVelScale=<0.4 0.4 0.4>
    physicsBodyDefinitions [{
      boundBone {boneName=dyn_zipper_upper}
      boxExtents=<8 8 20>
      constraintSetup {angularLimitsMax=<0 20 70>, angularLimitsMin=<0 -20 -5>, angularTarget=<1 0 0>}
      localJointOffset=<6 0 0>
     }]
    simSpaceSettings {maxAngularAcceleration=20, maxAngularVelocity=10, simSpaceAngularAlpha=1}
   }
   n6 = RigidBody!(
    ComponentPose=@n5
    Alpha
    OverrideWorldGravity:Vector="0,0,-980"
    SimSpaceSettings:Sim Space Settings Structure=Get Collar Sim Space Settings() -> (Collar Sim Space Settings:Sim Space Settings Structure)
   ) -> (Pose) {
    bOverrideWorldGravity=true
    overridePhysicsAsset=&/F_MED_FermiumStraw_Collar_PhysicsAsset
    simulationSpace=BaseBoneSpace
    baseBoneRef {boneName=spine_05}
    componentAppliedLinearAccClamp=<3000 3000 3000>
    simSpaceSettings {
     dampingAlpha=0
     maxAngularAcceleration=7
     maxAngularVelocity=7
     maxLinearAcceleration=500
     maxLinearVelocity=500
     velocityScaleZ=0.3
     worldAlpha=1
    }
   }
   n7 = Anim Dynamics - Bone: base_arm_band(
    ComponentPose=Local To Component(LocalPose=Use cached pose 'CRCache_Input'())
    Alpha=1
    SimSpaceSettings=Get FE Arm Band Sim Space Settings() -> (FE Arm Band Sim Space Settings:Anim Phys Sim Space Settings Structure)
    ExternalForce="0,0,0"
   ) {
    alpha=0.9
    angularDampingOverride=0.9
    angularSpringConstant=7500
    bAngularSpring=true
    bGravityOverrideInSimSpace=true
    bOverrideAngularDamping=true
    bUseGravityOverride=true
    numSolverIterationsPostUpdate=4
    numSolverIterationsPreUpdate=16
    simulationSpace=RootRelative
    alphaBoolBlend {blendOption=Linear}
    boundBone {boneName=base_arm_band}
    componentAppliedLinearAccClamp=<1400 1400 1400>
    componentLinearAccScale=<1 1 1>
    componentLinearVelScale=<1 1 1>
    gravityOverride {z=-980}
    physicsBodyDefinitions [{
      sphereCollisionRadius=0
      boundBone {boneName=base_arm_band}
      boxExtents=<30 30 30>
      constraintSetup {angularTargetAxis=AxisZ, twistAxis=AxisY, angularLimitsMax=<20 15 20>, angularLimitsMin=<-20 -15 -20>, angularTarget=<0 0 1>}
      localJointOffset=<0 -2 0>
     }]
    relativeSpaceBone {boneName=upperarm_r}
   }
   n8 = Anim Dynamics - Bone: dyn_arm_upper(ComponentPose=@n7, Alpha=1, SimSpaceSettings=Get FE Arm Band Sim Space Settings(), ExternalForce="0,0,0") {
    angularDampingOverride=0.99
    angularSpringConstant=10000
    bAngularSpring=true
    bOverrideAngularDamping=true
    bUseGravityOverride=true
    gravityScale=0.35
    simulationSpace=RootRelative
    alphaBoolBlend {blendOption=Linear}
    boundBone {boneName=dyn_arm_upper}
    componentAppliedLinearAccClamp=<1400 1400 1400>
    componentLinearAccScale=<1 1 1>
    componentLinearVelScale=<1 1 1>
    gravityOverride {z=-980}
    physicsBodyDefinitions [{
      boundBone {boneName=dyn_arm_upper}
      boxExtents=<20 30 30>
      constraintSetup {angularLimitsMax=<45 45 60>, angularLimitsMin=<-45 -45 -15>, angularTarget=<1 0 0>}
      localJointOffset=<8 2 -2>
     }]
    relativeSpaceBone {boneName=calf_r}
   }
   n9 = Anim Dynamics - Bone: dyn_arm_lower(ComponentPose=@n8, Alpha=1, SimSpaceSettings=Get FE Arm Band Sim Space Settings(), ExternalForce="0,0,0") {
    angularDampingOverride=0.99
    angularSpringConstant=10000
    bAngularSpring=true
    bOverrideAngularDamping=true
    bUseGravityOverride=true
    gravityScale=0.35
    simulationSpace=RootRelative
    alphaBoolBlend {blendOption=Linear}
    boundBone {boneName=dyn_arm_lower}
    componentAppliedLinearAccClamp=<1400 1400 1400>
    componentLinearAccScale=<1 1 1>
    componentLinearVelScale=<1 1 1>
    gravityOverride {z=-980}
    physicsBodyDefinitions [{
      boundBone {boneName=dyn_arm_lower}
      boxExtents=<20 30 30>
      constraintSetup {angularLimitsMax=<5 5 5>, angularLimitsMin=<-5 -5 -5>, angularTarget=<1 0 0>}
      localJointOffset=<8 -4 3>
     }]
    relativeSpaceBone {boneName=calf_r}
   }
   n10 = Anim Dynamics - Bone: dyn_zipper_upper(
    ComponentPose=@n9
    Alpha=1
    GravityOverride="0,0,-980"
    SimSpaceSettings=Get FE Zipper Sim Space Settings() -> (FE Zipper Sim Space Settings:Anim Phys Sim Space Settings Structure)
   ) {
    angularDampingOverride=1
    angularSpringConstant=9000
    bAngularSpring=true
    bOverrideAngularDamping=true
    gravityScale=0.35
    simulationSpace=RootRelative
    boundBone {boneName=dyn_zipper_upper}
    componentAppliedLinearAccClamp=<3000 3000 3000>
    componentLinearAccScale=<0.4 0.4 0.4>
    componentLinearVelScale=<0.4 0.4 0.4>
    physicsBodyDefinitions [{
      boundBone {boneName=dyn_zipper_upper}
      boxExtents=<8 8 20>
      constraintSetup {angularLimitsMax=<0 20 70>, angularLimitsMin=<0 -20 -5>, angularTarget=<1 0 0>}
      localJointOffset=<6 0 0>
     }]
    simSpaceSettings {maxAngularAcceleration=20, maxAngularVelocity=10, simSpaceAngularAlpha=1}
   }
   n11 = RigidBody(
    ComponentPose=@n10
    OverrideWorldGravity="0,0,-980"
    SimSpaceSettings=Get FE Collar Sim Space Settings() -> (FE Collar Sim Space Settings:Sim Space Settings Structure)
   ) {
    bOverrideWorldGravity=true
    overridePhysicsAsset=&F_MED_FermiumStraw_Collar_PhysicsAsset
    simulationSpace=BaseBoneSpace
    baseBoneRef {boneName=spine_05}
    componentAppliedLinearAccClamp=<3000 3000 3000>
    simSpaceSettings {
     dampingAlpha=0
     maxAngularAcceleration=7
     maxAngularVelocity=7
     maxLinearAcceleration=500
     maxLinearVelocity=500
     velocityScaleZ=0.3
     worldAlpha=1
    }
   }
   Save cached pose 'CRCache_Dyn'(Pose=Blend Poses by bool(BlendPose_0=Component To Local(ComponentPose=@n11), BlendPose_1=Component To Local(ComponentPose=@n6))) {cacheName=CRCache_Dyn}
  ]}
 EventGraph {nodes [
   Event OnParachuteOpened()
   > Switch on EGliderType(Selection=Get PlayerGliderType() -> (PlayerGliderType=HangGlider))
   > Set isHangGliding(isHangGliding:bool=true) -> (Output_Get:bool)
   Event OnParachuteClosed()
   > Set isHangGliding(isHangGliding=false)
   Event BlueprintUpdateAnimation()
   > n1 = "Gravity Override State Function (Parameter Asset) (Nativized)"(
    CharacterPartAnimInstance=Self-Reference()
    GameTimeInSeconds=GetGameTimeInSeconds()
    Parameters=&/F_MED_FermiumStraw_Parameters
   )
   > Set GravityOverride(GravityOverride=@n1)
  ]}
]
variables [
 Arm Band Sim Space Settings {
  type=AnimPhysSimSpaceSettings
  default {maxAngularAcceleration=10000, maxAngularVelocity=10000, simSpaceAngularAlpha=0, externalAngularVelocity=<0 0 0>}
 }
 Collar Sim Space Settings {
  type=SimSpaceSettings
  default {
   dampingAlpha=1
   maxAngularAcceleration=10000
   maxAngularVelocity=10000
   maxLinearAcceleration=10000
   maxLinearVelocity=10000
   velocityScaleZ=1
   worldAlpha=0
   externalLinearDragV=<0 0 0>
  }
 }
 FE Arm Band Sim Space Settings {
  type=AnimPhysSimSpaceSettings
  default {maxAngularAcceleration=10000, maxAngularVelocity=10000, simSpaceAngularAlpha=0.3, externalAngularVelocity=<0 0 0>}
 }
 FE Collar Sim Space Settings {
  type=SimSpaceSettings
  default {
   dampingAlpha=1
   maxAngularAcceleration=10000
   maxAngularVelocity=10000
   maxLinearAcceleration=10000
   maxLinearVelocity=10000
   velocityScaleZ=1
   worldAlpha=1
   externalLinearDragV=<0 0 0>
  }
 }
 FE Zipper Sim Space Settings {
  type=AnimPhysSimSpaceSettings
  default {maxAngularAcceleration=10000, maxAngularVelocity=10000, simSpaceAngularAlpha=0.05, externalAngularVelocity=<0 0 0>}
 }
 GravityOverride {type=Vector}
 isHangGliding {type=boolean}
 Zipper Sim Space Settings {
  type=AnimPhysSimSpaceSettings
  default {maxAngularAcceleration=10000, maxAngularVelocity=10000, simSpaceAngularAlpha=0, externalAngularVelocity=<0 0 0>}
 }
]
