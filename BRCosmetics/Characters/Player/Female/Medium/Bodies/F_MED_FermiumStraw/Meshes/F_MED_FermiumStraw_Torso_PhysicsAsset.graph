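# Physics asset for the torso mesh of the FermiumStraw female medium player body, giving its skeleton simplified physics bodies.
# Capsules cover the head, spine, pelvis, forearms, hands and feet, with extra angled capsules on the middle spine bones.
# Thigh, calf, upper arm and neck bodies list no shapes.
# Constraints are partial, linking some spine bones, the right arm chain and left wrist with uniform swing and twist limit angles.
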PhysicsAsset /BRCosmetics/Characters/Player/Female/Medium/Bodies/F_MED_FermiumStraw/Meshes/F_MED_FermiumStraw_Torso_PhysicsAsset
bodies {
 calf_l []
 calf_r []
 foot_l [{shapeType=Capsule, length=19.3444, radius=8.0344, center=<-4.1811 6.0027 -0.338>, rotation=<rot 0.0388 1.4701 -89.3982>}]
 foot_r [{shapeType=Capsule, length=19.3444, radius=8.0344, center=<4.1811 -6.0027 0.338>, rotation=<rot 0.0388 1.4701 90.6018>}]
 hand_l [{shapeType=Capsule, length=9.8756, radius=5.6983, center=<7.1748 -2.4688 -0.2928>, rotation=<rot 0 81.4683 -80.3867>}]
 hand_r [{shapeType=Capsule, length=9.8756, radius=5.6983, center=<-7.1748 2.4687 0.2928>, rotation=<rot 0 81.4683 -80.3867>}]
 head [{shapeType=Capsule, length=0, radius=18, center=<9.0698 0.0746 0>, rotation=<rot 0 -90 0>}]
 lowerarm_l [{shapeType=Capsule, length=9, radius=4, shapeName=lowerarm_l_capsule, center=<12.8994 0 6.4479>, rotation=<rot -83.8255 179.9995 -179.9995>}]
 lowerarm_r [{shapeType=Capsule, length=9, radius=4, shapeName=lowerarm_r_capsule, center=<-12.8994 0 -6.4479>, rotation=<rot -83.8255 179.9995 0.0005>}]
 neck_02 []
 pelvis [{
   shapeType=Capsule
   length=10.5966
   radius=12.3673
   shapeName=pelvis_capsule
   center=<-4.9751 -1.715 0.0626>
   rotation=<rot 2.3359 -90 0.2178>
  }]
 spine_01 [{shapeType=Capsule, length=10.6293, radius=9.3097, center=<-2.8472 -0.4882 -0.1981>}]
 spine_02 [
  {shapeType=Capsule, length=7.5219, radius=9.7317, center=<-2.6885 -1.2641 0>}
  {shapeType=Capsule, length=20.6875, radius=9.3125, center=<-1.8317 1.9362 0>, rotation=<rot 90 0 0>}
 ]
 spine_03 [
  {shapeType=Capsule, length=8.1863, radius=8.4854, center=<0.0574 0.4416 0>}
  {shapeType=Capsule, length=12.7188, radius=9.2812, center=<2.9052 -1.7618 3.1284>, rotation=<rot 56.1353 -4.9268 8.5359>}
 ]
 spine_04 [
  {shapeType=Capsule, length=9.0076, radius=10.3916, center=<1.5386 1.16 0>}
  {shapeType=Capsule, length=10.0312, radius=9.2812, center=<-5.5291 11.7737 0>, rotation=<rot 37.2082 -0.0001 0>}
 ]
 spine_05 [{shapeType=Capsule, length=9.678, radius=12.3571, center=<0.6969 1.083 0>}]
 thigh_l []
 thigh_r []
 upperarm_l []
 upperarm_r []
}
constraints [
 {bone1Name=spine_04, bone2Name=spine_02, swing1LimitDegrees=45, swing2LimitDegrees=45, twistLimitDegrees=45}
 {bone1Name=spine_05, bone2Name=spine_04, swing1LimitDegrees=45, swing2LimitDegrees=45, twistLimitDegrees=45}
 {bone1Name=hand_l, bone2Name=lowerarm_l, swing1LimitDegrees=45, swing2LimitDegrees=45, twistLimitDegrees=45}
 {bone1Name=upperarm_r, bone2Name=spine_05, swing1LimitDegrees=45, swing2LimitDegrees=45, twistLimitDegrees=45}
 {bone1Name=lowerarm_r, bone2Name=upperarm_r, swing1LimitDegrees=45, swing2LimitDegrees=45, twistLimitDegrees=45}
 {bone1Name=hand_r, bone2Name=lowerarm_r, swing1LimitDegrees=45, swing2LimitDegrees=45, twistLimitDegrees=45}
]
